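# Niagara cosmetic effect for the Alien Flora player skin that sheds small bursts of tinted mesh leaves from the character.
# A single Leaves emitter spawns at an offset from sampled User Skeletal Mesh positions, within a cylinder, with outward velocity from the cylinder normal.
# Leaves fall under randomized downward acceleration and drag while partly inheriting source movement.
# Curves slow their tumbling, pop then shrink their scale, and fade alpha late in life.
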NiagaraSystem /BRCosmetics/Characters/Player/Male/Medium/Bodies/M_MED_Alien_Flora/FX/NS_Skin_AlienFlora_FallenLeaves
dependencies {
 usedDataInterfaces [&NiagaraDataInterfaceCurve, &NiagaraDataInterfaceVectorCurve]
 usedDynamicInputs []
 usedModules [&SystemState]
}
emitters {Leaves {
  emitterUpdateScript [
   &EmitterState(Life Cycle Mode=::NewEnumerator1, Loop Behavior=::NewEnumerator1, Loop Delay=0.05, Loop Duration=1, UseLoopDelay=true)
   &SpawnBurst_Instantaneous(
    Spawn Count=&RandomRangeInt(Minimum=5, Maximum=10)
    Use Spawn Probability=false
    UseLoopCountLimit=false
   )
  ]
  particleSpawnScript [
   &InitializeParticle(
    Color=<rgba 5 0.1302 2.4839 1>
    Lifetime Max=1.5
    Lifetime Min=0.5
    Lifetime Mode=::NewEnumerator1
    Mesh Scale Mode=::NewEnumerator1
    Mesh Uniform Scale Max=1.2
    Mesh Uniform Scale Min=0.6
    UsePositionOffset=false
    Write Material Random=true
   )
   &InitialMeshOrientation(
    Rotation=&RandomRangeVector(Minimum=<0 0 0>, Maximum=<1 1 1>)
    Use Orientation Vector=true
    Use Rotation Vector=true
   )
   &SkeletalMeshLocation(Bone Sampling Mode=::NewEnumerator4, MaskSpawnGroup=false, Sampled Position Offset=<40 0 0>, Skeletal Mesh=$User.Skeletal Mesh)
   &CylinderLocation(Cylinder Height=60, Cylinder Radius=20, MaskSpawnGroup=false, Override Seed=false, UseOwnerScale=true)
   &AddVelocityFromPoint(
    Module.Use Falloff Distance=false
    Velocity Origin:Vector3f=&Subtract_Vector(A=$Particles.CylinderLocation.CylinderPosition, B=$Particles.CylinderLocation.CylinderNormal)
    Velocity Strength=&RandomRangeFloat(Minimum=50, Maximum=100)
   )
  ]
  particleUpdateScript [
   &ParticleState()
   &ScaleColor(
    Scale Alpha=&FloatFromCurve(FloatCurve=DataInterface() {propertyValues {Curve {keys [<at 0.7 1>,<at 1 0>]}}})
    ScaleA=true
    ScaleRGB=true
   )
   &UpdateMeshOrientation(Rotation Vector=&Multiply_Vector(
     A=&RandomRangeVector(Minimum=<0.3 0.3 0.3>, Maximum=<1 1 1>)
     B=&VectorFromCurve(VectorCurve=DataInterface() {propertyValues {XCurve {keys [<at 0 2>,<at 1 0>]}, YCurve {keys [<at 0 2>,<at 1 0>]}, ZCurve {keys [<at 0 2>,<at 1 0>]}}})
    ))
   &AccelerationForce(Acceleration=&RandomRangeVector(Minimum=<0 0 -200>, Maximum=<0 0 -400>))
   &Drag(Use Linear Drag=true, Use Rotational Drag=false)
   &InheritSourceMovement(Applied Position Scale=<0.7 0.7 0.7>, ApplyForce=false, ApplyPosition=true, ApplyPositionFalloff=false)
   &SolveForcesAndVelocity(Clamp Velocity=false, Limit Acceleration=false)
   &ScaleMeshSize(Scale Factor=&VectorFromFloat(Value=&FloatFromCurve(FloatCurve=DataInterface() {propertyValues {Curve {keys [
          <at 0 0 RCIM_Cubic>
          <at 0.05 1 RCIM_Cubic>
          {arriveTangent=-1.0526, interpMode=RCIM_Cubic, leaveTangent=-1.0526, time=0.27, value=0.85}
          <at 1 0 RCIM_Cubic>
         ]}}})))
  ]
  rendererClasses [&NiagaraMeshRendererProperties]
  renderers [{rendererClass=&NiagaraMeshRendererProperties, rendererIndex=0}]
 }}
summary {
 systemName=NS_Skin_AlienFlora_FallenLeaves
 emitters [{emitterName=Leaves, rendererClasses [&NiagaraMeshRendererProperties]}]
 userVariables [User.Skeletal Mesh {
   type=NiagaraDataInterfaceSkeletalMesh
   defaultValue {
    struct=NiagaraExt_VariableValue_DataInterface
    value {dataInterface=&self:NiagaraDataInterfaceSkeletalMesh_0}
   }
  }]
}
