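# Additive master material for the eyes of the Alien Flora player character, drawn as tinted emissive shapes that switch off during blinks.
# The eye texture's red channel drives opacity and, multiplied by a color parameter, the emissive color.
# The R_blink_pose scalar passes through a Sign and Saturate step threshold, hiding the glow once the blink value reaches about 0.1.
# The result runs through the shared character PostFX function, fed by PreFX, then the cinematic post-FX function.
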Material /BRCosmetics/Characters/Player/Male/Medium/Bodies/M_MED_Alien_Flora/Materials/M_MED_Alien_Flora_Eyes_MASTER
graph {
 properties {
  bUseMaterialAttributes=true
  bUsedWithMeshDeformer=true
  bUsedWithSkeletalMesh=true
  blendMode=BLEND_Additive
  pixelDepthOffsetMode=PDOM_Legacy
  usedShadingModels=Default Lit
 }
 nodes [
  n1 = MaterialFunctionCall() {materialFunction=&BaseFN_CharacterFunction_PreFX}
  n2 = TextureSample() {texture=/BRCosmetics/Characters/Player/Male/Medium/Bodies/M_MED_Alien_Flora/Textures/T_M_MED_Alien_Flora_Eyes_D}
  n3 = Multiply(A=@n2.R, B=VectorParameter() {parameterName=Eyes Color Multiplier, defaultValue=<rgba 0.1025 0.7449 0.77 1>}.RGB)
  n4 = OneMinus(None=ScalarParameter() {parameterName=R_blink_pose})
  n5 = Saturate(None=Sign(None=Subtract(A=@n4) {constB=0.9}))
  n6 = MaterialFunctionCall(
   AO=@n1.AO
   Distance Blend=@n1.AO
   Fresnel=@n1.AO
   Gmap=@n1.AO
   Pre Skinned Local Position=@n1.AO
   Material=SetMaterialAttributes!(MaterialAttributes, Emissive Color=Multiply(A=@n3, B=@n5), Opacity=@n2.R) {attributeSetTypes [#1, #2]}
  ) {materialFunction=&BaseFN_CharacterFunction_PostFX}
  MaterialFunctionCall(In=@n6.Result) {materialFunction=&CINE_MF_CharPostFX}
 ]
}
