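# Master material for the Alien Flora character body, adding petal color decay and animated blinking mini eyes to the shared character FX function chain.
# Damage-transition curve parameters and a decay mask feed petal recoloring, alongside roughness and time-of-day emissive adjustments.
# Vertex color encodes each eye's direction, mask and random phase for world-position-offset wobble, blinking and timed hiding.
# PreviousFrameSwitch prevents motion blur when eyes vanish.
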Material /BRCosmetics/Characters/Player/Male/Medium/Bodies/M_MED_Alien_Flora/Materials/M_MED_Alien_Flora_MASTER
graph {
 properties {
  bAutomaticallySetUsageInEditor=false
  bUseMaterialAttributes=true
  bUsedWithMeshDeformer=true
  bUsedWithMorphTargets=true
  bUsedWithSkeletalMesh=true
  pixelDepthOffsetMode=PDOM_Legacy
  refractionMethod=RM_IndexOfRefraction
  shadingModel=MSM_FromMaterialExpression
  subsurfaceProfile=&SS_HeroSkin_02
  usedShadingModels=Default Lit | Subsurface | Clear Coat | Subsurface Profile | Cloth | Eye
 }
 nodes [
  n1 = Constant() {r=30}
  Add(A=@n1)
  n2 = MaterialFunctionCall() {materialFunction=&BaseFN_CharacterFunction_PreFX}
  n3 = GetMaterialAttributes(None=MaterialFunctionCall(
    PreskinnedPosition=@n2.Pre skinned Local Position
    Gmap=@n2.Pre skinned Local Position
    MaterialAttributes=@n2.Pre skinned Local Position
    AO=@n2.Pre skinned Local Position
   ) {materialFunction=&CharacterSkinFX}.Result) {attributeGetTypes [
    #1
    #2
    #3
    #4
   ]}
  n4 = SetMaterialAttributes!(
   MaterialAttributes=@n3.MaterialAttributes
   Base Color=NamedRerouteUsage()
   Emissive Color=NamedRerouteUsage()
   Roughness=NamedRerouteUsage()
   World Position Offset=NamedRerouteUsage()
  ) {attributeSetTypes [
    #1
    #2
    #3
    #5
   ]}
  n5 = MaterialFunctionCall(AO=@n2.AO, Distance Blend=@n2.AO, Fresnel=@n2.AO, Gmap=@n2.AO, Pre Skinned Local Position=@n2.AO, Material=@n4) {materialFunction=&BaseFN_CharacterFunction_PostFX}
  MaterialFunctionCall(In=@n5.Result) {materialFunction=&CINE_MF_CharPostFX}
  n6 = Add(
   A=Add(
    A=MaterialFunctionCall() {materialFunction=&ObjectPivotPoint}.Object Pivot Location
    B=Constant3Vector() {constant=<rgba 0 0 60 0>}
   )
   B=Multiply(A=WorldPosition().XYZ) {constB=-1}
  )
  n7 = Time() {bIgnorePause=true}
  n8 = VertexColor()
  n9 = OneMinus(None=Add(A=@n8.A) {constB=0.0001})
  n10 = Add(A=Sine(None=Add(A=@n7, B=@n9)) {period=3}) {constB=-0.95}
  n11 = Saturate(None=Sign(None=@n9))
  n12 = Time() {period=10}
  n13 = Constant() {r=20}
  n14 = Constant()
  n15 = Constant() {r=1}
  n16 = If(A=@n9, B=Divide(A=Fmod(A=@n12, B=@n13), B=@n13) {constB=10}, "A > B"=@n14, "A == B"=@n15, "A < B"=@n14) {equalsThreshold=0.1}
  n17 = Constant() {r=23}
  n18 = If(A=@n9, B=Divide(A=Fmod(A=@n12, B=@n17), B=@n17) {constB=12}, "A > B"=@n14, "A == B"=@n15, "A < B"=@n14) {equalsThreshold=0.1}
  n19 = LinearInterpolate(B=Saturate(None=Sign(None=@n10)), Alpha=Multiply(A=@n11, B=Max(A=@n16, B=@n18))) {constA=1}
  n20 = Multiply(A=@n19, B=Multiply(A=@n19, B=@n11))
  n21 = Add(A=Multiply(A=Time() {bIgnorePause=true}, B=ScalarParameter() {defaultValue=0.1, parameterName=Speed}), B=@n8.A)
  n22 = Transform(None=Normalize(VectorInput=ConstantBiasScale(None=@n8) {bias=-0.5, scale=2})) {transformSourceType=TRANSFORMSOURCE_Local}
  n23 = Constant3Vector() {constant=<rgba 0 0 1 0>}
  n24 = Add(A=Multiply(A=Sine(None=@n21) {period=0.7}, B=CrossProduct(A=@n22, B=@n23)), B=Multiply(A=Cosine(None=@n21) {period=0.3}, B=@n23))
  n25 = Multiply(A=Add(A=Multiply(A=@n24) {constB=0.8}, B=Multiply(A=@n22)), B=@n11)
  Multiply(
   A=PreviousFrameSwitch(Current Frame=LinearInterpolate(A=@n25, B=@n6, Alpha=@n20), Previous Frame=@n25)
   B=ScalarParameter() {parameterName=enable Blinky Eyes}
  )
  NamedRerouteDeclaration(None=@n3.Base Color) {name=Base Color Input, nodeColor=<rgba 0 0.3672 1 1>}
  NamedRerouteDeclaration(None=@n3.Emissive Color) {name=Emissive Color Input, nodeColor=<rgba 1 0.8033 0 1>}
  n26 = VertexColor()
  n27 = OneMinus(None=@n26.A)
  n28 = Subtract(A=@n27) {constB=0}
  NamedRerouteDeclaration(None=Saturate(None=Sign(None=@n28))) {name=Mini Eyes Mask, nodeColor=<rgba 0.4531 0.3 1 1>}
  n29 = ConstantBiasScale(None=@n26) {bias=-0.5, scale=2}
  n30 = NamedRerouteDeclaration(None=Transform(None=Normalize(VectorInput=@n29)) {transformSourceType=TRANSFORMSOURCE_Local}) {name=Mini Eyes Forward, nodeColor=<rgba 0.1763 1 0.2049 1>}
  n31 = Normalize(VectorInput=CrossProduct(A=@n30, B=Constant3Vector() {constant=<rgba 0 0 -1 0>}))
  NamedRerouteDeclaration(None=CrossProduct(A=@n30, B=NamedRerouteDeclaration(None=@n31) {name=Mini Eyes Right, nodeColor=<rgba 1 0.1978 0.2078 1>})) {name=Mini Eyes Up, nodeColor=<rgba 0.2299 0.2304 1 1>}
  NamedRerouteDeclaration(None=@n27) {name=Mini Eyes Random, nodeColor=<rgba 0.9453 0 1 1>}
  n32 = ChannelMaskParameter(None=TextureSample(Tex=TextureObjectParameter() {group=Skin FX, parameterName=SkinFX_Mask}).RGB) {parameterName=Petals Decay}
  NamedRerouteDeclaration(None=LinearInterpolate(A=@n32.R, B=OneMinus(None=@n32.R), Alpha=ScalarParameter() {parameterName=Invert Decay Mask})) {name=FXM_Blue, nodeColor=<rgba 0 0.206 1 1>}
  n33 = NamedRerouteUsage()
  n34 = Desaturation(None=@n33)
  n35 = Subtract(A=ComponentMask(None=@n33) {g=true}) {constB=0.05}
  n36 = LinearInterpolate(
   A=VectorParameter() {parameterName=Petals Color Override 2, defaultValue=<rgba 0.8646 0.4615 0.2809 1>}.RGB
   B=VectorParameter() {parameterName=Petals Color Override 1, defaultValue=<rgba 0.2417 0.5 0.1839 1>}.RGB
   Alpha=Saturate(None=Multiply(A=@n35) {constB=4})
  )
  NamedRerouteDeclaration(None=LinearInterpolate(A=@n33, B=Multiply(A=@n34, B=@n36), Alpha=NamedRerouteUsage())) {name=Base Color Out, nodeColor=<rgba 0 0.032 0.6094 1>}
  n37 = CurveAtlasRowParameter(CurveTime=CollectionParameter() {collection=&FortniteMaterialParameters, parameterName=TimeOfDay_0-1}) {
   atlas=/Game/Characters/Shared/CurveAtlas/AthenaCharacterCurveAtlas
   curve=/Game/Characters/Player/Male/Medium/Bodies/M_MED_Buff_Cat_Comic/Textures/RetroCartoonCurve
   defaultValue=18
   group=TOD_Adjustments
   parameterName=TODBrightnessAdjustmentCurve
  }
  n38 = Multiply(B=@n37.R) {constA=0.005, constB=0.005}
  n39 = OneMinus(None=Sign(None=NamedRerouteUsage()))
  n40 = NamedRerouteUsage()
  n41 = OneMinus(None=Fresnel() {baseReflectFraction=0, exponent=1})
  NamedRerouteDeclaration(None=Add(A=Multiply(A=@n38, B=@n39), B=Multiply(A=@n40, B=@n41))) {name=Emissive Color Out, nodeColor=<rgba 0.5208 0.3211 0 1>}
  n42 = Max(A=ScalarParameter() {parameterName=MaterialCurveA}, B=ScalarParameter() {parameterName=MaterialCurveA_Render})
  n43 = ScalarParameter() {defaultValue=1, parameterName=DamageTransitionExponent}
  NamedRerouteDeclaration(None=Multiply(A=Power(Base=@n42, Exp=@n43), B=NamedRerouteUsage())) {name=Petels Transition, nodeColor=<rgba 0.7812 1 0 1>}
  NamedRerouteDeclaration(None=@n3.Roughness) {name=Roughness Input, nodeColor=<rgba 0.3359 1 0 1>}
  n44 = NamedRerouteUsage()
  NamedRerouteDeclaration(None=LinearInterpolate(A=@n44, B=Multiply(A=@n44) {constB=1.3}, Alpha=NamedRerouteUsage())) {name=Roughness Output, nodeColor=<rgba 0.5234 1 0 1>}
  NamedRerouteDeclaration(None=@n3.Specular) {name=Specular Input, nodeColor=<rgba 0.3359 0 1 1>}
  n45 = NamedRerouteUsage()
  n46 = NamedRerouteUsage()
  n47 = Subtract(A=@n46) {constB=0.5}
  n48 = Add(A=Multiply(A=Time(), B=Sign(None=@n47)), B=Multiply(A=@n46) {constB=1.5})
  n49 = Add(A=Multiply(A=Sine(None=@n48) {period=1.6}, B=NamedRerouteUsage()), B=Multiply(A=Cosine(None=@n48) {period=0.8}, B=NamedRerouteUsage()))
  n50 = Multiply(A=VertexNormalWS(), B=Constant3Vector() {constant=<rgba 0.2 0.2 -1.3 0>}) {constB=-0.6}
  n51 = NamedRerouteUsage()
  n52 = Add(A=Time(), B=@n51)
  n53 = NamedRerouteUsage()
  n54 = ScalarParameter() {defaultValue=0.005, parameterName=Ease}
  n55 = Time()
  n56 = Multiply(A=Divide(A=Fmod(A=@n55, B=@n1), B=@n1)) {constB=1.26}
  n57 = Add(A=@n53, B=ScalarParameter() {defaultValue=0.26, parameterName="Visibility Duration (0-1)"})
  n58 = SmoothStep(Min=@n57, Max=Subtract(A=@n57, B=@n54), Value=@n56)
  n59 = Min(A=SmoothStep(Min=@n53, Max=Add(A=@n53, B=@n54), Value=@n56), B=@n58)
  n60 = LinearInterpolate(A=Power(Base=Sine(None=@n52) {period=3}) {constExponent=200}, Alpha=OneMinus(None=@n59))
  n61 = Subtract(A=@n51) {constB=0.5}
  n62 = Multiply(A=Multiply(A=NamedRerouteUsage(), B=Sign(None=@n61))) {constB=1.5}
  n63 = OneMinus(None=@n58)
  n64 = Add(
   A=Add(A=Multiply(A=@n45) {constA=0.5, constB=0.5}, B=Multiply(A=@n49) {constB=0.2})
   B=Add(A=Multiply(A=@n50, B=@n60), B=Multiply(A=@n62, B=@n63))
  )
  n65 = WorldPosition()
  n66 = Sign(None=@n59)
  n67 = PreviousFrameSwitch(Current Frame=LinearInterpolate(A=@n64, B=Multiply(A=@n65.XYZ) {constB=-1}, Alpha=OneMinus(None=@n66)), Previous Frame=@n64) {desc=Prevents motion blur}
  n68 = NamedRerouteUsage()
  NamedRerouteDeclaration(None=Multiply(A=Multiply(A=@n67, B=@n68), B=ScalarParameter() {parameterName=enable Blinky Eyes})) {name=WPO Output, nodeColor=<rgba 0 0.7891 1 1>}
 ]
}
