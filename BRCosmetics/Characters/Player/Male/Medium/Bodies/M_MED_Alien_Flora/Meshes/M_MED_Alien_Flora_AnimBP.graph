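# Animation Blueprint for the M_MED_Alien_Flora cosmetic player body, adding secondary motion to its flowers, petals and vines.
# It copies pose and curves from the attached parent mesh, then runs a long chain of single-bone Anim Dynamics nodes with damping, gravity overrides and tight angular limits.
# Weighted multi-target constraints position vine and petal base bones against pelvis, spine and leg bones.
# A Dynamics_Alpha variable scales the inner waist and shoulder flower simulation.
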AnimBlueprint /BRCosmetics/Characters/Player/Male/Medium/Bodies/M_MED_Alien_Flora/Meshes/M_MED_Alien_Flora_AnimBP
target_skeleton=&/M_MED_Alien_Flora_Skeleton
graphs [
 AnimGraph {nodes [
   Anim Dynamics - Bone: dyn_leg_vine_r_parent_b(Alpha=1, ExternalForce="0,0,0") {
    angularDampingOverride=0.999
    bGravityOverrideInSimSpace=true
    bOverrideAngularDamping=true
    bOverrideLinearDamping=true
    bUseGravityOverride=true
    linearDampingOverride=0.999
    simulationSpace=RootRelative
    alphaBoolBlend {blendOption=Linear}
    alphaScaleBiasClamp {bias=0.2, scale=0.8}
    boundBone {boneName=dyn_leg_vine_r_parent_b}
    componentAppliedLinearAccClamp=<100 100 100>
    componentLinearAccScale=<0.15 0.15 0.15>
    componentLinearVelScale=<0.15 0.15 0.15>
    gravityOverride=<0.85 0.85 0.85>
    physicsBodyDefinitions [{
      sphereCollisionRadius=0
      boundBone {boneName=dyn_leg_vine_r_parent_b}
      boxExtents=<5 4 1.5>
      constraintSetup {angularLimitsMax=<0 20 0>, angularLimitsMin=<0 -10 0>, angularTarget=<0 0 0>}
      localJointOffset=<1 0 0>
     }]
    relativeSpaceBone {boneName=root}
   }
   n1 = Anim Dynamics - Bone: dyn_chest_flower_mid_01(
    ComponentPose=Local To Component(LocalPose=Copy Pose From Mesh() {bCopyCurves=true, bUseAttachedParent=true})
    Alpha=1
    ExternalForce="0,0,0"
   ) {
    angularDampingOverride=0.9999
    bGravityOverrideInSimSpace=true
    bOverrideAngularDamping=true
    bOverrideLinearDamping=true
    bUseGravityOverride=true
    linearDampingOverride=0.9999
    simulationSpace=BoneRelative
    alphaBoolBlend {blendOption=Linear}
    boundBone {boneName=dyn_chest_flower_mid_01}
    componentAppliedLinearAccClamp=<100 100 100>
    componentLinearAccScale=<0.15 0.15 0.15>
    componentLinearVelScale=<0.15 0.15 0.15>
    gravityOverride=<0.7 0.7 0.7>
    physicsBodyDefinitions [{
      sphereCollisionRadius=0
      boundBone {boneName=dyn_chest_flower_mid_01}
      boxExtents=<5 6 1>
      constraintSetup {angularLimitsMax=<0 10 10>, angularLimitsMin=<0 -10 -10>, angularTarget=<0 0 0>}
      localJointOffset=<2 0 0>
     }]
    relativeSpaceBone {boneName=pelvis}
   }
   n2 = Anim Dynamics - Bone: dyn_chest_flower_mid_02(ComponentPose=@n1, Alpha=1, ExternalForce="0,0,0") {
    angularDampingOverride=0.999
    bGravityOverrideInSimSpace=true
    bOverrideAngularDamping=true
    bOverrideLinearDamping=true
    bUseGravityOverride=true
    linearDampingOverride=0.999
    simulationSpace=RootRelative
    alphaBoolBlend {blendOption=Linear}
    boundBone {boneName=dyn_chest_flower_mid_02}
    componentAppliedLinearAccClamp=<3000 3000 3000>
    componentLinearAccScale=<0.25 0.25 0.5>
    componentLinearVelScale=<0.25 0.25 0.5>
    gravityOverride=<0.85 0.85 0.85>
    physicsBodyDefinitions [{
      sphereCollisionRadius=0
      boundBone {boneName=dyn_chest_flower_mid_02}
      boxExtents=<4 4 2>
      constraintSetup {angularLimitsMax=<0 5 15>, angularLimitsMin=<0 -5 -10>, angularTarget=<0 0 0>}
      localJointOffset=<1 0 0>
     }]
    relativeSpaceBone {boneName=root}
   }
   n3 = Anim Dynamics - Bone: dyn_chest_flower_mid_03(ComponentPose=@n2, Alpha=1, ExternalForce="0,0,0") {
    angularDampingOverride=0.999
    bGravityOverrideInSimSpace=true
    bOverrideAngularDamping=true
    bOverrideLinearDamping=true
    bUseGravityOverride=true
    linearDampingOverride=0.999
    simulationSpace=RootRelative
    alphaBoolBlend {blendOption=Linear}
    boundBone {boneName=dyn_chest_flower_mid_03}
    componentAppliedLinearAccClamp=<3000 3000 3000>
    componentLinearAccScale=<0.25 0.25 0.5>
    componentLinearVelScale=<0.25 0.25 0.5>
    gravityOverride=<0.85 0.85 0.85>
    physicsBodyDefinitions [{
      sphereCollisionRadius=0
      boundBone {boneName=dyn_chest_flower_mid_03}
      boxExtents=<5 4 1>
      constraintSetup {angularLimitsMax=<0 10 5>, angularLimitsMin=<0 -10 -5>, angularTarget=<0 0 0>}
      localJointOffset=<1.5 0 0>
     }]
    relativeSpaceBone {boneName=root}
   }
   n4 = Anim Dynamics - Bone: dyn_leg_vine_r_parent_a(ComponentPose=@n3, Alpha=1, ExternalForce="0,0,0") {
    angularDampingOverride=0.999
    bGravityOverrideInSimSpace=true
    bOverrideAngularDamping=true
    bOverrideLinearDamping=true
    bUseGravityOverride=true
    linearDampingOverride=0.999
    simulationSpace=RootRelative
    alphaBoolBlend {blendOption=Linear}
    alphaScaleBiasClamp {bias=0.2, scale=0.8}
    boundBone {boneName=dyn_leg_vine_r_parent_a}
    componentAppliedLinearAccClamp=<100 100 100>
    componentLinearAccScale=<0.25 0.25 0.5>
    componentLinearVelScale=<0.25 0.25 0.5>
    gravityOverride=<0.85 0.85 0.85>
    physicsBodyDefinitions [{
      sphereCollisionRadius=0
      boundBone {boneName=dyn_leg_vine_r_parent_a}
      boxExtents=<6 4 2>
      constraintSetup {angularLimitsMax=<0 30 15>, angularLimitsMin=<0 -20 -15>, angularTarget=<0 0 0>}
      localJointOffset=<2 0 0>
     }]
    relativeSpaceBone {boneName=root}
   }
   n5 = Anim Dynamics - Bone: dyn_leg_vine_r_parent_c(ComponentPose=@n4, Alpha=0.8, ExternalForce="0,0,0") {
    angularDampingOverride=0.9999
    bGravityOverrideInSimSpace=true
    bOverrideAngularDamping=true
    bOverrideLinearDamping=true
    bUseGravityOverride=true
    linearDampingOverride=0.9999
    simulationSpace=RootRelative
    alphaBoolBlend {blendOption=Linear}
    alphaScaleBiasClamp {bias=0.2, scale=0.8}
    boundBone {boneName=dyn_leg_vine_r_parent_c}
    componentAppliedLinearAccClamp=<100 100 100>
    componentLinearAccScale=<0.15 0.15 0.15>
    componentLinearVelScale=<0.15 0.15 0.15>
    gravityOverride=<0.7 0.7 0.7>
    physicsBodyDefinitions [{
      sphereCollisionRadius=0
      boundBone {boneName=dyn_leg_vine_r_parent_c}
      boxExtents=<8 12 2>
      constraintSetup {angularLimitsMax=<0 20 20>, angularLimitsMin=<0 -15 -10>, angularTarget=<0 0 0>}
      localJointOffset=<1.5 0 0>
     }]
    relativeSpaceBone {boneName=root}
   }
   n6 = Anim Dynamics - Bone: dyn_leg_vine_r_parent_d(ComponentPose=@n5, Alpha=1, ExternalForce="0,0,0") {
    angularDampingOverride=0.9999
    bAngularSpring=true
    bGravityOverrideInSimSpace=true
    bLinearSpring=true
    bOverrideAngularDamping=true
    bOverrideLinearDamping=true
    bUseGravityOverride=true
    gravityScale=0.25
    linearDampingOverride=0.9999
    simulationSpace=RootRelative
    alphaBoolBlend {blendOption=Linear}
    alphaScaleBiasClamp {bias=0.2, scale=0.8}
    boundBone {boneName=dyn_leg_vine_r_parent_d}
    componentAppliedLinearAccClamp=<100 100 100>
    componentLinearAccScale=<0.15 0.15 0.15>
    componentLinearVelScale=<0.15 0.15 0.15>
    gravityOverride=<0.7 0.7 0.7>
    physicsBodyDefinitions [{
      sphereCollisionRadius=0
      boundBone {boneName=dyn_leg_vine_r_parent_d}
      boxExtents=<8 10 1.5>
      constraintSetup {angularLimitsMax=<0 20 30>, angularLimitsMin=<0 -10 -20>, angularTarget=<0 0 0>}
      localJointOffset=<1.5 0 0>
     }]
    relativeSpaceBone {boneName=root}
   }
   n7 = "Constraint - dyn_leg_vine_r_base_b (PPP)"!(ComponentPose=@n6, Alpha=1, ConstraintWeights_0=0.3003, ConstraintWeights_1:float=0.0055, ConstraintWeights_2:float=0.6942) -> (Pose) {
    alphaBoolBlend {blendOption=Linear}
    boneToModify {boneName=dyn_leg_vine_r_base_b}
    constraintSetup [{targetBone {boneName=thigh_r}}, {targetBone {boneName=calf_twist_02_r}}, {targetBone {boneName=calf_r}}]
    constraintWeights [1,1,1]
   }
   n8 = "Constraint - dyn_leg_petals_l_base_c (PPPP)"!(
    ComponentPose=@n7
    Alpha=1
    ConstraintWeights_0=0.4039
    ConstraintWeights_1:float=0.5587
    ConstraintWeights_2:float=0.0003
    ConstraintWeights_3:float=0.0372
   ) -> (Pose) {
    alphaBoolBlend {blendOption=Linear}
    boneToModify {boneName=dyn_leg_petals_l_base_c}
    constraintSetup [
     {targetBone {boneName=pelvis}}
     {targetBone {boneName=thigh_twist_01_l}}
     {targetBone {boneName=thigh_twist_01_r}}
     {targetBone {boneName=spine_01}}
    ]
    constraintWeights [1,1,1,1]
   }
   n9 = "Constraint - dyn_pelvis_petals_l_base_a (PPP)"!(ComponentPose=@n8, Alpha=1, ConstraintWeights_0=0.7661, ConstraintWeights_1:float=0.0633, ConstraintWeights_2:float=0.1706) -> (Pose) {
    alphaBoolBlend {blendOption=Linear}
    boneToModify {boneName=dyn_pelvis_petals_l_base_a}
    constraintSetup [{targetBone {boneName=pelvis}}, {targetBone {boneName=thigh_twist_01_l}}, {targetBone {boneName=spine_01}}]
    constraintWeights [1,1,1]
   }
   n10 = "Constraint - dyn_pelvis_petals_l_base_b (PPPP)"!(
    ComponentPose=@n9
    Alpha=1
    ConstraintWeights_0=0.7138
    ConstraintWeights_1:float=0.0104
    ConstraintWeights_2:float=0.0253
    ConstraintWeights_3:float=0.2505
   ) -> (Pose) {
    alphaBoolBlend {blendOption=Linear}
    boneToModify {boneName=dyn_pelvis_petals_l_base_b}
    constraintSetup [{targetBone {boneName=pelvis}}, {targetBone {boneName=spine_02}}, {targetBone {boneName=thigh_twist_01_l}}, {targetBone {boneName=spine_01}}]
    constraintWeights [1,1,1,1]
   }
   n11 = Anim Dynamics - Bone: dyn_leg_petals_l_parent_a(ComponentPose=@n10, Alpha=1, ExternalForce="0,0,0") {
    angularDampingOverride=0.9999
    bGravityOverrideInSimSpace=true
    bOverrideAngularDamping=true
    bOverrideLinearDamping=true
    bUseGravityOverride=true
    linearDampingOverride=0.9999
    simulationSpace=RootRelative
    alphaBoolBlend {blendOption=Linear}
    alphaScaleBiasClamp {bias=0.2, scale=0.8}
    boundBone {boneName=dyn_leg_petals_l_parent_a}
    componentAppliedLinearAccClamp=<100 100 100>
    componentLinearAccScale=<0.15 0.15 0.15>
    componentLinearVelScale=<0.15 0.15 0.15>
    gravityOverride=<0.7 0.7 0.7>
    physicsBodyDefinitions [{
      sphereCollisionRadius=0
      boundBone {boneName=dyn_leg_petals_l_parent_a}
      boxExtents=<6 8 2>
      constraintSetup {angularLimitsMax=<0 30 5>, angularLimitsMin=<0 -20 -5>, angularTarget=<0 0 0>}
      localJointOffset=<1.5 0 0>
     }]
    relativeSpaceBone {boneName=root}
   }
   n12 = Anim Dynamics - Bone: dyn_leg_petals_l_parent_b(ComponentPose=@n11, Alpha=1, ExternalForce="0,0,0") {
    angularDampingOverride=0.9999
    bGravityOverrideInSimSpace=true
    bOverrideAngularDamping=true
    bOverrideLinearDamping=true
    bUseGravityOverride=true
    linearDampingOverride=0.9999
    simulationSpace=RootRelative
    alphaBoolBlend {blendOption=Linear}
    alphaScaleBiasClamp {bias=0.2, scale=0.8}
    boundBone {boneName=dyn_leg_petals_l_parent_b}
    componentAppliedLinearAccClamp=<100 100 100>
    componentLinearAccScale=<0.15 0.15 0.15>
    componentLinearVelScale=<0.15 0.15 0.15>
    gravityOverride=<0.7 0.7 0.7>
    physicsBodyDefinitions [{
      sphereCollisionRadius=0
      boundBone {boneName=dyn_leg_petals_l_parent_b}
      boxExtents=<5 5 1.5>
      constraintSetup {angularLimitsMax=<0 30 20>, angularLimitsMin=<0 -30 -20>, angularTarget=<0 0 0>}
      localJointOffset=<1 0 0>
     }]
    relativeSpaceBone {boneName=root}
   }
   n13 = Anim Dynamics - Bone: dyn_leg_petals_l_parent_c(ComponentPose=@n12, Alpha=1, ExternalForce="0,0,0") {
    angularDampingOverride=0.9999
    bGravityOverrideInSimSpace=true
    bOverrideAngularDamping=true
    bOverrideLinearDamping=true
    bUseGravityOverride=true
    linearDampingOverride=0.9999
    simulationSpace=BoneRelative
    alphaBoolBlend {blendOption=Linear}
    alphaScaleBiasClamp {bias=0.2, scale=0.8}
    boundBone {boneName=dyn_leg_petals_l_parent_c}
    componentAppliedLinearAccClamp=<100 100 100>
    componentLinearAccScale=<0.15 0.15 0.15>
    componentLinearVelScale=<0.15 0.15 0.15>
    gravityOverride=<0.7 0.7 0.7>
    physicsBodyDefinitions [{
      sphereCollisionRadius=0
      boundBone {boneName=dyn_leg_petals_l_parent_c}
      boxExtents=<8 14 2>
      constraintSetup {angularLimitsMax=<0 15 10>, angularLimitsMin=<0 -0.5 -10>, angularTarget=<0 0 0>}
      localJointOffset=<2 0 0>
     }]
    relativeSpaceBone {boneName=root}
   }
   n14 = Anim Dynamics - Bone: dyn_pelvis_petals_l_parent_a(ComponentPose=@n13, Alpha=1, ExternalForce="0,0,0") {
    angularDampingOverride=0.9999
    bGravityOverrideInSimSpace=true
    bOverrideAngularDamping=true
    bOverrideLinearDamping=true
    bUseGravityOverride=true
    linearDampingOverride=0.9999
    simulationSpace=RootRelative
    alphaBoolBlend {blendOption=Linear}
    alphaScaleBias {bias=-0.3333, scale=1.6667}
    boundBone {boneName=dyn_pelvis_petals_l_parent_a}
    componentAppliedLinearAccClamp=<100 100 100>
    componentLinearAccScale=<0.15 0.15 0.15>
    componentLinearVelScale=<0.15 0.15 0.15>
    gravityOverride=<0.7 0.7 0.7>
    physicsBodyDefinitions [{
      sphereCollisionRadius=0
      boundBone {boneName=dyn_pelvis_petals_l_parent_a}
      boxExtents=<5 4 1>
      constraintSetup {angularLimitsMax=<0 10 10>, angularLimitsMin=<0 -10 -10>, angularTarget=<0 0 0>}
      localJointOffset=<2 0 0>
     }]
    relativeSpaceBone {boneName=pelvis}
   }
   n15 = Anim Dynamics - Bone: dyn_pelvis_petals_l_parent_b(ComponentPose=@n14, Alpha=1, ExternalForce="0,0,0") {
    angularDampingOverride=0.9999
    bGravityOverrideInSimSpace=true
    bOverrideAngularDamping=true
    bOverrideLinearDamping=true
    bUseGravityOverride=true
    linearDampingOverride=0.9999
    simulationSpace=RootRelative
    alphaBoolBlend {blendOption=Linear}
    alphaScaleBias {bias=-0.3333, scale=1.6667}
    boundBone {boneName=dyn_pelvis_petals_l_parent_b}
    componentAppliedLinearAccClamp=<100 100 100>
    componentLinearAccScale=<0.15 0.15 0.15>
    componentLinearVelScale=<0.15 0.15 0.15>
    gravityOverride=<0.7 0.7 0.7>
    physicsBodyDefinitions [{
      sphereCollisionRadius=0
      boundBone {boneName=dyn_pelvis_petals_l_parent_b}
      boxExtents=<6 6 2>
      constraintSetup {angularLimitsMax=<0 5 15>, angularLimitsMin=<0 -5 -10>, angularTarget=<0 0 0>}
      localJointOffset=<1 0 0>
     }]
    relativeSpaceBone {boneName=pelvis}
   }
   n16 = Anim Dynamics - Bone: dyn_pelvis_petals_l_parent_c(ComponentPose=@n15, Alpha=1, ExternalForce="0,0,0") {
    angularDampingOverride=0.9999
    bGravityOverrideInSimSpace=true
    bOverrideAngularDamping=true
    bOverrideLinearDamping=true
    bUseGravityOverride=true
    linearDampingOverride=0.9999
    simulationSpace=RootRelative
    alphaBoolBlend {blendOption=Linear}
    alphaScaleBias {bias=-0.3333, scale=1.6667}
    boundBone {boneName=dyn_pelvis_petals_l_parent_c}
    componentAppliedLinearAccClamp=<100 100 100>
    componentLinearAccScale=<0.15 0.15 0.15>
    componentLinearVelScale=<0.15 0.15 0.15>
    gravityOverride=<0.7 0.7 0.7>
    physicsBodyDefinitions [{
      sphereCollisionRadius=0
      boundBone {boneName=dyn_pelvis_petals_l_parent_c}
      boxExtents=<5 4 1>
      constraintSetup {angularLimitsMax=<0 10 5>, angularLimitsMin=<0 -10 -5>, angularTarget=<0 0 0>}
      localJointOffset=<1.5 0 0>
     }]
    relativeSpaceBone {boneName=pelvis}
   }
   n17 = Anim Dynamics - Bone: dyn_waist_flower_outer_r_01(ComponentPose=@n16, Alpha=1, ExternalForce="0,0,0") {
    angularDampingOverride=0.9999
    bGravityOverrideInSimSpace=true
    bOverrideAngularDamping=true
    bOverrideLinearDamping=true
    bUseGravityOverride=true
    linearDampingOverride=0.9999
    simulationSpace=RootRelative
    alphaBoolBlend {blendOption=Linear}
    boundBone {boneName=dyn_waist_flower_outer_r_01}
    componentAppliedLinearAccClamp=<100 100 100>
    componentLinearAccScale=<0.15 0.15 0.15>
    componentLinearVelScale=<0.15 0.15 0.15>
    gravityOverride=<0.7 0.7 0.7>
    physicsBodyDefinitions [{
      sphereCollisionRadius=0
      boundBone {boneName=dyn_waist_flower_outer_r_01}
      boxExtents=<6 8 2>
      constraintSetup {angularLimitsMax=<0 15 5>, angularLimitsMin=<0 -15 -20>, angularTarget=<0 0 0>}
      localJointOffset=<2 0 0>
     }]
    relativeSpaceBone {boneName=root}
   }
   n18 = Anim Dynamics - Bone: dyn_waist_flower_outer_r_02(ComponentPose=@n17, Alpha=1, ExternalForce="0,0,0") {
    angularDampingOverride=0.999
    bGravityOverrideInSimSpace=true
    bOverrideAngularDamping=true
    bOverrideLinearDamping=true
    bUseGravityOverride=true
    linearDampingOverride=0.999
    simulationSpace=RootRelative
    alphaBoolBlend {blendOption=Linear}
    boundBone {boneName=dyn_waist_flower_outer_r_02}
    componentAppliedLinearAccClamp=<3000 3000 3000>
    componentLinearAccScale=<0.25 0.25 0.5>
    componentLinearVelScale=<0.25 0.25 0.5>
    gravityOverride=<0.85 0.85 0.85>
    physicsBodyDefinitions [{
      sphereCollisionRadius=0
      boundBone {boneName=dyn_waist_flower_outer_r_02}
      boxExtents=<5 4 1>
      constraintSetup {angularLimitsMax=<0 10 5>, angularLimitsMin=<0 -10 -5>, angularTarget=<0 0 0>}
      localJointOffset=<1.5 0 0>
     }]
    relativeSpaceBone {boneName=root}
   }
   n19 = Anim Dynamics - Bone: dyn_waist_flower_outer_r_03(ComponentPose=@n18, Alpha=1, ExternalForce="0,0,0") {
    angularDampingOverride=0.9999
    bGravityOverrideInSimSpace=true
    bOverrideAngularDamping=true
    bOverrideLinearDamping=true
    bUseGravityOverride=true
    linearDampingOverride=0.9999
    simulationSpace=RootRelative
    alphaBoolBlend {blendOption=Linear}
    boundBone {boneName=dyn_waist_flower_outer_r_03}
    componentAppliedLinearAccClamp=<100 100 100>
    componentLinearAccScale=<0.15 0.15 0.15>
    componentLinearVelScale=<0.15 0.15 0.15>
    gravityOverride=<0.7 0.7 0.7>
    physicsBodyDefinitions [{
      sphereCollisionRadius=0
      boundBone {boneName=dyn_waist_flower_outer_r_03}
      boxExtents=<5 8 1>
      constraintSetup {angularLimitsMax=<0 20 15>, angularLimitsMin=<0 -15 -15>, angularTarget=<0 0 0>}
      localJointOffset=<2 0 0>
     }]
    relativeSpaceBone {boneName=root}
   }
   n20 = Anim Dynamics - Bone: dyn_waist_flower_outer_r_04(ComponentPose=@n19, Alpha=1, ExternalForce="0,0,0") {
    angularDampingOverride=0.999
    bGravityOverrideInSimSpace=true
    bOverrideAngularDamping=true
    bOverrideLinearDamping=true
    bUseGravityOverride=true
    linearDampingOverride=0.999
    simulationSpace=RootRelative
    alphaBoolBlend {blendOption=Linear}
    boundBone {boneName=dyn_waist_flower_outer_r_04}
    componentAppliedLinearAccClamp=<3000 3000 3000>
    componentLinearAccScale=<0.25 0.25 0.5>
    componentLinearVelScale=<0.25 0.25 0.5>
    gravityOverride=<0.85 0.85 0.85>
    physicsBodyDefinitions [{
      sphereCollisionRadius=0
      boundBone {boneName=dyn_waist_flower_outer_r_04}
      boxExtents=<4 4 2>
      constraintSetup {angularLimitsMax=<0 25 15>, angularLimitsMin=<0 -15 -15>, angularTarget=<0 0 0>}
      localJointOffset=<1 0 0>
     }]
    relativeSpaceBone {boneName=root}
   }
   n21 = Anim Dynamics - Bone: dyn_waist_flower_outer_r_07(ComponentPose=@n20, Alpha=1, ExternalForce="0,0,0") {
    angularDampingOverride=0.999
    bGravityOverrideInSimSpace=true
    bOverrideAngularDamping=true
    bOverrideLinearDamping=true
    bUseGravityOverride=true
    linearDampingOverride=0.999
    simulationSpace=RootRelative
    alphaBoolBlend {blendOption=Linear}
    boundBone {boneName=dyn_waist_flower_outer_r_07}
    componentAppliedLinearAccClamp=<3000 3000 3000>
    componentLinearAccScale=<0.25 0.25 0.5>
    componentLinearVelScale=<0.25 0.25 0.5>
    gravityOverride=<0.85 0.85 0.85>
    physicsBodyDefinitions [{
      sphereCollisionRadius=0
      boundBone {boneName=dyn_waist_flower_outer_r_07}
      boxExtents=<5 4 1>
      constraintSetup {angularLimitsMax=<0 15 15>, angularLimitsMin=<0 -15 -15>, angularTarget=<0 0 0>}
      localJointOffset=<2 0 0>
     }]
    relativeSpaceBone {boneName=root}
   }
   n22 = Anim Dynamics - Bone: dyn_waist_flower_outer_r_06(ComponentPose=@n21, Alpha=1, ExternalForce="0,0,0") {
    angularDampingOverride=0.999
    bGravityOverrideInSimSpace=true
    bOverrideAngularDamping=true
    bOverrideLinearDamping=true
    bUseGravityOverride=true
    linearDampingOverride=0.999
    simulationSpace=RootRelative
    alphaBoolBlend {blendOption=Linear}
    boundBone {boneName=dyn_waist_flower_outer_r_06}
    componentAppliedLinearAccClamp=<3000 3000 3000>
    componentLinearAccScale=<0.25 0.25 0.5>
    componentLinearVelScale=<0.25 0.25 0.5>
    gravityOverride=<0.85 0.85 0.85>
    physicsBodyDefinitions [{
      sphereCollisionRadius=0
      boundBone {boneName=dyn_waist_flower_outer_r_06}
      boxExtents=<5 5 1.5>
      constraintSetup {angularLimitsMax=<0 10 10>, angularLimitsMin=<0 -10 -10>, angularTarget=<0 0 0>}
      localJointOffset=<1 0 0>
     }]
    relativeSpaceBone {boneName=root}
   }
   n23 = Anim Dynamics - Bone: dyn_waist_flower_outer_r_05(ComponentPose=@n22, Alpha=1, ExternalForce="0,0,0") {
    angularDampingOverride=0.999
    bGravityOverrideInSimSpace=true
    bOverrideAngularDamping=true
    bOverrideLinearDamping=true
    bUseGravityOverride=true
    linearDampingOverride=0.999
    simulationSpace=RootRelative
    alphaBoolBlend {blendOption=Linear}
    boundBone {boneName=dyn_waist_flower_outer_r_05}
    componentAppliedLinearAccClamp=<3000 3000 3000>
    componentLinearAccScale=<0.25 0.25 0.5>
    componentLinearVelScale=<0.25 0.25 0.5>
    gravityOverride=<0.85 0.85 0.85>
    physicsBodyDefinitions [{
      sphereCollisionRadius=0
      boundBone {boneName=dyn_waist_flower_outer_r_05}
      boxExtents=<5 4 1.5>
      constraintSetup {angularLimitsMax=<0 10 5>, angularLimitsMin=<0 -1 -5>, angularTarget=<0 0 0>}
      localJointOffset=<1 0 0>
     }]
    relativeSpaceBone {boneName=root}
   }
   n24 = Get Dynamics_Alpha() -> (Dynamics_Alpha:double)
   n25 = Anim Dynamics - Bone: dyn_waist_flower_inner_r_01(ComponentPose=@n23, Alpha=@n24, ExternalForce="0,0,0") {
    angularDampingOverride=0.999
    bGravityOverrideInSimSpace=true
    bOverrideAngularDamping=true
    bOverrideLinearDamping=true
    bUseGravityOverride=true
    linearDampingOverride=0.999
    simulationSpace=RootRelative
    alphaBoolBlend {blendOption=Linear}
    boundBone {boneName=dyn_waist_flower_inner_r_01}
    componentAppliedLinearAccClamp=<3000 3000 3000>
    componentLinearAccScale=<0.25 0.25 0.5>
    componentLinearVelScale=<0.25 0.25 0.5>
    gravityOverride=<0.85 0.85 0.85>
    physicsBodyDefinitions [{
      sphereCollisionRadius=0
      boundBone {boneName=dyn_waist_flower_inner_r_01}
      boxExtents=<4 4 2>
      constraintSetup {angularLimitsMax=<0 30 20>, angularLimitsMin=<0 -5 -20>, angularTarget=<0 0 0>}
      localJointOffset=<1 0 0>
     }]
    relativeSpaceBone {boneName=root}
   }
   n26 = Anim Dynamics - Bone: dyn_waist_flower_inner_r_02(ComponentPose=@n25, Alpha=@n24, ExternalForce="0,0,0") {
    angularDampingOverride=0.999
    bGravityOverrideInSimSpace=true
    bOverrideAngularDamping=true
    bOverrideLinearDamping=true
    bUseGravityOverride=true
    linearDampingOverride=0.999
    simulationSpace=RootRelative
    alphaBoolBlend {blendOption=Linear}
    boundBone {boneName=dyn_waist_flower_inner_r_02}
    componentAppliedLinearAccClamp=<3000 3000 3000>
    componentLinearAccScale=<0.25 0.25 0.5>
    componentLinearVelScale=<0.25 0.25 0.5>
    gravityOverride=<0.85 0.85 0.85>
    physicsBodyDefinitions [{
      sphereCollisionRadius=0
      boundBone {boneName=dyn_waist_flower_inner_r_02}
      boxExtents=<5 4 1>
      constraintSetup {angularLimitsMax=<0 10 5>, angularLimitsMin=<0 -5 -5>, angularTarget=<0 0 0>}
      localJointOffset=<1.5 0 0>
     }]
    relativeSpaceBone {boneName=root}
   }
   n27 = Anim Dynamics - Bone: dyn_waist_flower_inner_r_03(ComponentPose=@n26, Alpha=@n24, ExternalForce="0,0,0") {
    angularDampingOverride=0.999
    bGravityOverrideInSimSpace=true
    bOverrideAngularDamping=true
    bOverrideLinearDamping=true
    bUseGravityOverride=true
    linearDampingOverride=0.999
    simulationSpace=RootRelative
    alphaBoolBlend {blendOption=Linear}
    boundBone {boneName=dyn_waist_flower_inner_r_03}
    componentAppliedLinearAccClamp=<3000 3000 3000>
    componentLinearAccScale=<0.25 0.25 0.5>
    componentLinearVelScale=<0.25 0.25 0.5>
    gravityOverride=<0.85 0.85 0.85>
    physicsBodyDefinitions [{
      sphereCollisionRadius=0
      boundBone {boneName=dyn_waist_flower_inner_r_03}
      boxExtents=<5 4 1>
      constraintSetup {angularLimitsMax=<0 20 20>, angularLimitsMin=<0 -5 -20>, angularTarget=<0 0 0>}
      localJointOffset=<2 0 0>
     }]
    relativeSpaceBone {boneName=spine_05}
   }
   n28 = Anim Dynamics - Bone: dyn_waist_flower_inner_r_06(ComponentPose=@n27, Alpha=@n24, ExternalForce="0,0,0") {
    angularDampingOverride=0.999
    bGravityOverrideInSimSpace=true
    bOverrideAngularDamping=true
    bOverrideLinearDamping=true
    bUseGravityOverride=true
    linearDampingOverride=0.999
    simulationSpace=RootRelative
    alphaBoolBlend {blendOption=Linear}
    boundBone {boneName=dyn_waist_flower_inner_r_06}
    componentAppliedLinearAccClamp=<3000 3000 3000>
    componentLinearAccScale=<0.25 0.25 0.5>
    componentLinearVelScale=<0.25 0.25 0.5>
    gravityOverride=<0.85 0.85 0.85>
    physicsBodyDefinitions [{
      sphereCollisionRadius=0
      boundBone {boneName=dyn_waist_flower_inner_r_06}
      boxExtents=<5 4 1>
      constraintSetup {angularLimitsMax=<0 15 20>, angularLimitsMin=<0 -5 -20>, angularTarget=<0 0 0>}
      localJointOffset=<2 0 0>
     }]
    relativeSpaceBone {boneName=root}
   }
   n29 = Anim Dynamics - Bone: dyn_waist_flower_inner_r_05(ComponentPose=@n28, Alpha=@n24, ExternalForce="0,0,0") {
    angularDampingOverride=0.999
    bGravityOverrideInSimSpace=true
    bOverrideAngularDamping=true
    bOverrideLinearDamping=true
    bUseGravityOverride=true
    linearDampingOverride=0.999
    simulationSpace=RootRelative
    alphaBoolBlend {blendOption=Linear}
    boundBone {boneName=dyn_waist_flower_inner_r_05}
    componentAppliedLinearAccClamp=<3000 3000 3000>
    componentLinearAccScale=<0.25 0.25 0.5>
    componentLinearVelScale=<0.25 0.25 0.5>
    gravityOverride=<0.85 0.85 0.85>
    physicsBodyDefinitions [{
      sphereCollisionRadius=0
      boundBone {boneName=dyn_waist_flower_inner_r_05}
      boxExtents=<5 5 1.5>
      constraintSetup {angularLimitsMax=<0 10 10>, angularLimitsMin=<0 -2 -10>, angularTarget=<0 0 0>}
      localJointOffset=<1 0 0>
     }]
    relativeSpaceBone {boneName=spine_05}
   }
   n30 = Anim Dynamics - Bone: dyn_waist_flower_inner_r_04(ComponentPose=@n29, Alpha=@n24, ExternalForce="0,0,0") {
    angularDampingOverride=0.999
    bGravityOverrideInSimSpace=true
    bOverrideAngularDamping=true
    bOverrideLinearDamping=true
    bUseGravityOverride=true
    linearDampingOverride=0.999
    simulationSpace=RootRelative
    alphaBoolBlend {blendOption=Linear}
    boundBone {boneName=dyn_waist_flower_inner_r_04}
    componentAppliedLinearAccClamp=<3000 3000 3000>
    componentLinearAccScale=<0.25 0.25 0.5>
    componentLinearVelScale=<0.25 0.25 0.5>
    gravityOverride=<0.85 0.85 0.85>
    physicsBodyDefinitions [{
      sphereCollisionRadius=0
      boundBone {boneName=dyn_waist_flower_inner_r_04}
      boxExtents=<5 4 1.5>
      constraintSetup {angularLimitsMax=<0 10 5>, angularLimitsMin=<0 -1 -5>, angularTarget=<0 0 0>}
      localJointOffset=<1 0 0>
     }]
    relativeSpaceBone {boneName=spine_05}
   }
   n31 = Anim Dynamics - Bone: dyn_shoulder_flower_outter_l_parent_a(ComponentPose=@n30, Alpha=1, ExternalForce="0,0,0") {
    angularDampingOverride=0.9999
    bGravityOverrideInSimSpace=true
    bOverrideAngularDamping=true
    bOverrideLinearDamping=true
    bUseGravityOverride=true
    linearDampingOverride=0.9999
    simulationSpace=BoneRelative
    alphaBoolBlend {blendOption=Linear}
    boundBone {boneName=dyn_shoulder_flower_outter_l_parent_a}
    componentAppliedLinearAccClamp=<100 100 100>
    componentLinearAccScale=<0.15 0.15 0.15>
    componentLinearVelScale=<0.15 0.15 0.15>
    gravityOverride=<0.7 0.7 0.7>
    physicsBodyDefinitions [{
      sphereCollisionRadius=0
      boundBone {boneName=dyn_shoulder_flower_outter_l_parent_a}
      boxExtents=<6 10 2>
      constraintSetup {angularLimitsMax=<0 30 10>, angularLimitsMin=<0 -1 -10>, angularTarget=<0 0 0>}
      localJointOffset=<2 0 0>
     }]
    relativeSpaceBone {boneName=spine_05}
   }
   n32 = Anim Dynamics - Bone: dyn_shoulder_flower_outter_l_parent_b(ComponentPose=@n31, Alpha=1, ExternalForce="0,0,0") {
    angularDampingOverride=0.9999
    bGravityOverrideInSimSpace=true
    bOverrideAngularDamping=true
    bOverrideLinearDamping=true
    bUseGravityOverride=true
    linearDampingOverride=0.9999
    simulationSpace=RootRelative
    alphaBoolBlend {blendOption=Linear}
    boundBone {boneName=dyn_shoulder_flower_outter_l_parent_b}
    componentAppliedLinearAccClamp=<100 100 100>
    componentLinearAccScale=<0.15 0.15 0.15>
    componentLinearVelScale=<0.15 0.15 0.15>
    gravityOverride=<0.7 0.7 0.7>
    physicsBodyDefinitions [{
      sphereCollisionRadius=0
      boundBone {boneName=dyn_shoulder_flower_outter_l_parent_b}
      boxExtents=<5 4 1>
      constraintSetup {angularLimitsMax=<0 20 5>, angularLimitsMin=<0 -1 -5>, angularTarget=<0 0 0>}
      localJointOffset=<1.5 0 0>
     }]
    relativeSpaceBone {boneName=clavicle_l}
   }
   n33 = Anim Dynamics - Bone: dyn_shoulder_flower_outter_l_parent_c(ComponentPose=@n32, Alpha=1, ExternalForce="0,0,0") {
    angularDampingOverride=0.9999
    bGravityOverrideInSimSpace=true
    bOverrideAngularDamping=true
    bOverrideLinearDamping=true
    bUseGravityOverride=true
    linearDampingOverride=0.9999
    simulationSpace=RootRelative
    alphaBoolBlend {blendOption=Linear}
    boundBone {boneName=dyn_shoulder_flower_outter_l_parent_c}
    componentAppliedLinearAccClamp=<100 100 100>
    componentLinearAccScale=<0.15 0.15 0.15>
    componentLinearVelScale=<0.15 0.15 0.15>
    gravityOverride=<0.7 0.7 0.7>
    physicsBodyDefinitions [{
      sphereCollisionRadius=0
      boundBone {boneName=dyn_shoulder_flower_outter_l_parent_c}
      boxExtents=<10 10 1>
      constraintSetup {angularLimitsMax=<0 20 10>, angularLimitsMin=<0 -30 -10>, angularTarget=<0 0 0>}
      localJointOffset=<2 0 0>
     }]
    relativeSpaceBone {boneName=clavicle_l}
   }
   n34 = Anim Dynamics - Bone: dyn_shoulder_flower_outter_l_parent_g(ComponentPose=@n33, Alpha=1, ExternalForce="0,0,0") {
    angularDampingOverride=0.9999
    bGravityOverrideInSimSpace=true
    bOverrideAngularDamping=true
    bOverrideLinearDamping=true
    bUseGravityOverride=true
    linearDampingOverride=0.9999
    simulationSpace=BoneRelative
    alphaBoolBlend {blendOption=Linear}
    boundBone {boneName=dyn_shoulder_flower_outter_l_parent_g}
    componentAppliedLinearAccClamp=<100 100 100>
    componentLinearAccScale=<0.15 0.15 0.15>
    componentLinearVelScale=<0.15 0.15 0.15>
    gravityOverride=<0.7 0.7 0.7>
    physicsBodyDefinitions [{
      sphereCollisionRadius=0
      boundBone {boneName=dyn_shoulder_flower_outter_l_parent_g}
      boxExtents=<5 8 2>
      constraintSetup {angularLimitsMax=<0 20 5>, angularLimitsMin=<0 -10 -5>, angularTarget=<0 0 0>}
      localJointOffset=<1 0 0>
     }]
    relativeSpaceBone {boneName=spine_05}
   }
   n35 = Anim Dynamics - Bone: dyn_shoulder_flower_outter_l_parent_i(ComponentPose=@n34, Alpha=1, ExternalForce="0,0,0") {
    angularDampingOverride=0.9999
    bGravityOverrideInSimSpace=true
    bOverrideAngularDamping=true
    bOverrideLinearDamping=true
    bUseGravityOverride=true
    linearDampingOverride=0.9999
    simulationSpace=BoneRelative
    alphaBoolBlend {blendOption=Linear}
    boundBone {boneName=dyn_shoulder_flower_outter_l_parent_i}
    componentAppliedLinearAccClamp=<100 100 100>
    componentLinearAccScale=<0.15 0.15 0.15>
    componentLinearVelScale=<0.15 0.15 0.15>
    gravityOverride=<0.7 0.7 0.7>
    physicsBodyDefinitions [{
      sphereCollisionRadius=0
      boundBone {boneName=dyn_shoulder_flower_outter_l_parent_i}
      boxExtents=<8 8 1>
      constraintSetup {angularLimitsMax=<0 25 5>, angularLimitsMin=<0 -10 -5>, angularTarget=<0 0 0>}
      localJointOffset=<1.5 0 0>
     }]
    relativeSpaceBone {boneName=spine_05}
   }
   n36 = Anim Dynamics - Bone: dyn_shoulder_flower_outter_l_parent_h(ComponentPose=@n35, Alpha=1, ExternalForce="0,0,0") {
    angularDampingOverride=0.9999
    bGravityOverrideInSimSpace=true
    bOverrideAngularDamping=true
    bOverrideLinearDamping=true
    bUseGravityOverride=true
    linearDampingOverride=0.9999
    simulationSpace=RootRelative
    alphaBoolBlend {blendOption=Linear}
    boundBone {boneName=dyn_shoulder_flower_outter_l_parent_h}
    componentAppliedLinearAccClamp=<100 100 100>
    componentLinearAccScale=<0.15 0.15 0.15>
    componentLinearVelScale=<0.15 0.15 0.15>
    gravityOverride=<0.7 0.7 0.7>
    physicsBodyDefinitions [{
      sphereCollisionRadius=0
      boundBone {boneName=dyn_shoulder_flower_outter_l_parent_h}
      boxExtents=<6 6 1.5>
      constraintSetup {angularLimitsMax=<0 30 5>, angularLimitsMin=<0 -2 -5>, angularTarget=<0 0 0>}
      localJointOffset=<1 0 0>
     }]
    relativeSpaceBone {boneName=clavicle_l}
   }
   n37 = Anim Dynamics - Bone: dyn_shoulder_flower_outter_l_parent_f(ComponentPose=@n36, Alpha=1, ExternalForce="0,0,0") {
    angularDampingOverride=0.9999
    bGravityOverrideInSimSpace=true
    bOverrideAngularDamping=true
    bOverrideLinearDamping=true
    bUseGravityOverride=true
    linearDampingOverride=0.9999
    simulationSpace=BoneRelative
    alphaBoolBlend {blendOption=Linear}
    boundBone {boneName=dyn_shoulder_flower_outter_l_parent_f}
    componentAppliedLinearAccClamp=<100 100 100>
    componentLinearAccScale=<0.15 0.15 0.15>
    componentLinearVelScale=<0.15 0.15 0.15>
    gravityOverride=<0.7 0.7 0.7>
    physicsBodyDefinitions [{
      sphereCollisionRadius=0
      boundBone {boneName=dyn_shoulder_flower_outter_l_parent_f}
      boxExtents=<5 6 1>
      constraintSetup {angularLimitsMax=<0 15 20>, angularLimitsMin=<0 -15 -20>, angularTarget=<0 0 0>}
      localJointOffset=<2 0 0>
     }]
    relativeSpaceBone {boneName=spine_05}
   }
   n38 = Anim Dynamics - Bone: dyn_shoulder_flower_outter_l_parent_e(ComponentPose=@n37, Alpha=1, ExternalForce="0,0,0") {
    angularDampingOverride=0.9999
    bGravityOverrideInSimSpace=true
    bOverrideAngularDamping=true
    bOverrideLinearDamping=true
    bUseGravityOverride=true
    linearDampingOverride=0.9999
    simulationSpace=RootRelative
    alphaBoolBlend {blendOption=Linear}
    boundBone {boneName=dyn_shoulder_flower_outter_l_parent_e}
    componentAppliedLinearAccClamp=<100 100 100>
    componentLinearAccScale=<0.15 0.15 0.15>
    componentLinearVelScale=<0.15 0.15 0.15>
    gravityOverride=<0.7 0.7 0.7>
    physicsBodyDefinitions [{
      sphereCollisionRadius=0
      boundBone {boneName=dyn_shoulder_flower_outter_l_parent_e}
      boxExtents=<5 5 1.5>
      constraintSetup {angularLimitsMax=<0 10 10>, angularLimitsMin=<0 -10 -10>, angularTarget=<0 0 0>}
      localJointOffset=<1 0 0>
     }]
    relativeSpaceBone {boneName=clavicle_l}
   }
   n39 = Anim Dynamics - Bone: dyn_shoulder_flower_outter_l_parent_d(ComponentPose=@n38, Alpha=1, ExternalForce="0,0,0") {
    angularDampingOverride=0.9999
    bGravityOverrideInSimSpace=true
    bOverrideAngularDamping=true
    bOverrideLinearDamping=true
    bUseGravityOverride=true
    linearDampingOverride=0.9999
    simulationSpace=RootRelative
    alphaBoolBlend {blendOption=Linear}
    boundBone {boneName=dyn_shoulder_flower_outter_l_parent_d}
    componentAppliedLinearAccClamp=<100 100 100>
    componentLinearAccScale=<0.15 0.15 0.15>
    componentLinearVelScale=<0.15 0.15 0.15>
    gravityOverride=<0.7 0.7 0.7>
    physicsBodyDefinitions [{
      sphereCollisionRadius=0
      boundBone {boneName=dyn_shoulder_flower_outter_l_parent_d}
      boxExtents=<5 4 1.5>
      constraintSetup {angularLimitsMax=<0 10 5>, angularLimitsMin=<0 -1 -5>, angularTarget=<0 0 0>}
      localJointOffset=<1 0 0>
     }]
    relativeSpaceBone {boneName=clavicle_l}
   }
   n40 = Get Dynamics_Alpha()
   n41 = Anim Dynamics - Bone: dyn_shoulder_flower_inner_l_01(ComponentPose=@n39, Alpha=@n40, ExternalForce="0,0,0") {
    angularDampingOverride=0.9999
    bGravityOverrideInSimSpace=true
    bOverrideAngularDamping=true
    bOverrideLinearDamping=true
    bUseGravityOverride=true
    linearDampingOverride=0.9999
    simulationSpace=BoneRelative
    alphaBoolBlend {blendOption=Linear}
    alphaScaleBiasClamp {bInterpResult=true, bias=0.2, interpSpeedDecreasing=2, interpSpeedIncreasing=8, scale=0.8}
    boundBone {boneName=dyn_shoulder_flower_inner_l_01}
    componentAppliedLinearAccClamp=<100 100 100>
    componentLinearAccScale=<0.15 0.15 0.15>
    componentLinearVelScale=<0.15 0.15 0.15>
    gravityOverride=<0.7 0.7 0.7>
    physicsBodyDefinitions [{
      sphereCollisionRadius=0
      boundBone {boneName=dyn_shoulder_flower_inner_l_01}
      boxExtents=<4 4 2>
      constraintSetup {angularLimitsMax=<0 30 15>, angularLimitsMin=<0 -5 -15>, angularTarget=<0 0 0>}
      localJointOffset=<1 0 0>
     }]
    relativeSpaceBone {boneName=upperarm_twist_01_l}
   }
   n42 = Anim Dynamics - Bone: dyn_shoulder_flower_inner_l_02(ComponentPose=@n41, Alpha=@n40, ExternalForce="0,0,0") {
    angularDampingOverride=0.9999
    bGravityOverrideInSimSpace=true
    bOverrideAngularDamping=true
    bOverrideLinearDamping=true
    bUseGravityOverride=true
    linearDampingOverride=0.9999
    simulationSpace=BoneRelative
    alphaBoolBlend {blendOption=Linear}
    alphaScaleBiasClamp {bInterpResult=true, bias=0.2, interpSpeedDecreasing=2, interpSpeedIncreasing=8, scale=0.8}
    boundBone {boneName=dyn_shoulder_flower_inner_l_02}
    componentAppliedLinearAccClamp=<100 100 100>
    componentLinearAccScale=<0.15 0.15 0.15>
    componentLinearVelScale=<0.15 0.15 0.15>
    gravityOverride=<0.7 0.7 0.7>
    physicsBodyDefinitions [{
      sphereCollisionRadius=0
      boundBone {boneName=dyn_shoulder_flower_inner_l_02}
      boxExtents=<5 4 1>
      constraintSetup {angularLimitsMax=<0 15 5>, angularLimitsMin=<0 -5 -5>, angularTarget=<0 0 0>}
      localJointOffset=<1.5 0 0>
     }]
    relativeSpaceBone {boneName=upperarm_twist_01_l}
   }
   n43 = Anim Dynamics - Bone: dyn_shoulder_flower_inner_l_03(ComponentPose=@n42, Alpha=@n40, ExternalForce="0,0,0") {
    angularDampingOverride=0.9999
    bGravityOverrideInSimSpace=true
    bOverrideAngularDamping=true
    bOverrideLinearDamping=true
    bUseGravityOverride=true
    linearDampingOverride=0.9999
    simulationSpace=BoneRelative
    alphaBoolBlend {blendOption=Linear}
    alphaScaleBiasClamp {bInterpResult=true, bias=0.2, interpSpeedDecreasing=2, interpSpeedIncreasing=8, scale=0.8}
    boundBone {boneName=dyn_shoulder_flower_inner_l_03}
    componentAppliedLinearAccClamp=<100 100 100>
    componentLinearAccScale=<0.15 0.15 0.15>
    componentLinearVelScale=<0.15 0.15 0.15>
    gravityOverride=<0.7 0.7 0.7>
    physicsBodyDefinitions [{
      sphereCollisionRadius=0
      boundBone {boneName=dyn_shoulder_flower_inner_l_03}
      boxExtents=<5 4 1>
      constraintSetup {angularLimitsMax=<0 20 15>, angularLimitsMin=<0 -2 -15>, angularTarget=<0 0 0>}
      localJointOffset=<2 0 0>
     }]
    relativeSpaceBone {boneName=upperarm_twist_01_l}
   }
   n44 = Anim Dynamics - Bone: dyn_shoulder_flower_inner_l_06(ComponentPose=@n43, Alpha=@n40, ExternalForce="0,0,0") {
    angularDampingOverride=0.9999
    bGravityOverrideInSimSpace=true
    bOverrideAngularDamping=true
    bOverrideLinearDamping=true
    bUseGravityOverride=true
    linearDampingOverride=0.9999
    simulationSpace=BoneRelative
    alphaBoolBlend {blendOption=Linear}
    alphaScaleBiasClamp {bInterpResult=true, bias=0.2, interpSpeedDecreasing=2, interpSpeedIncreasing=8, scale=0.8}
    boundBone {boneName=dyn_shoulder_flower_inner_l_06}
    componentAppliedLinearAccClamp=<100 100 100>
    componentLinearAccScale=<0.15 0.15 0.15>
    componentLinearVelScale=<0.15 0.15 0.15>
    gravityOverride=<0.7 0.7 0.7>
    physicsBodyDefinitions [{
      sphereCollisionRadius=0
      boundBone {boneName=dyn_shoulder_flower_inner_l_06}
      boxExtents=<5 4 1>
      constraintSetup {angularLimitsMax=<0 15 10>, angularLimitsMin=<0 -5 -2>, angularTarget=<0 0 0>}
      localJointOffset=<2 0 0>
     }]
    relativeSpaceBone {boneName=upperarm_twist_01_l}
   }
   n45 = Anim Dynamics - Bone: dyn_shoulder_flower_inner_l_05(ComponentPose=@n44, Alpha=@n40, ExternalForce="0,0,0") {
    angularDampingOverride=0.9999
    bGravityOverrideInSimSpace=true
    bOverrideAngularDamping=true
    bOverrideLinearDamping=true
    bUseGravityOverride=true
    linearDampingOverride=0.9999
    simulationSpace=BoneRelative
    alphaBoolBlend {blendOption=Linear}
    alphaScaleBiasClamp {bInterpResult=true, bias=0.2, interpSpeedDecreasing=2, interpSpeedIncreasing=8, scale=0.8}
    boundBone {boneName=dyn_shoulder_flower_inner_l_05}
    componentAppliedLinearAccClamp=<100 100 100>
    componentLinearAccScale=<0.15 0.15 0.15>
    componentLinearVelScale=<0.15 0.15 0.15>
    gravityOverride=<0.7 0.7 0.7>
    physicsBodyDefinitions [{
      sphereCollisionRadius=0
      boundBone {boneName=dyn_shoulder_flower_inner_l_05}
      boxExtents=<5 5 1.5>
      constraintSetup {angularLimitsMax=<0 15 10>, angularLimitsMin=<0 -5 -10>, angularTarget=<0 0 0>}
      localJointOffset=<1 0 0>
     }]
    relativeSpaceBone {boneName=upperarm_twist_01_l}
   }
   n46 = Anim Dynamics - Bone: dyn_shoulder_flower_inner_l_04(ComponentPose=@n45, Alpha=@n40, ExternalForce="0,0,0") {
    angularDampingOverride=0.9999
    bGravityOverrideInSimSpace=true
    bOverrideAngularDamping=true
    bOverrideLinearDamping=true
    bUseGravityOverride=true
    linearDampingOverride=0.9999
    simulationSpace=BoneRelative
    alphaBoolBlend {blendOption=Linear}
    alphaScaleBiasClamp {bInterpResult=true, bias=0.2, interpSpeedDecreasing=2, interpSpeedIncreasing=8, scale=0.8}
    boundBone {boneName=dyn_shoulder_flower_inner_l_04}
    componentAppliedLinearAccClamp=<100 100 100>
    componentLinearAccScale=<0.15 0.15 0.15>
    componentLinearVelScale=<0.15 0.15 0.15>
    gravityOverride=<0.7 0.7 0.7>
    physicsBodyDefinitions [{
      sphereCollisionRadius=0
      boundBone {boneName=dyn_shoulder_flower_inner_l_04}
      boxExtents=<5 4 1.5>
      constraintSetup {angularLimitsMax=<0 10 5>, angularLimitsMin=<0 -1 -5>, angularTarget=<0 0 0>}
      localJointOffset=<1 0 0>
     }]
    relativeSpaceBone {boneName=upperarm_twist_01_l}
   }
   n47 = Get Dynamics_Alpha()
   n48 = Anim Dynamics - Bone: dyn_elbow_flower_l_01(ComponentPose=@n46, Alpha=@n47, ExternalForce="0,0,0") {
    angularDampingOverride=0.9999
    bGravityOverrideInSimSpace=true
    bOverrideAngularDamping=true
    bOverrideLinearDamping=true
    bUseGravityOverride=true
    linearDampingOverride=0.9999
    simulationSpace=RootRelative
    alphaBoolBlend {blendOption=Linear}
    alphaScaleBiasClamp {bInterpResult=true, interpSpeedDecreasing=6, interpSpeedIncreasing=8}
    boundBone {boneName=dyn_elbow_flower_l_01}
    componentAppliedLinearAccClamp=<100 100 100>
    componentLinearAccScale=<0.15 0.15 0.15>
    componentLinearVelScale=<0.15 0.15 0.15>
    gravityOverride=<0.8 0.8 0.8>
    physicsBodyDefinitions [{
      sphereCollisionRadius=0
      boundBone {boneName=dyn_elbow_flower_l_01}
      boxExtents=<4 4 2>
      constraintSetup {angularLimitsMax=<0 30 20>, angularLimitsMin=<0 -20 -20>, angularTarget=<0 0 0>}
      localJointOffset=<1 0 0>
     }]
    relativeSpaceBone {boneName=root}
   }
   n49 = Anim Dynamics - Bone: dyn_elbow_flower_l_02(ComponentPose=@n48, Alpha=@n47, ExternalForce="0,0,0") {
    angularDampingOverride=0.9999
    bGravityOverrideInSimSpace=true
    bOverrideAngularDamping=true
    bOverrideLinearDamping=true
    bUseGravityOverride=true
    linearDampingOverride=0.9999
    simulationSpace=RootRelative
    alphaBoolBlend {blendOption=Linear}
    alphaScaleBiasClamp {bInterpResult=true, interpSpeedDecreasing=6, interpSpeedIncreasing=8}
    boundBone {boneName=dyn_elbow_flower_l_02}
    componentAppliedLinearAccClamp=<100 100 100>
    componentLinearAccScale=<0.15 0.15 0.15>
    componentLinearVelScale=<0.15 0.15 0.15>
    gravityOverride=<0.8 0.8 0.8>
    physicsBodyDefinitions [{
      sphereCollisionRadius=0
      boundBone {boneName=dyn_elbow_flower_l_02}
      boxExtents=<5 4 1>
      constraintSetup {angularLimitsMax=<0 10 5>, angularLimitsMin=<0 -10 -5>, angularTarget=<0 0 0>}
      localJointOffset=<1.5 0 0>
     }]
    relativeSpaceBone {boneName=root}
   }
   n50 = Anim Dynamics - Bone: dyn_elbow_flower_l_03(ComponentPose=@n49, Alpha=@n47, ExternalForce="0,0,0") {
    angularDampingOverride=0.9999
    bGravityOverrideInSimSpace=true
    bOverrideAngularDamping=true
    bOverrideLinearDamping=true
    bUseGravityOverride=true
    linearDampingOverride=0.9999
    simulationSpace=RootRelative
    alphaBoolBlend {blendOption=Linear}
    alphaScaleBiasClamp {bInterpResult=true, interpSpeedDecreasing=6, interpSpeedIncreasing=8}
    boundBone {boneName=dyn_elbow_flower_l_03}
    componentAppliedLinearAccClamp=<100 100 100>
    componentLinearAccScale=<0.15 0.15 0.15>
    componentLinearVelScale=<0.15 0.15 0.15>
    gravityOverride=<0.8 0.8 0.8>
    physicsBodyDefinitions [{
      sphereCollisionRadius=0
      boundBone {boneName=dyn_elbow_flower_l_03}
      boxExtents=<5 4 1>
      constraintSetup {angularLimitsMax=<0 20 20>, angularLimitsMin=<0 -30 -20>, angularTarget=<0 0 0>}
      localJointOffset=<2 0 0>
     }]
    relativeSpaceBone {boneName=spine_05}
   }
   n51 = "Constraint - dyn_elbow_flower_l_base (PPPPPP)"!(
    ComponentPose=@n50
    Alpha=1
    ConstraintWeights_0=0.1103
    ConstraintWeights_1:float=0.0275
    ConstraintWeights_2:float=0.1689
    ConstraintWeights_3:float=0.0065
    ConstraintWeights_4:float=0.4228
    ConstraintWeights_5:float=0.264
   ) -> (Pose) {
    alphaBoolBlend {blendOption=Linear}
    boneToModify {boneName=dyn_elbow_flower_l_base}
    constraintSetup [
     {targetBone {boneName=upperarm_twist_02_l}}
     {targetBone {boneName=upperarm_twist_01_l}}
     {targetBone {boneName=lowerarm_l}}
     {targetBone {boneName=lowerarm_twist_01_l}}
     {targetBone {boneName=upperarm_l}}
     {targetBone {boneName=lowerarm_twist_02_l}}
    ]
    constraintWeights [1,1,1,1,1,1]
   }
   n52 = Anim Dynamics - Bone: dyn_elbow_flower_l_06(ComponentPose=@n51, Alpha=@n47, ExternalForce="0,0,0") {
    angularDampingOverride=0.9999
    bGravityOverrideInSimSpace=true
    bOverrideAngularDamping=true
    bOverrideLinearDamping=true
    bUseGravityOverride=true
    linearDampingOverride=0.9999
    simulationSpace=RootRelative
    alphaBoolBlend {blendOption=Linear}
    alphaScaleBiasClamp {bInterpResult=true, interpSpeedDecreasing=6, interpSpeedIncreasing=8}
    boundBone {boneName=dyn_elbow_flower_l_06}
    componentAppliedLinearAccClamp=<100 100 100>
    componentLinearAccScale=<0.15 0.15 0.15>
    componentLinearVelScale=<0.15 0.15 0.15>
    gravityOverride=<0.8 0.8 0.8>
    physicsBodyDefinitions [{
      sphereCollisionRadius=0
      boundBone {boneName=dyn_elbow_flower_l_06}
      boxExtents=<5 4 1>
      constraintSetup {angularLimitsMax=<0 15 20>, angularLimitsMin=<0 -15 -20>, angularTarget=<0 0 0>}
      localJointOffset=<2 0 0>
     }]
    relativeSpaceBone {boneName=root}
   }
   n53 = Anim Dynamics - Bone: dyn_elbow_flower_l_05(ComponentPose=@n52, Alpha=@n47, ExternalForce="0,0,0") {
    angularDampingOverride=0.9999
    bGravityOverrideInSimSpace=true
    bOverrideAngularDamping=true
    bOverrideLinearDamping=true
    bUseGravityOverride=true
    linearDampingOverride=0.9999
    simulationSpace=RootRelative
    alphaBoolBlend {blendOption=Linear}
    alphaScaleBiasClamp {bInterpResult=true, interpSpeedDecreasing=6, interpSpeedIncreasing=8}
    boundBone {boneName=dyn_elbow_flower_l_05}
    componentAppliedLinearAccClamp=<100 100 100>
    componentLinearAccScale=<0.15 0.15 0.15>
    componentLinearVelScale=<0.15 0.15 0.15>
    gravityOverride=<0.8 0.8 0.8>
    physicsBodyDefinitions [{
      sphereCollisionRadius=0
      boundBone {boneName=dyn_elbow_flower_l_05}
      boxExtents=<5 5 1.5>
      constraintSetup {angularLimitsMax=<0 10 10>, angularLimitsMin=<0 -10 -10>, angularTarget=<0 0 0>}
      localJointOffset=<1 0 0>
     }]
    relativeSpaceBone {boneName=spine_05}
   }
   n54 = Anim Dynamics - Bone: dyn_elbow_flower_l_04(ComponentPose=@n53, Alpha=@n47, ExternalForce="0,0,0") {
    angularDampingOverride=0.9999
    bGravityOverrideInSimSpace=true
    bOverrideAngularDamping=true
    bOverrideLinearDamping=true
    bUseGravityOverride=true
    linearDampingOverride=0.9999
    simulationSpace=RootRelative
    alphaBoolBlend {blendOption=Linear}
    alphaScaleBiasClamp {bInterpResult=true, interpSpeedDecreasing=6, interpSpeedIncreasing=8}
    boundBone {boneName=dyn_elbow_flower_l_04}
    componentAppliedLinearAccClamp=<100 100 100>
    componentLinearAccScale=<0.15 0.15 0.15>
    componentLinearVelScale=<0.15 0.15 0.15>
    gravityOverride=<0.8 0.8 0.8>
    physicsBodyDefinitions [{
      sphereCollisionRadius=0
      boundBone {boneName=dyn_elbow_flower_l_04}
      boxExtents=<5 4 1.5>
      constraintSetup {angularLimitsMax=<0 10 5>, angularLimitsMin=<0 -1 -5>, angularTarget=<0 0 0>}
      localJointOffset=<1 0 0>
     }]
    relativeSpaceBone {boneName=spine_05}
   }
   n55 = Get Dynamics_Alpha()
   n56 = Anim Dynamics - Bone: dyn_hip_flower_l_01(ComponentPose=@n54, Alpha=@n55, ExternalForce="0,0,0") {
    angularDampingOverride=0.9999
    bGravityOverrideInSimSpace=true
    bOverrideAngularDamping=true
    bOverrideLinearDamping=true
    bUseGravityOverride=true
    linearDampingOverride=0.9999
    simulationSpace=RootRelative
    alphaBoolBlend {blendOption=Linear}
    alphaScaleBiasClamp {bInterpResult=true, interpSpeedDecreasing=4, interpSpeedIncreasing=8}
    boundBone {boneName=dyn_hip_flower_l_01}
    componentAppliedLinearAccClamp=<100 100 100>
    componentLinearAccScale=<0.15 0.15 0.15>
    componentLinearVelScale=<0.15 0.15 0.15>
    gravityOverride=<0.8 0.8 0.8>
    physicsBodyDefinitions [{
      sphereCollisionRadius=0
      boundBone {boneName=dyn_hip_flower_l_01}
      boxExtents=<6 8 2>
      constraintSetup {angularLimitsMax=<0 30 20>, angularLimitsMin=<0 -10 -20>, angularTarget=<0 0 0>}
      localJointOffset=<1 0 0>
     }]
    relativeSpaceBone {boneName=root}
   }
   n57 = Anim Dynamics - Bone: dyn_hip_flower_l_02(ComponentPose=@n56, Alpha=@n55, ExternalForce="0,0,0") {
    angularDampingOverride=0.9999
    bGravityOverrideInSimSpace=true
    bOverrideAngularDamping=true
    bOverrideLinearDamping=true
    bUseGravityOverride=true
    linearDampingOverride=0.9999
    simulationSpace=RootRelative
    alphaBoolBlend {blendOption=Linear}
    alphaScaleBiasClamp {bInterpResult=true, interpSpeedDecreasing=4, interpSpeedIncreasing=8}
    boundBone {boneName=dyn_hip_flower_l_02}
    componentAppliedLinearAccClamp=<100 100 100>
    componentLinearAccScale=<0.15 0.15 0.15>
    componentLinearVelScale=<0.15 0.15 0.15>
    gravityOverride=<0.8 0.8 0.8>
    physicsBodyDefinitions [{
      sphereCollisionRadius=0
      boundBone {boneName=dyn_hip_flower_l_02}
      boxExtents=<6 6 1>
      constraintSetup {angularLimitsMax=<0 30 5>, angularLimitsMin=<0 -5 -5>, angularTarget=<0 0 0>}
      localJointOffset=<1.5 0 0>
     }]
    relativeSpaceBone {boneName=root}
   }
   n58 = Anim Dynamics - Bone: dyn_hip_flower_l_03(ComponentPose=@n57, Alpha=@n55, ExternalForce="0,0,0") {
    angularDampingOverride=0.9999
    bGravityOverrideInSimSpace=true
    bOverrideAngularDamping=true
    bOverrideLinearDamping=true
    bUseGravityOverride=true
    linearDampingOverride=0.9999
    simulationSpace=RootRelative
    alphaBoolBlend {blendOption=Linear}
    alphaScaleBiasClamp {bInterpResult=true, interpSpeedDecreasing=4, interpSpeedIncreasing=8}
    boundBone {boneName=dyn_hip_flower_l_03}
    componentAppliedLinearAccClamp=<100 100 100>
    componentLinearAccScale=<0.15 0.15 0.15>
    componentLinearVelScale=<0.15 0.15 0.15>
    gravityOverride=<0.8 0.8 0.8>
    physicsBodyDefinitions [{
      sphereCollisionRadius=0
      boundBone {boneName=dyn_hip_flower_l_03}
      boxExtents=<6 10 1>
      constraintSetup {angularLimitsMax=<0 20 20>, angularLimitsMin=<0 -10 -20>, angularTarget=<0 0 0>}
      localJointOffset=<2 0 0>
     }]
    relativeSpaceBone {boneName=spine_05}
   }
   n59 = Anim Dynamics - Bone: dyn_hip_flower_l_06(ComponentPose=@n58, Alpha=@n55, ExternalForce="0,0,0") {
    angularDampingOverride=0.9999
    bGravityOverrideInSimSpace=true
    bOverrideAngularDamping=true
    bOverrideLinearDamping=true
    bUseGravityOverride=true
    linearDampingOverride=0.9999
    simulationSpace=RootRelative
    alphaBoolBlend {blendOption=Linear}
    alphaScaleBiasClamp {bInterpResult=true, interpSpeedDecreasing=4, interpSpeedIncreasing=8}
    boundBone {boneName=dyn_hip_flower_l_06}
    componentAppliedLinearAccClamp=<100 100 100>
    componentLinearAccScale=<0.15 0.15 0.15>
    componentLinearVelScale=<0.15 0.15 0.15>
    gravityOverride=<0.8 0.8 0.8>
    physicsBodyDefinitions [{
      sphereCollisionRadius=0
      boundBone {boneName=dyn_hip_flower_l_06}
      boxExtents=<6 8 1>
      constraintSetup {angularLimitsMax=<0 15 20>, angularLimitsMin=<0 -10 -20>, angularTarget=<0 0 0>}
      localJointOffset=<2 0 0>
     }]
    relativeSpaceBone {boneName=root}
   }
   n60 = Anim Dynamics - Bone: dyn_hip_flower_l_05(ComponentPose=@n59, Alpha=@n55, ExternalForce="0,0,0") {
    angularDampingOverride=0.9999
    bGravityOverrideInSimSpace=true
    bOverrideAngularDamping=true
    bOverrideLinearDamping=true
    bUseGravityOverride=true
    linearDampingOverride=0.9999
    simulationSpace=RootRelative
    alphaBoolBlend {blendOption=Linear}
    alphaScaleBiasClamp {bInterpResult=true, interpSpeedDecreasing=4, interpSpeedIncreasing=8}
    boundBone {boneName=dyn_hip_flower_l_05}
    componentAppliedLinearAccClamp=<100 100 100>
    componentLinearAccScale=<0.15 0.15 0.15>
    componentLinearVelScale=<0.15 0.15 0.15>
    gravityOverride=<0.8 0.8 0.8>
    physicsBodyDefinitions [{
      sphereCollisionRadius=0
      boundBone {boneName=dyn_hip_flower_l_05}
      boxExtents=<6 5 1.5>
      constraintSetup {angularLimitsMax=<0 25 10>, angularLimitsMin=<0 -10 -10>, angularTarget=<0 0 0>}
      localJointOffset=<1 0 0>
     }]
    relativeSpaceBone {boneName=spine_05}
   }
   n61 = Anim Dynamics - Bone: dyn_hip_flower_l_04(ComponentPose=@n60, Alpha=@n55, ExternalForce="0,0,0") {
    angularDampingOverride=0.9999
    bGravityOverrideInSimSpace=true
    bOverrideAngularDamping=true
    bOverrideLinearDamping=true
    bUseGravityOverride=true
    linearDampingOverride=0.9999
    simulationSpace=RootRelative
    alphaBoolBlend {blendOption=Linear}
    alphaScaleBiasClamp {bInterpResult=true, interpSpeedDecreasing=4, interpSpeedIncreasing=8}
    boundBone {boneName=dyn_hip_flower_l_04}
    componentAppliedLinearAccClamp=<100 100 100>
    componentLinearAccScale=<0.15 0.15 0.15>
    componentLinearVelScale=<0.15 0.15 0.15>
    gravityOverride=<0.8 0.8 0.8>
    physicsBodyDefinitions [{
      sphereCollisionRadius=0
      boundBone {boneName=dyn_hip_flower_l_04}
      boxExtents=<6 5 1.5>
      constraintSetup {angularLimitsMax=<0 10 5>, angularLimitsMin=<0 -1 -5>, angularTarget=<0 0 0>}
      localJointOffset=<1 0 0>
     }]
    relativeSpaceBone {boneName=spine_05}
   }
   n62 = Get Dynamics_Alpha()
   n63 = Anim Dynamics - Bone: dyn_hip_flower_r_01(ComponentPose=@n61, Alpha=@n62, ExternalForce="0,0,0") {
    angularDampingOverride=0.999
    bGravityOverrideInSimSpace=true
    bOverrideAngularDamping=true
    bOverrideLinearDamping=true
    bUseGravityOverride=true
    linearDampingOverride=0.999
    simulationSpace=RootRelative
    alphaBoolBlend {blendOption=Linear}
    boundBone {boneName=dyn_hip_flower_r_01}
    componentAppliedLinearAccClamp=<1000 1000 1000>
    componentLinearAccScale=<0.15 0.15 0.15>
    componentLinearVelScale=<0.15 0.15 0.15>
    gravityOverride=<0.85 0.85 0.85>
    physicsBodyDefinitions [{
      sphereCollisionRadius=0
      boundBone {boneName=dyn_hip_flower_r_01}
      boxExtents=<4 4 2>
      constraintSetup {angularLimitsMax=<0 30 20>, angularLimitsMin=<0 -10 -20>, angularTarget=<0 0 0>}
      localJointOffset=<1 0 0>
     }]
    relativeSpaceBone {boneName=root}
   }
   n64 = Anim Dynamics - Bone: dyn_hip_flower_r_02(ComponentPose=@n63, Alpha=@n62, ExternalForce="0,0,0") {
    angularDampingOverride=0.999
    bGravityOverrideInSimSpace=true
    bOverrideAngularDamping=true
    bOverrideLinearDamping=true
    bUseGravityOverride=true
    linearDampingOverride=0.999
    simulationSpace=RootRelative
    alphaBoolBlend {blendOption=Linear}
    boundBone {boneName=dyn_hip_flower_r_02}
    componentAppliedLinearAccClamp=<1000 1000 1000>
    componentLinearAccScale=<0.15 0.15 0.15>
    componentLinearVelScale=<0.15 0.15 0.15>
    gravityOverride=<0.85 0.85 0.85>
    physicsBodyDefinitions [{
      sphereCollisionRadius=0
      boundBone {boneName=dyn_hip_flower_r_02}
      boxExtents=<5 4 1>
      constraintSetup {angularLimitsMax=<0 10 5>, angularLimitsMin=<0 -10 -5>, angularTarget=<0 0 0>}
      localJointOffset=<1.5 0 0>
     }]
    relativeSpaceBone {boneName=root}
   }
   n65 = Anim Dynamics - Bone: dyn_hip_flower_r_03(ComponentPose=@n64, Alpha=@n62, ExternalForce="0,0,0") {
    angularDampingOverride=0.999
    bGravityOverrideInSimSpace=true
    bOverrideAngularDamping=true
    bOverrideLinearDamping=true
    bUseGravityOverride=true
    linearDampingOverride=0.999
    simulationSpace=RootRelative
    alphaBoolBlend {blendOption=Linear}
    boundBone {boneName=dyn_hip_flower_r_03}
    componentAppliedLinearAccClamp=<1000 1000 1000>
    componentLinearAccScale=<0.15 0.15 0.15>
    componentLinearVelScale=<0.15 0.15 0.15>
    gravityOverride=<0.85 0.85 0.85>
    physicsBodyDefinitions [{
      sphereCollisionRadius=0
      boundBone {boneName=dyn_hip_flower_r_03}
      boxExtents=<5 4 1>
      constraintSetup {angularLimitsMax=<0 20 20>, angularLimitsMin=<0 -10 -20>, angularTarget=<0 0 0>}
      localJointOffset=<2 0 0>
     }]
    relativeSpaceBone {boneName=spine_05}
   }
   n66 = Anim Dynamics - Bone: dyn_hip_flower_r_06(ComponentPose=@n65, Alpha=@n62, ExternalForce="0,0,0") {
    angularDampingOverride=0.999
    bGravityOverrideInSimSpace=true
    bOverrideAngularDamping=true
    bOverrideLinearDamping=true
    bUseGravityOverride=true
    linearDampingOverride=0.999
    simulationSpace=RootRelative
    alphaBoolBlend {blendOption=Linear}
    boundBone {boneName=dyn_hip_flower_r_06}
    componentAppliedLinearAccClamp=<1000 1000 1000>
    componentLinearAccScale=<0.15 0.15 0.15>
    componentLinearVelScale=<0.15 0.15 0.15>
    gravityOverride=<0.85 0.85 0.85>
    physicsBodyDefinitions [{
      sphereCollisionRadius=0
      boundBone {boneName=dyn_hip_flower_r_06}
      boxExtents=<5 4 1>
      constraintSetup {angularLimitsMax=<0 15 20>, angularLimitsMin=<0 -2 -20>, angularTarget=<0 0 0>}
      localJointOffset=<2 0 0>
     }]
    relativeSpaceBone {boneName=root}
   }
   n67 = Anim Dynamics - Bone: dyn_hip_flower_r_05(ComponentPose=@n66, Alpha=@n62, ExternalForce="0,0,0") {
    angularDampingOverride=0.999
    bGravityOverrideInSimSpace=true
    bOverrideAngularDamping=true
    bOverrideLinearDamping=true
    bUseGravityOverride=true
    linearDampingOverride=0.999
    simulationSpace=RootRelative
    alphaBoolBlend {blendOption=Linear}
    boundBone {boneName=dyn_hip_flower_r_05}
    componentAppliedLinearAccClamp=<1000 1000 1000>
    componentLinearAccScale=<0.15 0.15 0.15>
    componentLinearVelScale=<0.15 0.15 0.15>
    gravityOverride=<0.85 0.85 0.85>
    physicsBodyDefinitions [{
      sphereCollisionRadius=0
      boundBone {boneName=dyn_hip_flower_r_05}
      boxExtents=<5 5 1.5>
      constraintSetup {angularLimitsMax=<0 10 10>, angularLimitsMin=<0 -10 -10>, angularTarget=<0 0 0>}
      localJointOffset=<1 0 0>
     }]
    relativeSpaceBone {boneName=spine_05}
   }
   n68 = Anim Dynamics - Bone: dyn_hip_flower_r_04(ComponentPose=@n67, Alpha=@n62, ExternalForce="0,0,0") {
    angularDampingOverride=0.999
    bGravityOverrideInSimSpace=true
    bOverrideAngularDamping=true
    bOverrideLinearDamping=true
    bUseGravityOverride=true
    linearDampingOverride=0.999
    simulationSpace=RootRelative
    alphaBoolBlend {blendOption=Linear}
    boundBone {boneName=dyn_hip_flower_r_04}
    componentAppliedLinearAccClamp=<1000 1000 1000>
    componentLinearAccScale=<0.15 0.15 0.15>
    componentLinearVelScale=<0.15 0.15 0.15>
    gravityOverride=<0.85 0.85 0.85>
    physicsBodyDefinitions [{
      sphereCollisionRadius=0
      boundBone {boneName=dyn_hip_flower_r_04}
      boxExtents=<5 4 1.5>
      constraintSetup {angularLimitsMax=<0 10 5>, angularLimitsMin=<0 -1 -5>, angularTarget=<0 0 0>}
      localJointOffset=<1 0 0>
     }]
    relativeSpaceBone {boneName=spine_05}
   }
   n69 = Anim Dynamics - Bone: dyn_chest_petals_l_parent_a(ComponentPose=@n68, Alpha=1, ExternalForce="0,0,0") {
    angularDampingOverride=0.9999
    bGravityOverrideInSimSpace=true
    bOverrideAngularDamping=true
    bOverrideLinearDamping=true
    bUseGravityOverride=true
    linearDampingOverride=0.9999
    simulationSpace=BoneRelative
    alphaBoolBlend {blendOption=Linear}
    boundBone {boneName=dyn_chest_petals_l_parent_a}
    componentAppliedLinearAccClamp=<100 100 100>
    componentLinearAccScale=<0.15 0.15 0.15>
    componentLinearVelScale=<0.15 0.15 0.15>
    gravityOverride=<0.7 0.7 0.7>
    physicsBodyDefinitions [{
      sphereCollisionRadius=0
      boundBone {boneName=dyn_chest_petals_l_parent_a}
      boxExtents=<6 8 2>
      constraintSetup {angularLimitsMax=<0 30 20>, angularLimitsMin=<0 -20 -20>, angularTarget=<0 0 0>}
      localJointOffset=<1 0 0>
     }]
    relativeSpaceBone {boneName=root}
   }
   n70 = Anim Dynamics - Bone: dyn_chest_flower_r_parent(ComponentPose=@n69, Alpha=1, ExternalForce="0,0,0") {
    angularDampingOverride=0.9999
    bGravityOverrideInSimSpace=true
    bOverrideAngularDamping=true
    bOverrideLinearDamping=true
    bUseGravityOverride=true
    linearDampingOverride=0.9999
    simulationSpace=RootRelative
    alphaBoolBlend {blendOption=Linear}
    boundBone {boneName=dyn_chest_flower_r_parent}
    componentAppliedLinearAccClamp=<100 100 100>
    componentLinearAccScale=<0.15 0.15 0.15>
    componentLinearVelScale=<0.15 0.15 0.15>
    gravityOverride=<0.7 0.7 0.7>
    physicsBodyDefinitions [{
      sphereCollisionRadius=0
      boundBone {boneName=dyn_chest_flower_r_parent}
      boxExtents=<5 4 1>
      constraintSetup {angularLimitsMax=<0 15 20>, angularLimitsMin=<0 -15 -20>, angularTarget=<0 0 0>}
      localJointOffset=<2 0 0>
     }]
    relativeSpaceBone {boneName=root}
   }
   n71 = "Constraint - dyn_chest_flower_r_base (PPPPP)"!(
    ComponentPose=@n70
    Alpha=1
    ConstraintWeights_0=0.1858
    ConstraintWeights_1:float=0.0102
    ConstraintWeights_2:float=0.0902
    ConstraintWeights_3:float=0.2713
    ConstraintWeights_4:float=0.4425
   ) -> (Pose) {
    alphaBoolBlend {blendOption=Linear}
    boneToModify {boneName=dyn_chest_flower_r_base}
    constraintSetup [
     {targetBone {boneName=clavicle_r}}
     {targetBone {boneName=neck_02}}
     {targetBone {boneName=spine_03}}
     {targetBone {boneName=spine_04}}
     {targetBone {boneName=spine_05}}
    ]
    constraintWeights [1,1,1,1,1]
   }
   n72 = Anim Dynamics - Bone: dyn_arm_vine_l_parent_a(ComponentPose=@n71, Alpha=1, ExternalForce="0,0,0") {
    angularDampingOverride=0.999
    bGravityOverrideInSimSpace=true
    bOverrideAngularDamping=true
    bOverrideLinearDamping=true
    bUseGravityOverride=true
    linearDampingOverride=0.999
    simulationSpace=RootRelative
    alphaBoolBlend {blendOption=Linear}
    boundBone {boneName=dyn_arm_vine_l_parent_a}
    componentAppliedLinearAccClamp=<3000 3000 3000>
    componentLinearAccScale=<0.25 0.25 0.5>
    componentLinearVelScale=<0.25 0.25 0.5>
    gravityOverride=<0.85 0.85 0.85>
    physicsBodyDefinitions [{
      sphereCollisionRadius=0
      boundBone {boneName=dyn_arm_vine_l_parent_a}
      boxExtents=<5 4 1>
      constraintSetup {angularLimitsMax=<0 20 20>, angularLimitsMin=<0 -30 -20>, angularTarget=<0 0 0>}
      localJointOffset=<2 0 0>
     }]
    relativeSpaceBone {boneName=spine_05}
   }
   n73 = Anim Dynamics - Bone: dyn_arm_vine_l_parent_b(ComponentPose=@n72, Alpha=1, ExternalForce="0,0,0") {
    angularDampingOverride=0.999
    bGravityOverrideInSimSpace=true
    bOverrideAngularDamping=true
    bOverrideLinearDamping=true
    bUseGravityOverride=true
    linearDampingOverride=0.999
    simulationSpace=RootRelative
    alphaBoolBlend {blendOption=Linear}
    boundBone {boneName=dyn_arm_vine_l_parent_b}
    componentAppliedLinearAccClamp=<3000 3000 3000>
    componentLinearAccScale=<0.25 0.25 0.5>
    componentLinearVelScale=<0.25 0.25 0.5>
    gravityOverride=<0.85 0.85 0.85>
    physicsBodyDefinitions [{
      sphereCollisionRadius=0
      boundBone {boneName=dyn_arm_vine_l_parent_b}
      boxExtents=<4 4 2>
      constraintSetup {angularLimitsMax=<0 30 20>, angularLimitsMin=<0 -20 -20>, angularTarget=<0 0 0>}
      localJointOffset=<1 0 0>
     }]
    relativeSpaceBone {boneName=spine_05}
   }
   n74 = "Constraint - dyn_arm_vine_l_base_b (PPPPP)"!(
    ComponentPose=@n73
    Alpha=1
    ConstraintWeights_0=0.2029
    ConstraintWeights_1:float=0.0031
    ConstraintWeights_2:float=0.1677
    ConstraintWeights_3:float=0.0856
    ConstraintWeights_4:float=0.5406
   ) -> (Pose) {
    alphaBoolBlend {blendOption=Linear}
    boneToModify {boneName=dyn_arm_vine_l_base_b}
    constraintSetup [
     {targetBone {boneName=upperarm_twist_02_l}}
     {targetBone {boneName=spine_03}}
     {targetBone {boneName=upperarm_l}}
     {targetBone {boneName=clavicle_l}}
     {targetBone {boneName=upperarm_twist_01_l}}
    ]
    constraintWeights [1,1,1,1,1]
   }
   n75 = "Constraint - dyn_hip_flower_r_base (PPP)"!(ComponentPose=@n74, Alpha=1, ConstraintWeights_0=0.5969, ConstraintWeights_1:float=0.3842, ConstraintWeights_2:float=0.0188) -> (Pose) {
    alphaBoolBlend {blendOption=Linear}
    boneToModify {boneName=dyn_hip_flower_r_base}
    constraintSetup [{targetBone {boneName=pelvis}}, {targetBone {boneName=thigh_twist_01_r}}, {targetBone {boneName=spine_01}}]
    constraintWeights [1,1,1]
   }
   n76 = "Constraint - dyn_waist_flower_r_base (PPPPP)"!(
    ComponentPose=@n75
    Alpha=1
    ConstraintWeights_0=0.0225
    ConstraintWeights_1:float=0.2771
    ConstraintWeights_2:float=0.3917
    ConstraintWeights_3:float=0.2818
    ConstraintWeights_4:float=0.0268
   ) -> (Pose) {
    alphaBoolBlend {blendOption=Linear}
    boneToModify {boneName=dyn_waist_flower_r_base}
    constraintSetup [
     {targetBone {boneName=spine_01}}
     {targetBone {boneName=spine_02}}
     {targetBone {boneName=spine_03}}
     {targetBone {boneName=spine_04}}
     {targetBone {boneName=spine_05}}
    ]
    constraintWeights [1,1,1,1,1]
   }
   n77 = "Constraint - dyn_hip_flower_l_base (PPPP)"!(
    ComponentPose=@n76
    Alpha=1
    ConstraintWeights_0=0.3501
    ConstraintWeights_1:float=0.17
    ConstraintWeights_2:float=0.0673
    ConstraintWeights_3:float=0.4126
   ) -> (Pose) {
    alphaBoolBlend {blendOption=Linear}
    boneToModify {boneName=dyn_hip_flower_l_base}
    constraintSetup [{targetBone {boneName=pelvis}}, {targetBone {boneName=spine_02}}, {targetBone {boneName=spine_03}}, {targetBone {boneName=spine_01}}]
    constraintWeights [1,1,1,1]
   }
   n78 = "Constraint - dyn_shoulder_flower_outter_l_master (PPPPPP)"!(
    ComponentPose=@n77
    Alpha=1
    ConstraintWeights_0=0.6377
    ConstraintWeights_1:float=0.0368
    ConstraintWeights_2:float=0.0256
    ConstraintWeights_3:float=0.1872
    ConstraintWeights_4:float=0.094
    ConstraintWeights_5:float=0.0187
   ) -> (Pose) {
    alphaBoolBlend {blendOption=Linear}
    boneToModify {boneName=dyn_shoulder_flower_outter_l_master}
    constraintSetup [
     {targetBone {boneName=clavicle_l}}
     {targetBone {boneName=spine_03}}
     {targetBone {boneName=spine_04}}
     {targetBone {boneName=spine_05}}
     {targetBone {boneName=upperarm_twist_01_l}}
     {targetBone {boneName=neck_02}}
    ]
    constraintWeights [1,1,1,1,1,1]
   }
   n79 = "Constraint - dyn_shoulder_flower_l_parent (P)"!(ComponentPose=@n78, Alpha=1, ConstraintWeights_0=1, ConstraintWeights_1:float=0) -> (Pose) {
    alphaBoolBlend {blendOption=Linear}
    boneToModify {boneName=dyn_shoulder_flower_l_parent}
    constraintSetup [{targetBone {boneName=dyn_shoulder_flower_outter_l_master}}]
    constraintWeights [1,0]
   }
   n80 = Get Took_Damage_Float() -> (Took_Damage_Float:double)
   n81 = "Transform (Modify) Bone - Bone: dyn_waist_flower_inner_r_01"(ComponentPose=@n79, Alpha=@n80, Translation="0,0,0", Rotation="-32.321,175.832,-166.001", Scale="1,0.65,1") {
    rotationMode=BMM_Additive
    rotationSpace=BCS_BoneSpace
    scaleMode=BMM_Replace
    scaleSpace=BCS_BoneSpace
    alphaBoolBlend {blendOption=Linear}
    alphaScaleBiasClamp {bInterpResult=true, interpSpeedDecreasing=4, interpSpeedIncreasing=6}
    boneToModify {boneName=dyn_waist_flower_inner_r_01}
    rotation=<rot -32.3209 175.832 -166.0008>
    scale {y=0.65}
   }
   n82 = "Transform (Modify) Bone - Bone: dyn_waist_flower_inner_r_02"(ComponentPose=@n81, Alpha=@n80, Translation="0,0,0", Rotation="-46.351,-179.820,171.122", Scale="1,1,1") {
    rotationMode=BMM_Additive
    rotationSpace=BCS_BoneSpace
    scaleMode=BMM_Replace
    scaleSpace=BCS_BoneSpace
    alphaBoolBlend {blendOption=Linear}
    alphaScaleBiasClamp {bInterpResult=true, interpSpeedDecreasing=4, interpSpeedIncreasing=6}
    boneToModify {boneName=dyn_waist_flower_inner_r_02}
    rotation=<rot -46.351 -179.82 171.1219>
   }
   n83 = "Transform (Modify) Bone - Bone: dyn_waist_flower_inner_r_03"(ComponentPose=@n82, Alpha=@n80, Translation="0,0,0", Rotation="-33.423,-178.605,-156.287", Scale="1,0.65,1") {
    rotationMode=BMM_Additive
    rotationSpace=BCS_BoneSpace
    scaleMode=BMM_Replace
    scaleSpace=BCS_BoneSpace
    alphaBoolBlend {blendOption=Linear}
    alphaScaleBiasClamp {bInterpResult=true, interpSpeedDecreasing=4, interpSpeedIncreasing=6}
    boneToModify {boneName=dyn_waist_flower_inner_r_03}
    rotation=<rot -33.423 -178.605 -156.2871>
    scale {y=0.65}
   }
   n84 = Get Took_Damage_Float()
   n85 = "Transform (Modify) Bone - Bone: dyn_waist_flower_inner_r_06"(ComponentPose=@n83, Alpha=@n84, Translation="0,0,0", Rotation="-42.807,-180.000,180.000", Scale="1,1,1") {
    rotationMode=BMM_Additive
    rotationSpace=BCS_BoneSpace
    scaleMode=BMM_Replace
    scaleSpace=BCS_BoneSpace
    alphaBoolBlend {blendOption=Linear}
    alphaScaleBiasClamp {bInterpResult=true, interpSpeedDecreasing=4, interpSpeedIncreasing=6}
    boneToModify {boneName=dyn_waist_flower_inner_r_06}
    rotation=<rot -42.807 -180 180>
   }
   n86 = "Transform (Modify) Bone - Bone: dyn_waist_flower_inner_r_05"(ComponentPose=@n85, Alpha=@n84, Translation="0,0,0", Rotation="-30.491,179.411,-166.981", Scale="1,0.65,1") {
    rotationMode=BMM_Additive
    rotationSpace=BCS_BoneSpace
    scaleMode=BMM_Replace
    scaleSpace=BCS_BoneSpace
    alphaBoolBlend {blendOption=Linear}
    alphaScaleBiasClamp {bInterpResult=true, interpSpeedDecreasing=4, interpSpeedIncreasing=6}
    boneToModify {boneName=dyn_waist_flower_inner_r_05}
    rotation=<rot -30.491 179.411 -166.9808>
    scale {y=0.65}
   }
   n87 = "Transform (Modify) Bone - Bone: dyn_waist_flower_inner_r_04"(ComponentPose=@n86, Alpha=@n84, Translation="0,0,0", Rotation="-41.333,179.673,173.843", Scale="1,1,1") {
    rotationMode=BMM_Additive
    rotationSpace=BCS_BoneSpace
    scaleMode=BMM_Replace
    scaleSpace=BCS_BoneSpace
    alphaBoolBlend {blendOption=Linear}
    alphaScaleBiasClamp {bInterpResult=true, interpSpeedDecreasing=4, interpSpeedIncreasing=6}
    boneToModify {boneName=dyn_waist_flower_inner_r_04}
    rotation=<rot -41.333 179.673 173.843>
   }
   n88 = Get Took_Damage_Float()
   n89 = "Transform (Modify) Bone - Bone: dyn_hip_flower_r_01"(ComponentPose=@n87, Alpha=@n88, Translation="0,0,0", Rotation="-27.507,176.650,-179.453", Scale="1,0.65,1") {
    rotationMode=BMM_Additive
    rotationSpace=BCS_BoneSpace
    scaleMode=BMM_Replace
    scaleSpace=BCS_BoneSpace
    alphaBoolBlend {blendOption=Linear}
    alphaScaleBiasClamp {bInterpResult=true, interpSpeedDecreasing=4, interpSpeedIncreasing=6}
    boneToModify {boneName=dyn_hip_flower_r_01}
    rotation=<rot -27.507 176.65 -179.453>
    scale {y=0.65}
   }
   n90 = "Transform (Modify) Bone - Bone: dyn_hip_flower_r_02"(ComponentPose=@n89, Alpha=@n88, Translation="0,0,0", Rotation="-46.368,-178.846,167.614", Scale="1,1,1") {
    rotationMode=BMM_Additive
    rotationSpace=BCS_BoneSpace
    scaleMode=BMM_Replace
    scaleSpace=BCS_BoneSpace
    alphaBoolBlend {blendOption=Linear}
    alphaScaleBiasClamp {bInterpResult=true, interpSpeedDecreasing=4, interpSpeedIncreasing=6}
    boneToModify {boneName=dyn_hip_flower_r_02}
    rotation=<rot -46.3679 -178.846 167.6138>
   }
   n91 = "Transform (Modify) Bone - Bone: dyn_hip_flower_r_03"(ComponentPose=@n90, Alpha=@n88, Translation="0,0,0", Rotation="-33.333,-178.508,-177.812", Scale="1,0.65,1") {
    rotationMode=BMM_Additive
    rotationSpace=BCS_BoneSpace
    scaleMode=BMM_Replace
    scaleSpace=BCS_BoneSpace
    alphaBoolBlend {blendOption=Linear}
    alphaScaleBiasClamp {bInterpResult=true, interpSpeedDecreasing=4, interpSpeedIncreasing=6}
    boneToModify {boneName=dyn_hip_flower_r_03}
    rotation=<rot -33.3328 -178.508 -177.812>
    scale {y=0.65}
   }
   n92 = Get Took_Damage_Float()
   n93 = "Transform (Modify) Bone - Bone: dyn_hip_flower_r_06"(ComponentPose=@n91, Alpha=@n92, Translation="0,0,0", Rotation="-41.291,-179.508,-172.200", Scale="1,1,1") {
    rotationMode=BMM_Additive
    rotationSpace=BCS_BoneSpace
    scaleMode=BMM_Replace
    scaleSpace=BCS_BoneSpace
    alphaBoolBlend {blendOption=Linear}
    alphaScaleBiasClamp {bInterpResult=true, interpSpeedDecreasing=4, interpSpeedIncreasing=6}
    boneToModify {boneName=dyn_hip_flower_r_06}
    rotation=<rot -41.291 -179.508 -172.2>
   }
   n94 = "Transform (Modify) Bone - Bone: dyn_hip_flower_r_05"(ComponentPose=@n93, Alpha=@n92, Translation="0,0,0", Rotation="-38.601,-175.548,177.110", Scale="1,0.65,1") {
    rotationMode=BMM_Additive
    rotationSpace=BCS_BoneSpace
    scaleMode=BMM_Replace
    scaleSpace=BCS_BoneSpace
    alphaBoolBlend {blendOption=Linear}
    alphaScaleBiasClamp {bInterpResult=true, interpSpeedDecreasing=4, interpSpeedIncreasing=6}
    boneToModify {boneName=dyn_hip_flower_r_05}
    rotation=<rot -38.601 -175.548 177.11>
    scale {y=0.65}
   }
   n95 = "Transform (Modify) Bone - Bone: dyn_hip_flower_r_04"(ComponentPose=@n94, Alpha=@n92, Translation="0,0,0", Rotation="-48.557,-168.362,169.907", Scale="1,1,1") {
    rotationMode=BMM_Additive
    rotationSpace=BCS_BoneSpace
    scaleMode=BMM_Replace
    scaleSpace=BCS_BoneSpace
    alphaBoolBlend {blendOption=Linear}
    alphaScaleBiasClamp {bInterpResult=true, interpSpeedDecreasing=4, interpSpeedIncreasing=6}
    boneToModify {boneName=dyn_hip_flower_r_04}
    rotation=<rot -48.5569 -168.3619 169.9069>
   }
   n96 = Get Took_Damage_Float()
   n97 = "Transform (Modify) Bone - Bone: dyn_hip_flower_l_01"(ComponentPose=@n95, Alpha=@n96, Translation="0,0,0", Rotation="-28.509,175.196,166.763", Scale="1,0.65,1") {
    rotationMode=BMM_Additive
    rotationSpace=BCS_BoneSpace
    scaleMode=BMM_Replace
    scaleSpace=BCS_BoneSpace
    alphaBoolBlend {blendOption=Linear}
    alphaScaleBiasClamp {bInterpResult=true, interpSpeedDecreasing=4, interpSpeedIncreasing=6}
    boneToModify {boneName=dyn_hip_flower_l_01}
    rotation=<rot -28.509 175.196 166.763>
    scale {y=0.65}
   }
   n98 = "Transform (Modify) Bone - Bone: dyn_hip_flower_l_02"(ComponentPose=@n97, Alpha=@n96, Translation="0,0,0", Rotation="-41.590,167.108,-156.138", Scale="1,1,1") {
    rotationMode=BMM_Additive
    rotationSpace=BCS_BoneSpace
    scaleMode=BMM_Replace
    scaleSpace=BCS_BoneSpace
    alphaBoolBlend {blendOption=Linear}
    alphaScaleBiasClamp {bInterpResult=true, interpSpeedDecreasing=4, interpSpeedIncreasing=6}
    boneToModify {boneName=dyn_hip_flower_l_02}
    rotation=<rot -41.5899 167.1078 -156.1381>
   }
   n99 = "Transform (Modify) Bone - Bone: dyn_hip_flower_l_03"(ComponentPose=@n98, Alpha=@n96, Translation="0,0,0", Rotation="-34.026,-175.607,159.024", Scale="1,0.65,1") {
    rotationMode=BMM_Additive
    rotationSpace=BCS_BoneSpace
    scaleMode=BMM_Replace
    scaleSpace=BCS_BoneSpace
    alphaBoolBlend {blendOption=Linear}
    alphaScaleBiasClamp {bInterpResult=true, interpSpeedDecreasing=4, interpSpeedIncreasing=6}
    boneToModify {boneName=dyn_hip_flower_l_03}
    rotation=<rot -34.026 -175.607 159.024>
    scale {y=0.65}
   }
   n100 = Get Took_Damage_Float()
   n101 = "Transform (Modify) Bone - Bone: dyn_hip_flower_l_06"(ComponentPose=@n99, Alpha=@n100, Translation="0,0,0", Rotation="-36.648,-180.000,180.000", Scale="1,1,1") {
    rotationMode=BMM_Additive
    rotationSpace=BCS_BoneSpace
    scaleMode=BMM_Replace
    scaleSpace=BCS_BoneSpace
    alphaBoolBlend {blendOption=Linear}
    alphaScaleBiasClamp {bInterpResult=true, interpSpeedDecreasing=4, interpSpeedIncreasing=6}
    boneToModify {boneName=dyn_hip_flower_l_06}
    rotation=<rot -36.648 -180 180>
   }
   n102 = "Transform (Modify) Bone - Bone: dyn_hip_flower_l_05"(ComponentPose=@n101, Alpha=@n100, Translation="0,0,0", Rotation="-30.356,-174.568,-166.523", Scale="1,0.65,1") {
    rotationMode=BMM_Additive
    rotationSpace=BCS_BoneSpace
    scaleMode=BMM_Replace
    scaleSpace=BCS_BoneSpace
    alphaBoolBlend {blendOption=Linear}
    alphaScaleBiasClamp {bInterpResult=true, interpSpeedDecreasing=4, interpSpeedIncreasing=6}
    boneToModify {boneName=dyn_hip_flower_l_05}
    rotation=<rot -30.356 -174.568 -166.5229>
    scale {y=0.65}
   }
   n103 = "Transform (Modify) Bone - Bone: dyn_hip_flower_l_04"(ComponentPose=@n102, Alpha=@n100, Translation="0,0,0", Rotation="-45.330,-175.446,-178.280", Scale="1,1,1") {
    rotationMode=BMM_Additive
    rotationSpace=BCS_BoneSpace
    scaleMode=BMM_Replace
    scaleSpace=BCS_BoneSpace
    alphaBoolBlend {blendOption=Linear}
    alphaScaleBiasClamp {bInterpResult=true, interpSpeedDecreasing=4, interpSpeedIncreasing=6}
    boneToModify {boneName=dyn_hip_flower_l_04}
    rotation=<rot -45.33 -175.446 -178.28>
   }
   n104 = Get Took_Damage_Float()
   n105 = "Transform (Modify) Bone - Bone: dyn_elbow_flower_l_01"(ComponentPose=@n103, Alpha=@n104, Translation="0,0,0", Rotation="-32.771,-180.000,180.000", Scale="1,0.65,1") {
    rotationMode=BMM_Additive
    rotationSpace=BCS_BoneSpace
    scaleMode=BMM_Replace
    scaleSpace=BCS_BoneSpace
    alphaBoolBlend {blendOption=Linear}
    alphaScaleBiasClamp {bInterpResult=true, interpSpeedDecreasing=4, interpSpeedIncreasing=6}
    boneToModify {boneName=dyn_elbow_flower_l_01}
    rotation=<rot -32.771 -180 180>
    scale {y=0.65}
   }
   n106 = "Transform (Modify) Bone - Bone: dyn_elbow_flower_l_02"(ComponentPose=@n105, Alpha=@n104, Translation="0,0,0", Rotation="-42.391,-179.145,173.500", Scale="1,1,1") {
    rotationMode=BMM_Additive
    rotationSpace=BCS_BoneSpace
    scaleMode=BMM_Replace
    scaleSpace=BCS_BoneSpace
    alphaBoolBlend {blendOption=Linear}
    alphaScaleBiasClamp {bInterpResult=true, interpSpeedDecreasing=4, interpSpeedIncreasing=6}
    boneToModify {boneName=dyn_elbow_flower_l_02}
    rotation=<rot -42.3909 -179.145 173.5>
   }
   n107 = "Transform (Modify) Bone - Bone: dyn_elbow_flower_l_03"(ComponentPose=@n106, Alpha=@n104, Translation="0,0,0", Rotation="-31.811,-172.637,-143.614", Scale="1,0.65,1") {
    rotationMode=BMM_Additive
    rotationSpace=BCS_BoneSpace
    scaleMode=BMM_Replace
    scaleSpace=BCS_BoneSpace
    alphaBoolBlend {blendOption=Linear}
    alphaScaleBiasClamp {bInterpResult=true, interpSpeedDecreasing=4, interpSpeedIncreasing=6}
    boneToModify {boneName=dyn_elbow_flower_l_03}
    rotation=<rot -31.811 -172.637 -143.614>
    scale {y=0.65}
   }
   n108 = Get Took_Damage_Float()
   n109 = "Transform (Modify) Bone - Bone: dyn_elbow_flower_l_06"(ComponentPose=@n107, Alpha=@n108, Translation="0,0,0", Rotation="-42.824,180.000,-180.000", Scale="1,1,1") {
    rotationMode=BMM_Additive
    rotationSpace=BCS_BoneSpace
    scaleMode=BMM_Replace
    scaleSpace=BCS_BoneSpace
    alphaBoolBlend {blendOption=Linear}
    alphaScaleBiasClamp {bInterpResult=true, interpSpeedDecreasing=4, interpSpeedIncreasing=6}
    boneToModify {boneName=dyn_elbow_flower_l_06}
    rotation=<rot -42.824 180 -180>
   }
   n110 = "Transform (Modify) Bone - Bone: dyn_elbow_flower_l_05"(ComponentPose=@n109, Alpha=@n108, Translation="0,0,0", Rotation="-30.775,175.487,-172.211", Scale="1,0.65,1") {
    rotationMode=BMM_Additive
    rotationSpace=BCS_BoneSpace
    scaleMode=BMM_Replace
    scaleSpace=BCS_BoneSpace
    alphaBoolBlend {blendOption=Linear}
    alphaScaleBiasClamp {bInterpResult=true, interpSpeedDecreasing=4, interpSpeedIncreasing=6}
    boneToModify {boneName=dyn_elbow_flower_l_05}
    rotation=<rot -30.775 175.487 -172.2109>
    scale {y=0.65}
   }
   n111 = "Transform (Modify) Bone - Bone: dyn_elbow_flower_l_04"(ComponentPose=@n110, Alpha=@n108, Translation="0,0,0", Rotation="-43.933,178.037,-161.433", Scale="1,1,1") {
    rotationMode=BMM_Additive
    rotationSpace=BCS_BoneSpace
    scaleMode=BMM_Replace
    scaleSpace=BCS_BoneSpace
    alphaBoolBlend {blendOption=Linear}
    alphaScaleBiasClamp {bInterpResult=true, interpSpeedDecreasing=4, interpSpeedIncreasing=6}
    boneToModify {boneName=dyn_elbow_flower_l_04}
    rotation=<rot -43.933 178.037 -161.433>
   }
   n112 = Get Took_Damage_Float()
   n113 = "Transform (Modify) Bone - Bone: dyn_shoulder_flower_inner_l_01"(ComponentPose=@n111, Alpha=@n112, Translation="0,0,0", Rotation="-34.599,176.673,-164.702", Scale="1,0.6,1") {
    rotationMode=BMM_Additive
    rotationSpace=BCS_BoneSpace
    scaleMode=BMM_Replace
    scaleSpace=BCS_BoneSpace
    alphaBoolBlend {blendOption=Linear}
    alphaScaleBiasClamp {bInterpResult=true, interpSpeedDecreasing=4, interpSpeedIncreasing=6}
    boneToModify {boneName=dyn_shoulder_flower_inner_l_01}
    rotation=<rot -34.599 176.673 -164.7019>
    scale {y=0.6}
   }
   n114 = "Transform (Modify) Bone - Bone: dyn_shoulder_flower_inner_l_02"(ComponentPose=@n113, Alpha=@n112, Translation="0,0,0", Rotation="-44.507,-179.974,-173.436", Scale="1,1,1") {
    rotationMode=BMM_Additive
    rotationSpace=BCS_BoneSpace
    scaleMode=BMM_Replace
    scaleSpace=BCS_BoneSpace
    alphaBoolBlend {blendOption=Linear}
    alphaScaleBiasClamp {bInterpResult=true, interpSpeedDecreasing=4, interpSpeedIncreasing=6}
    boneToModify {boneName=dyn_shoulder_flower_inner_l_02}
    rotation=<rot -44.507 -179.974 -173.436>
   }
   n115 = "Transform (Modify) Bone - Bone: dyn_shoulder_flower_inner_l_03"(ComponentPose=@n114, Alpha=@n112, Translation="0,0,0", Rotation="-29.999,174.253,-170.488", Scale="1,0.6,1") {
    rotationMode=BMM_Additive
    rotationSpace=BCS_BoneSpace
    scaleMode=BMM_Replace
    scaleSpace=BCS_BoneSpace
    alphaBoolBlend {blendOption=Linear}
    alphaScaleBiasClamp {bInterpResult=true, interpSpeedDecreasing=4, interpSpeedIncreasing=6}
    boneToModify {boneName=dyn_shoulder_flower_inner_l_03}
    rotation=<rot -29.999 174.253 -170.488>
    scale {y=0.6}
   }
   n116 = Get Took_Damage_Float()
   n117 = "Transform (Modify) Bone - Bone: dyn_shoulder_flower_inner_l_06"(ComponentPose=@n115, Alpha=@n116, Translation="0,0,0", Rotation="-45.942,173.295,-171.713", Scale="1,1,1") {
    rotationMode=BMM_Additive
    rotationSpace=BCS_BoneSpace
    scaleMode=BMM_Replace
    scaleSpace=BCS_BoneSpace
    alphaBoolBlend {blendOption=Linear}
    alphaScaleBiasClamp {bInterpResult=true, interpSpeedDecreasing=4, interpSpeedIncreasing=6}
    boneToModify {boneName=dyn_shoulder_flower_inner_l_06}
    rotation=<rot -45.9419 173.295 -171.7129>
   }
   n118 = "Transform (Modify) Bone - Bone: dyn_shoulder_flower_inner_l_05"(ComponentPose=@n117, Alpha=@n116, Translation="0,0,0", Rotation="-36.718,-174.246,176.870", Scale="1,0.6,1") {
    rotationMode=BMM_Additive
    rotationSpace=BCS_BoneSpace
    scaleMode=BMM_Replace
    scaleSpace=BCS_BoneSpace
    alphaBoolBlend {blendOption=Linear}
    alphaScaleBiasClamp {bInterpResult=true, interpSpeedDecreasing=4, interpSpeedIncreasing=6}
    boneToModify {boneName=dyn_shoulder_flower_inner_l_05}
    rotation=<rot -36.7178 -174.2459 176.8699>
    scale {y=0.6}
   }
   n119 = "Transform (Modify) Bone - Bone: dyn_shoulder_flower_inner_l_04"(ComponentPose=@n118, Alpha=@n116, Translation="0,0,0", Rotation="-51.620,-177.802,178.410", Scale="1,1,1") {
    rotationMode=BMM_Additive
    rotationSpace=BCS_BoneSpace
    scaleMode=BMM_Replace
    scaleSpace=BCS_BoneSpace
    alphaBoolBlend {blendOption=Linear}
    alphaScaleBiasClamp {bInterpResult=true, interpSpeedDecreasing=4, interpSpeedIncreasing=6}
    boneToModify {boneName=dyn_shoulder_flower_inner_l_04}
    rotation=<rot -51.6197 -177.802 178.41>
   }
   n120 = Anim Dynamics - Bone: dyn_back_petals_mid_parent_a(ComponentPose=@n119, Alpha=1, ExternalForce="0,0,0") {
    angularDampingOverride=0.9999
    bGravityOverrideInSimSpace=true
    bOverrideAngularDamping=true
    bOverrideLinearDamping=true
    bUseGravityOverride=true
    gravityScale=0.5
    linearDampingOverride=0.9999
    simulationSpace=RootRelative
    alphaBoolBlend {blendOption=Linear}
    alphaScaleBiasClamp {bInterpResult=true, bias=0.2, interpSpeedDecreasing=4, interpSpeedIncreasing=6, scale=0.8}
    boundBone {boneName=dyn_back_petals_mid_parent_a}
    componentAppliedLinearAccClamp=<100 100 100>
    componentLinearAccScale=<0.25 0.25 0.5>
    componentLinearVelScale=<0.25 0.25 0.5>
    gravityOverride=<0.75 0.75 0.75>
    physicsBodyDefinitions [{
      sphereCollisionRadius=0
      boundBone {boneName=dyn_back_petals_mid_parent_a}
      boxExtents=<4 4 2>
      constraintSetup {angularLimitsMax=<0 35 30>, angularLimitsMin=<0 0 -30>, angularTarget=<0 0 0>}
      localJointOffset=<1 0 0>
     }]
    relativeSpaceBone {boneName=root}
   }
   n121 = Anim Dynamics - Bone: dyn_back_petals_mid_parent_b(ComponentPose=@n120, Alpha=0.75, ExternalForce="0,0,0") {
    angularDampingOverride=0.9999
    bGravityOverrideInSimSpace=true
    bOverrideAngularDamping=true
    bOverrideLinearDamping=true
    bUseGravityOverride=true
    gravityScale=0.5
    linearDampingOverride=0.9999
    simulationSpace=RootRelative
    alphaBoolBlend {blendOption=Linear}
    alphaScaleBiasClamp {bInterpResult=true, bias=0.2, interpSpeedDecreasing=4, interpSpeedIncreasing=6, scale=0.8}
    boundBone {boneName=dyn_back_petals_mid_parent_b}
    componentAppliedLinearAccClamp=<100 100 100>
    componentLinearAccScale=<0.15 0.15 0.15>
    componentLinearVelScale=<0.15 0.15 0.15>
    gravityOverride=<0.7 0.7 0.7>
    physicsBodyDefinitions [{
      sphereCollisionRadius=0
      boundBone {boneName=dyn_back_petals_mid_parent_b}
      boxExtents=<8 10 1>
      constraintSetup {angularLimitsMax=<0 25 20>, angularLimitsMin=<0 -1 -10>, angularTarget=<0 0 0>}
      localJointOffset=<1.5 0 0>
     }]
    relativeSpaceBone {boneName=root}
   }
   n122 = Anim Dynamics - Bone: dyn_back_petals_mid_parent_c(ComponentPose=@n121, Alpha=0.5, ExternalForce="0,0,0") {
    angularDampingOverride=0.9999
    bGravityOverrideInSimSpace=true
    bOverrideAngularDamping=true
    bOverrideLinearDamping=true
    bUseGravityOverride=true
    gravityScale=0.5
    linearDampingOverride=0.9999
    simulationSpace=RootRelative
    alphaBoolBlend {blendOption=Linear}
    alphaScaleBiasClamp {bInterpResult=true, bias=0.2, interpSpeedDecreasing=4, interpSpeedIncreasing=6, scale=0.8}
    boundBone {boneName=dyn_back_petals_mid_parent_c}
    componentAppliedLinearAccClamp=<100 100 100>
    componentLinearAccScale=<0.15 0.15 0.15>
    componentLinearVelScale=<0.15 0.15 0.15>
    gravityOverride=<0.7 7 7>
    physicsBodyDefinitions [{
      sphereCollisionRadius=0
      boundBone {boneName=dyn_back_petals_mid_parent_c}
      boxExtents=<8 10 1>
      constraintSetup {angularLimitsMax=<0 15 20>, angularLimitsMin=<0 -1 -20>, angularTarget=<0 0 0>}
      localJointOffset=<2 0 0>
     }]
    relativeSpaceBone {boneName=spine_05}
   }
   n123 = Anim Dynamics - Bone: dyn_back_petals_mid_parent_d(ComponentPose=@n122, Alpha=1, ExternalForce="0,0,0") {
    angularDampingOverride=0.9999
    bGravityOverrideInSimSpace=true
    bOverrideAngularDamping=true
    bOverrideLinearDamping=true
    bUseGravityOverride=true
    gravityScale=0.5
    linearDampingOverride=0.9999
    simulationSpace=RootRelative
    alphaBoolBlend {blendOption=Linear}
    alphaScaleBiasClamp {bInterpResult=true, bias=0.2, interpSpeedDecreasing=4, interpSpeedIncreasing=6, scale=0.8}
    boundBone {boneName=dyn_back_petals_mid_parent_d}
    componentAppliedLinearAccClamp=<100 100 100>
    componentLinearAccScale=<0.25 0.25 0.5>
    componentLinearVelScale=<0.25 0.25 0.5>
    gravityOverride=<0.75 0.75 0.75>
    physicsBodyDefinitions [{
      sphereCollisionRadius=0
      boundBone {boneName=dyn_back_petals_mid_parent_d}
      boxExtents=<6 6 1.5>
      constraintSetup {angularLimitsMax=<0 20 20>, angularLimitsMin=<0 -10 -20>, angularTarget=<0 0 0>}
      localJointOffset=<1 0 0>
     }]
    relativeSpaceBone {boneName=root}
   }
   n124 = Anim Dynamics - Bone: dyn_back_petals_mid_parent_g(ComponentPose=@n123, Alpha=1, ExternalForce="0,0,0") {
    angularDampingOverride=0.9999
    bGravityOverrideInSimSpace=true
    bOverrideAngularDamping=true
    bOverrideLinearDamping=true
    bUseGravityOverride=true
    gravityScale=0.5
    linearDampingOverride=0.9999
    simulationSpace=BoneRelative
    alphaBoolBlend {blendOption=Linear}
    alphaScaleBiasClamp {bInterpResult=true, bias=0.2, interpSpeedDecreasing=4, interpSpeedIncreasing=6, scale=0.8}
    boundBone {boneName=dyn_back_petals_mid_parent_g}
    componentAppliedLinearAccClamp=<100 100 100>
    componentLinearAccScale=<0.25 0.25 0.5>
    componentLinearVelScale=<0.25 0.25 0.5>
    gravityOverride=<0.75 0.75 0.75>
    physicsBodyDefinitions [{
      sphereCollisionRadius=0
      boundBone {boneName=dyn_back_petals_mid_parent_g}
      boxExtents=<5 4 1>
      constraintSetup {angularLimitsMax=<0 20 30>, angularLimitsMin=<0 -15 -30>, angularTarget=<0 0 0>}
      localJointOffset=<2 0 0>
     }]
    relativeSpaceBone {boneName=root}
   }
   n125 = Anim Dynamics - Bone: dyn_back_petals_mid_parent_f(ComponentPose=@n124, Alpha=1, ExternalForce="0,0,0") {
    angularDampingOverride=0.9999
    bGravityOverrideInSimSpace=true
    bOverrideAngularDamping=true
    bOverrideLinearDamping=true
    bUseGravityOverride=true
    gravityScale=0.5
    linearDampingOverride=0.9999
    simulationSpace=RootRelative
    alphaBoolBlend {blendOption=Linear}
    alphaScaleBiasClamp {bInterpResult=true, bias=0.2, interpSpeedDecreasing=4, interpSpeedIncreasing=6, scale=0.8}
    boundBone {boneName=dyn_back_petals_mid_parent_f}
    componentAppliedLinearAccClamp=<100 100 100>
    componentLinearAccScale=<0.25 0.25 0.5>
    componentLinearVelScale=<0.25 0.25 0.5>
    gravityOverride=<0.75 0.75 0.75>
    physicsBodyDefinitions [{
      sphereCollisionRadius=0
      boundBone {boneName=dyn_back_petals_mid_parent_f}
      boxExtents=<5 5 1.5>
      constraintSetup {angularLimitsMax=<0 15 20>, angularLimitsMin=<0 -10 -20>, angularTarget=<0 0 0>}
      localJointOffset=<1 0 0>
     }]
    relativeSpaceBone {boneName=spine_05}
   }
   n126 = Anim Dynamics - Bone: dyn_back_petals_mid_parent_e(ComponentPose=@n125, Alpha=1, ExternalForce="0,0,0") {
    angularDampingOverride=0.9999
    bGravityOverrideInSimSpace=true
    bOverrideAngularDamping=true
    bOverrideLinearDamping=true
    bUseGravityOverride=true
    gravityScale=0.5
    linearDampingOverride=0.9999
    simulationSpace=RootRelative
    alphaBoolBlend {blendOption=Linear}
    alphaScaleBiasClamp {bInterpResult=true, bias=0.2, interpSpeedDecreasing=4, interpSpeedIncreasing=6, scale=0.8}
    boundBone {boneName=dyn_back_petals_mid_parent_e}
    componentAppliedLinearAccClamp=<100 100 100>
    componentLinearAccScale=<0.25 0.25 0.5>
    componentLinearVelScale=<0.25 0.25 0.5>
    gravityOverride=<0.7 0.7 0.7>
    physicsBodyDefinitions [{
      sphereCollisionRadius=0
      boundBone {boneName=dyn_back_petals_mid_parent_e}
      boxExtents=<8 10 1.5>
      constraintSetup {angularLimitsMax=<0 15 10>, angularLimitsMin=<0 -1 -10>, angularTarget=<0 0 0>}
      localJointOffset=<1 0 0>
     }]
    relativeSpaceBone {boneName=spine_05}
   }
   n127 = "Constraint - dyn_arm_petals_r_base_b (PPPPPP)"!(
    ComponentPose=@n126
    Alpha=1
    ConstraintWeights_0=0.037
    ConstraintWeights_1:float=0.5255
    ConstraintWeights_2:float=0.0297
    ConstraintWeights_3:float=0.0402
    ConstraintWeights_4:float=0.3655
    ConstraintWeights_5:float=0.0023
   ) -> (Pose) {
    alphaBoolBlend {blendOption=Linear}
    boneToModify {boneName=dyn_arm_petals_r_base_b}
    constraintSetup [
     {targetBone {boneName=head}}
     {targetBone {boneName=upperarm_twist_01_r}}
     {targetBone {boneName=spine_03}}
     {targetBone {boneName=spine_05}}
     {targetBone {boneName=clavicle_r}}
     {targetBone {boneName=neck_02}}
    ]
    constraintWeights [1,1,1,1,1,1]
   }
   n128 = "Constraint - dyn_arm_petals_r_base_c (PPPPPPPP)"!(
    ComponentPose=@n127
    Alpha=1
    ConstraintWeights_0=0.0036
    ConstraintWeights_1:float=0.1225
    ConstraintWeights_2:float=0.0078
    ConstraintWeights_3:float=0.6648
    ConstraintWeights_4:float=0.0873
    ConstraintWeights_5:float=0.0068
    ConstraintWeights_6:float=0
    ConstraintWeights_7:float=0.1072
   ) -> (Pose) {
    alphaBoolBlend {blendOption=Linear}
    boneToModify {boneName=dyn_arm_petals_r_base_c}
    constraintSetup [
     {targetBone {boneName=lowerarm_twist_02_r}}
     {targetBone {boneName=upperarm_r}}
     {targetBone {boneName=lowerarm_r}}
     {targetBone {boneName=upperarm_twist_01_r}}
     {targetBone {boneName=upperarm_twist_02_r}}
     {targetBone {boneName=spine_03}}
     {targetBone {boneName=spine_05}}
     {targetBone {boneName=clavicle_r}}
    ]
    constraintWeights [1,1,1,1,1,1,1,1]
   }
   n129 = "Constraint - dyn_arm_petals_r_base_d (PPPPPP)"!(
    ComponentPose=@n128
    Alpha=1
    ConstraintWeights_0=0.0797
    ConstraintWeights_1:float=0.4041
    ConstraintWeights_2:float=0.1283
    ConstraintWeights_3:float=0.2344
    ConstraintWeights_4:float=0.1305
    ConstraintWeights_5:float=0.023
   ) -> (Pose) {
    alphaBoolBlend {blendOption=Linear}
    boneToModify {boneName=dyn_arm_petals_r_base_d}
    constraintSetup [
     {targetBone {boneName=lowerarm_twist_02_r}}
     {targetBone {boneName=upperarm_r}}
     {targetBone {boneName=lowerarm_r}}
     {targetBone {boneName=upperarm_twist_01_r}}
     {targetBone {boneName=upperarm_twist_02_r}}
     {targetBone {boneName=clavicle_r}}
    ]
    constraintWeights [1,1,1,1,1,1]
   }
   n130 = Anim Dynamics - Bone: dyn_arm_petals_r_parent_a(ComponentPose=@n129, Alpha=1, ExternalForce="0,0,0") {
    angularDampingOverride=0.9999
    bGravityOverrideInSimSpace=true
    bOverrideAngularDamping=true
    bOverrideLinearDamping=true
    bUseGravityOverride=true
    linearDampingOverride=0.9999
    simulationSpace=RootRelative
    alphaBoolBlend {blendOption=Linear}
    boundBone {boneName=dyn_arm_petals_r_parent_a}
    componentAppliedLinearAccClamp=<100 100 100>
    componentLinearAccScale=<0.25 0.25 0.5>
    componentLinearVelScale=<0.25 0.25 0.5>
    gravityOverride=<0.85 0.85 0.85>
    physicsBodyDefinitions [{
      sphereCollisionRadius=0
      boundBone {boneName=dyn_arm_petals_r_parent_a}
      boxExtents=<7 10 2>
      constraintSetup {angularLimitsMax=<0 30 5>, angularLimitsMin=<0 -1 -5>, angularTarget=<0 0 0>}
      localJointOffset=<2 0 0>
     }]
    relativeSpaceBone {boneName=spine_05}
   }
   n131 = Anim Dynamics - Bone: dyn_arm_petals_r_parent_b(ComponentPose=@n130, Alpha=1, ExternalForce="0,0,0") {
    angularDampingOverride=0.9999
    bGravityOverrideInSimSpace=true
    bOverrideAngularDamping=true
    bOverrideLinearDamping=true
    bUseGravityOverride=true
    linearDampingOverride=0.9999
    simulationSpace=RootRelative
    alphaBoolBlend {blendOption=Linear}
    boundBone {boneName=dyn_arm_petals_r_parent_b}
    componentAppliedLinearAccClamp=<100 100 100>
    componentLinearAccScale=<0.25 0.25 0.5>
    componentLinearVelScale=<0.25 0.25 0.5>
    gravityOverride=<0.85 0.85 0.85>
    physicsBodyDefinitions [{
      sphereCollisionRadius=0
      boundBone {boneName=dyn_arm_petals_r_parent_b}
      boxExtents=<5 12 1.5>
      constraintSetup {angularLimitsMax=<0 20 5>, angularLimitsMin=<0 -1 -5>, angularTarget=<0 0 0>}
      localJointOffset=<1 0 0>
     }]
    relativeSpaceBone {boneName=spine_05}
   }
   n132 = Anim Dynamics - Bone: dyn_arm_petals_r_parent_c(ComponentPose=@n131, Alpha=1, ExternalForce="0,0,0") {
    angularDampingOverride=0.9999
    bGravityOverrideInSimSpace=true
    bOverrideAngularDamping=true
    bOverrideLinearDamping=true
    bUseGravityOverride=true
    linearDampingOverride=0.9999
    simulationSpace=RootRelative
    alphaBoolBlend {blendOption=Linear}
    boundBone {boneName=dyn_arm_petals_r_parent_c}
    componentAppliedLinearAccClamp=<100 100 100>
    componentLinearAccScale=<0.25 0.25 0.5>
    componentLinearVelScale=<0.25 0.25 0.5>
    gravityOverride=<0.85 0.85 0.85>
    physicsBodyDefinitions [{
      sphereCollisionRadius=0
      boundBone {boneName=dyn_arm_petals_r_parent_c}
      boxExtents=<6 8 1>
      constraintSetup {angularLimitsMax=<0 5 15>, angularLimitsMin=<0 -2 -8>, angularTarget=<0 0 0>}
      localJointOffset=<1.5 0 0>
     }]
    relativeSpaceBone {boneName=spine_05}
   }
   n133 = Anim Dynamics - Bone: dyn_arm_petals_r_parent_i(ComponentPose=@n132, Alpha=1, ExternalForce="0,0,0") {
    angularDampingOverride=0.9999
    bGravityOverrideInSimSpace=true
    bOverrideAngularDamping=true
    bOverrideLinearDamping=true
    bUseGravityOverride=true
    linearDampingOverride=0.9999
    simulationSpace=RootRelative
    alphaBoolBlend {blendOption=Linear}
    boundBone {boneName=dyn_arm_petals_r_parent_i}
    componentAppliedLinearAccClamp=<100 100 100>
    componentLinearAccScale=<0.25 0.25 0.5>
    componentLinearVelScale=<0.25 0.25 0.5>
    gravityOverride=<0.85 0.85 0.85>
    physicsBodyDefinitions [{
      sphereCollisionRadius=0
      boundBone {boneName=dyn_arm_petals_r_parent_i}
      boxExtents=<4 4 2>
      constraintSetup {angularLimitsMax=<0 20 3>, angularLimitsMin=<0 -10 -3>, angularTarget=<0 0 0>}
      localJointOffset=<1 0 0>
     }]
    relativeSpaceBone {boneName=spine_05}
   }
   n134 = Anim Dynamics - Bone: dyn_arm_petals_r_parent_h(ComponentPose=@n133, Alpha=1, ExternalForce="0,0,0") {
    angularDampingOverride=0.9999
    bGravityOverrideInSimSpace=true
    bOverrideAngularDamping=true
    bOverrideLinearDamping=true
    bUseGravityOverride=true
    linearDampingOverride=0.9999
    simulationSpace=RootRelative
    alphaBoolBlend {blendOption=Linear}
    alphaScaleBiasClamp {bias=0.1, scale=0.8}
    boundBone {boneName=dyn_arm_petals_r_parent_h}
    componentAppliedLinearAccClamp=<100 100 100>
    componentLinearAccScale=<0.15 0.15 0.15>
    componentLinearVelScale=<0.15 0.15 0.15>
    gravityOverride=<0.8 0.8 0.8>
    physicsBodyDefinitions [{
      sphereCollisionRadius=0
      boundBone {boneName=dyn_arm_petals_r_parent_h}
      boxExtents=<7 8 1>
      constraintSetup {angularLimitsMax=<0 30 15>, angularLimitsMin=<0 -30 -8>, angularTarget=<0 0 0>}
      localJointOffset=<1.5 0 0>
     }]
    relativeSpaceBone {boneName=root}
   }
   n135 = Anim Dynamics - Bone: dyn_arm_petals_r_parent_k(ComponentPose=@n134, Alpha=1, ExternalForce="0,0,0") {
    angularDampingOverride=0.9999
    bGravityOverrideInSimSpace=true
    bOverrideAngularDamping=true
    bOverrideLinearDamping=true
    bUseGravityOverride=true
    linearDampingOverride=0.9999
    simulationSpace=RootRelative
    alphaBoolBlend {blendOption=Linear}
    boundBone {boneName=dyn_arm_petals_r_parent_k}
    componentAppliedLinearAccClamp=<100 100 100>
    componentLinearAccScale=<0.25 0.25 0.5>
    componentLinearVelScale=<0.25 0.25 0.5>
    gravityOverride=<0.85 0.85 0.85>
    physicsBodyDefinitions [{
      sphereCollisionRadius=0
      boundBone {boneName=dyn_arm_petals_r_parent_k}
      boxExtents=<8 10 1.5>
      constraintSetup {angularLimitsMax=<0 10 10>, angularLimitsMin=<0 -10 -10>, angularTarget=<0 0 0>}
      localJointOffset=<1 0 0>
     }]
    relativeSpaceBone {boneName=spine_05}
   }
   n136 = Anim Dynamics - Bone: dyn_arm_petals_r_parent_j(ComponentPose=@n135, Alpha=1, ExternalForce="0,0,0") {
    angularDampingOverride=0.9999
    bGravityOverrideInSimSpace=true
    bOverrideAngularDamping=true
    bOverrideLinearDamping=true
    bUseGravityOverride=true
    linearDampingOverride=0.9999
    simulationSpace=RootRelative
    alphaBoolBlend {blendOption=Linear}
    alphaScaleBiasClamp {bias=0.1, scale=0.8}
    boundBone {boneName=dyn_arm_petals_r_parent_j}
    componentAppliedLinearAccClamp=<100 100 100>
    componentLinearAccScale=<0.15 0.15 0.15>
    componentLinearVelScale=<0.15 0.15 0.15>
    gravityOverride=<0.8 0.8 0.8>
    physicsBodyDefinitions [{
      sphereCollisionRadius=0
      boundBone {boneName=dyn_arm_petals_r_parent_j}
      boxExtents=<5 5 1.5>
      constraintSetup {angularLimitsMax=<0 30 10>, angularLimitsMin=<0 -10 -15>, angularTarget=<0 0 0>}
      localJointOffset=<1 0 0>
     }]
    relativeSpaceBone {boneName=root}
   }
   n137 = Anim Dynamics - Bone: dyn_arm_petals_r_parent_g(ComponentPose=@n136, Alpha=1, ExternalForce="0,0,0") {
    angularDampingOverride=0.9999
    bGravityOverrideInSimSpace=true
    bOverrideAngularDamping=true
    bOverrideLinearDamping=true
    bUseGravityOverride=true
    linearDampingOverride=0.9999
    simulationSpace=RootRelative
    alphaBoolBlend {blendOption=Linear}
    boundBone {boneName=dyn_arm_petals_r_parent_g}
    componentAppliedLinearAccClamp=<100 100 100>
    componentLinearAccScale=<0.25 0.25 0.5>
    componentLinearVelScale=<0.25 0.25 0.5>
    gravityOverride=<0.85 0.85 0.85>
    physicsBodyDefinitions [{
      sphereCollisionRadius=0
      boundBone {boneName=dyn_arm_petals_r_parent_g}
      boxExtents=<5 5 1.5>
      constraintSetup {angularLimitsMax=<0 15 15>, angularLimitsMin=<0 -15 -15>, angularTarget=<0 0 0>}
      localJointOffset=<1 0 0>
     }]
    relativeSpaceBone {boneName=spine_05}
   }
   n138 = Anim Dynamics - Bone: dyn_arm_petals_r_parent_f(ComponentPose=@n137, Alpha=1, ExternalForce="0,0,0") {
    angularDampingOverride=0.9999
    bGravityOverrideInSimSpace=true
    bOverrideAngularDamping=true
    bOverrideLinearDamping=true
    bUseGravityOverride=true
    linearDampingOverride=0.9999
    simulationSpace=RootRelative
    alphaBoolBlend {blendOption=Linear}
    boundBone {boneName=dyn_arm_petals_r_parent_f}
    componentAppliedLinearAccClamp=<100 100 100>
    componentLinearAccScale=<0.25 0.25 0.5>
    componentLinearVelScale=<0.25 0.25 0.5>
    gravityOverride=<0.85 0.85 0.85>
    physicsBodyDefinitions [{
      sphereCollisionRadius=0
      boundBone {boneName=dyn_arm_petals_r_parent_f}
      boxExtents=<5 5 1.5>
      constraintSetup {angularLimitsMax=<0 10 10>, angularLimitsMin=<0 -10 -10>, angularTarget=<0 0 0>}
      localJointOffset=<1 0 0>
     }]
    relativeSpaceBone {boneName=spine_05}
   }
   n139 = Anim Dynamics - Bone: dyn_arm_petals_r_parent_e(ComponentPose=@n138, Alpha=1, ExternalForce="0,0,0") {
    angularDampingOverride=0.9999
    bGravityOverrideInSimSpace=true
    bOverrideAngularDamping=true
    bOverrideLinearDamping=true
    bUseGravityOverride=true
    linearDampingOverride=0.9999
    simulationSpace=RootRelative
    alphaBoolBlend {blendOption=Linear}
    boundBone {boneName=dyn_arm_petals_r_parent_e}
    componentAppliedLinearAccClamp=<100 100 100>
    componentLinearAccScale=<0.25 0.25 0.5>
    componentLinearVelScale=<0.25 0.25 0.5>
    gravityOverride=<0.85 0.85 0.85>
    physicsBodyDefinitions [{
      sphereCollisionRadius=0
      boundBone {boneName=dyn_arm_petals_r_parent_e}
      boxExtents=<4 6 2>
      constraintSetup {angularLimitsMax=<0 15 3>, angularLimitsMin=<0 -5 -3>, angularTarget=<0 0 0>}
      localJointOffset=<1 0 0>
     }]
    relativeSpaceBone {boneName=spine_05}
   }
   n140 = Anim Dynamics - Bone: dyn_arm_petals_r_parent_d(ComponentPose=@n139, Alpha=1, ExternalForce="0,0,0") {
    angularDampingOverride=0.9999
    bGravityOverrideInSimSpace=true
    bOverrideAngularDamping=true
    bOverrideLinearDamping=true
    bUseGravityOverride=true
    linearDampingOverride=0.9999
    simulationSpace=RootRelative
    alphaBoolBlend {blendOption=Linear}
    boundBone {boneName=dyn_arm_petals_r_parent_d}
    componentAppliedLinearAccClamp=<100 100 100>
    componentLinearAccScale=<0.25 0.25 0.5>
    componentLinearVelScale=<0.25 0.25 0.5>
    gravityOverride=<0.85 0.85 0.85>
    physicsBodyDefinitions [{
      sphereCollisionRadius=0
      boundBone {boneName=dyn_arm_petals_r_parent_d}
      boxExtents=<6 7 1.5>
      constraintSetup {angularLimitsMax=<0 20 10>, angularLimitsMin=<0 -10 -10>, angularTarget=<0 0 0>}
      localJointOffset=<1.5 0 0>
     }]
    relativeSpaceBone {boneName=spine_05}
   }
   n141 = "Constraint - dyn_arm_petals_r_base_h (PPPPP)"!(
    ComponentPose=@n140
    Alpha=1
    ConstraintWeights_0=0.6777
    ConstraintWeights_1:float=0.1992
    ConstraintWeights_2:float=0.0646
    ConstraintWeights_3:float=0.0058
    ConstraintWeights_4:float=0.0527
   ) -> (Pose) {
    alphaBoolBlend {blendOption=Linear}
    boneToModify {boneName=dyn_arm_petals_r_base_h}
    constraintSetup [
     {targetBone {boneName=clavicle_r}}
     {targetBone {boneName=upperarm_twist_01_r}}
     {targetBone {boneName=spine_03}}
     {targetBone {boneName=spine_04}}
     {targetBone {boneName=spine_05}}
    ]
    constraintWeights [1,1,1,1,1]
   }
   n142 = "Constraint - dyn_arm_petals_r_base_a (PPPPPPP)"!(
    ComponentPose=@n141
    Alpha=1
    ConstraintWeights_0=0.4196
    ConstraintWeights_1:float=0.0697
    ConstraintWeights_2:float=0.2191
    ConstraintWeights_3:float=0.0115
    ConstraintWeights_4:float=0.0991
    ConstraintWeights_5:float=0.1529
    ConstraintWeights_6:float=0.0282
   ) -> (Pose) {
    alphaBoolBlend {blendOption=Linear}
    boneToModify {boneName=dyn_arm_petals_r_base_a}
    constraintSetup [
     {targetBone {boneName=head}}
     {targetBone {boneName=upperarm_twist_01_r}}
     {targetBone {boneName=dyn_arm_petals_r_parent_a}}
     {targetBone {boneName=spine_03}}
     {targetBone {boneName=spine_05}}
     {targetBone {boneName=clavicle_r}}
     {targetBone {boneName=neck_02}}
    ]
    constraintWeights [1,1,1,1,1,1,1]
   }
   n143 = "Constraint - dyn_arm_petals_r_base_i (PPPP)"!(
    ComponentPose=@n142
    Alpha=1
    ConstraintWeights_0=0.4266
    ConstraintWeights_1:float=0.51
    ConstraintWeights_2:float=0.0432
    ConstraintWeights_3:float=0.0202
   ) -> (Pose) {
    alphaBoolBlend {blendOption=Linear}
    boneToModify {boneName=dyn_arm_petals_r_base_i}
    constraintSetup [
     {targetBone {boneName=clavicle_r}}
     {targetBone {boneName=upperarm_twist_01_r}}
     {targetBone {boneName=spine_03}}
     {targetBone {boneName=spine_05}}
    ]
    constraintWeights [1,1,1,1]
   }
   n144 = "Constraint - dyn_arm_petals_r_base_k (PPPPPPPP)"!(
    ComponentPose=@n143
    Alpha=1
    ConstraintWeights_0=0.0335
    ConstraintWeights_1:float=0.4022
    ConstraintWeights_2:float=0.0679
    ConstraintWeights_3:float=0.0316
    ConstraintWeights_4:float=0.0049
    ConstraintWeights_5:float=0.1379
    ConstraintWeights_6:float=0.2948
    ConstraintWeights_7:float=0.0272
   ) -> (Pose) {
    alphaBoolBlend {blendOption=Linear}
    boneToModify {boneName=dyn_arm_petals_r_base_k}
    constraintSetup [
     {targetBone {boneName=head}}
     {targetBone {boneName=dyn_arm_petals_r_parent_k}}
     {targetBone {boneName=upperarm_twist_01_r}}
     {targetBone {boneName=spine_03}}
     {targetBone {boneName=spine_04}}
     {targetBone {boneName=spine_05}}
     {targetBone {boneName=clavicle_r}}
     {targetBone {boneName=neck_02}}
    ]
    constraintWeights [1,1,1,1,1,1,1,1]
   }
   n145 = "Constraint - dyn_arm_petals_r_base_j (PPPP)"!(
    ComponentPose=@n144
    Alpha=1
    ConstraintWeights_0=0.4058
    ConstraintWeights_1:float=0.5331
    ConstraintWeights_2:float=0.0455
    ConstraintWeights_3:float=0.0156
   ) -> (Pose) {
    alphaBoolBlend {blendOption=Linear}
    boneToModify {boneName=dyn_arm_petals_r_base_j}
    constraintSetup [
     {targetBone {boneName=clavicle_r}}
     {targetBone {boneName=upperarm_twist_01_r}}
     {targetBone {boneName=spine_03}}
     {targetBone {boneName=spine_05}}
    ]
    constraintWeights [1,1,1,1]
   }
   n146 = "Transform (Modify) Bone - Bone: dyn_leg_petals_l_base_c"(ComponentPose=@n145, Alpha=Get legSwingDelta_l() -> (legSwingDelta_l:double), Translation="0,0,0", Rotation="-30.000,0,60", Scale="1,1,1") {
    rotationMode=BMM_Additive
    rotationSpace=BCS_BoneSpace
    translationMode=BMM_Additive
    translationSpace=BCS_BoneSpace
    alphaBoolBlend {blendOption=Linear}
    alphaScaleBiasClamp {bInterpResult=true, interpSpeedDecreasing=4, interpSpeedIncreasing=6}
    boneToModify {boneName=dyn_leg_petals_l_base_c}
    rotation {pitch=-30, roll=60}
   }
   n147 = Get Shoulder_L_Alpha() -> (Shoulder_L_Alpha:double)
   n148 = "Transform (Modify) Bone - Bone: dyn_shoulder_flower_outter_l_master"(ComponentPose=@n146, Alpha=@n147, Translation="0,0,0", Rotation="0.000,23,0.000", Scale="1,1,1") {
    rotationMode=BMM_Additive
    scaleMode=BMM_Replace
    scaleSpace=BCS_BoneSpace
    translationMode=BMM_Additive
    translationSpace=BCS_BoneSpace
    alphaBoolBlend {blendOption=Linear}
    alphaScaleBiasClamp {bInterpResult=true}
    boneToModify {boneName=dyn_shoulder_flower_outter_l_master}
    rotation {yaw=23}
   }
   n149 = "Transform (Modify) Bone - Bone: dyn_shoulder_flower_outter_l_base_e"(ComponentPose=@n148, Alpha=@n147, Translation="0,0,0", Rotation="45.000,25,0.000", Scale="0.5,0.5,0.5") {
    rotationMode=BMM_Additive
    scaleMode=BMM_Replace
    scaleSpace=BCS_BoneSpace
    translationMode=BMM_Additive
    translationSpace=BCS_BoneSpace
    alphaBoolBlend {blendOption=Linear}
    boneToModify {boneName=dyn_shoulder_flower_outter_l_base_e}
    rotation {pitch=45, yaw=25}
    scale=<0.5 0.5 0.5>
   }
   n150 = "Transform (Modify) Bone - Bone: dyn_shoulder_flower_l_base"(ComponentPose=@n149, Alpha=@n147, Translation="1,0,0", Rotation="0,0,0", Scale="1,1,1") {
    rotationMode=BMM_Additive
    rotationSpace=BCS_BoneSpace
    scaleMode=BMM_Replace
    scaleSpace=BCS_BoneSpace
    translationMode=BMM_Additive
    translationSpace=BCS_BoneSpace
    alphaBoolBlend {blendOption=Linear}
    boneToModify {boneName=dyn_shoulder_flower_l_base}
    translation {x=1}
   }
   n151 = "Transform (Modify) Bone - Bone: dyn_shoulder_flower_outter_l_base_c"(ComponentPose=@n150, Alpha=@n147, Translation="0,-2,0", Rotation="0,-20,0", Scale="1,1,1") {
    rotationMode=BMM_Additive
    rotationSpace=BCS_BoneSpace
    scaleMode=BMM_Replace
    scaleSpace=BCS_BoneSpace
    translationMode=BMM_Additive
    translationSpace=BCS_BoneSpace
    alphaBoolBlend {blendOption=Linear}
    boneToModify {boneName=dyn_shoulder_flower_outter_l_base_c}
    rotation {yaw=-20}
    translation {y=-2}
   }
   n152 = "Transform (Modify) Bone - Bone: dyn_shoulder_flower_outter_l_base_d"(ComponentPose=@n151, Alpha=@n147, Translation="0,0,0", Rotation="20,-60,0", Scale="1,1,1") {
    rotationMode=BMM_Additive
    rotationSpace=BCS_BoneSpace
    scaleMode=BMM_Replace
    scaleSpace=BCS_BoneSpace
    translationMode=BMM_Additive
    translationSpace=BCS_BoneSpace
    alphaBoolBlend {blendOption=Linear}
    boneToModify {boneName=dyn_shoulder_flower_outter_l_base_d}
    rotation {pitch=20, yaw=-60}
   }
   n153 = Get Alien_Flora_Pickaxe() -> (Alien_Flora_Pickaxe:double)
   n154 = "Transform (Modify) Bone - Bone: hand_r"(ComponentPose=@n152, Alpha=@n153, Translation="0,0,0", Rotation="0,0,0", Scale="0.1,0.1,0.1") {scaleMode=BMM_Replace, scaleSpace=BCS_BoneSpace, alphaBoolBlend {blendOption=Linear}, boneToModify {boneName=hand_r}, scale=<0.1 0.1 0.1>}
   n155 = "Transform (Modify) Bone - Bone: lowerarm_twist_02_r"(ComponentPose=@n154, Alpha=@n153, Translation="0,0,0", Rotation="0,0,0", Scale="0.75,0.75,0.75") {
    scaleMode=BMM_Replace
    scaleSpace=BCS_BoneSpace
    alphaBoolBlend {blendOption=Linear}
    boneToModify {boneName=lowerarm_twist_02_r}
    scale=<0.75 0.75 0.75>
   }
   n156 = "Transform (Modify) Bone - Bone: lowerarm_twist_01_r"(ComponentPose=@n155, Alpha=@n153, Translation="0,0,0", Rotation="0,0,0", Scale="0.75,0.75,0.75") {
    scaleMode=BMM_Replace
    scaleSpace=BCS_BoneSpace
    alphaBoolBlend {blendOption=Linear}
    alphaScaleBiasClamp {bInterpResult=true}
    boneToModify {boneName=lowerarm_twist_01_r}
    scale=<0.75 0.75 0.75>
   }
   Output Pose(Result=Component To Local(ComponentPose=@n156)) {layerGroup=None}
  ]}
 EventGraph {nodes [
   n1 = Cast To FortPlayerPawn(Object=TryGetFortPawn())
   > n2 = WatchCosmeticStat!(
    execute:Exec
    Pawn:Fort Player Pawn Object Reference=@n1
    StatTag:Fort Stat Manager Tag Structure=(TagName="StatManager.AthenaKills")
    DefaultValue:int=0
    bBroadcastInitialValue:bool=true
    InitialValueBroadcastDelay:float=-1
    ChangeValueBroadcastDelay:float=-1
    bConsumeWaitingNotifications:bool=true
   ) -> (then:Exec, OnStatChanged:Exec, StatName:Name, StatValue:int) {is=title K2Node_AsyncAction}
   Set Did_Kill(execute=@n2.OnStatChanged, Did_Kill:bool=false) -> (Output_Get:bool)
   > n3 = Set KillCount(KillCount:double="To Float (Integer)"(InInt=@n2.StatValue)) -> (Output_Get:double)
   > Branch(Condition="float > float"(A=@n3, B=Get Last_KillCount() -> (Last_KillCount:double)) {is=title K2Node_CallFunction})
   > Set Active_Kill(Active_Kill:bool=true) -> (Output_Get:bool)
   > Set Last_KillCount(Last_KillCount:double=@n3) -> (Output_Get:double)
   PrintString!(execute=@n1.CastFailed, InString=Hello, bPrintToScreen, bPrintToLog, TextColor, Duration, Key) -> (then)
   Event BlueprintProcessAnimInputEvent()
   > n4 = Sequence()
   > Do Once()
   > n5 = Cast To FortPlayerPawn(Object=TryGetFortPawn())
   > n6 = WatchCosmeticStat(
    Pawn=@n5
    StatTag=(TagName="StatManager.DamageReceived")
    DefaultValue=0
    bBroadcastInitialValue=true
    InitialValueBroadcastDelay=-1
    ChangeValueBroadcastDelay=-1
    bConsumeWaitingNotifications=true
   )
   n7 = "integer > integer"(A=@n6.StatValue, B=0) {is=title K2Node_CallFunction}
   Branch(execute=@n6.OnStatChanged, Condition=@n7)
   > PrintString(InString=Hello, TextColor=(R=0.2358,G=1,B=0.1941,A=1))
   > n8 = Set Took_Damage(Took_Damage:bool=@n7) -> (Output_Get:bool)
   > Set Took_Damage_Float(Took_Damage_Float:double="To Float (Boolean)"(InBool=@n8)) -> (Output_Get:double)
   n9 = Get PetelsTargetColor() -> (PetelsTargetColor:double)
   > Set PetelsTargetColor(PetelsTargetColor:double="Clamp (Float)"(Value=float + float(A=@n9, B=0.334) {is=title K2Node_CommutativeAssociativeBinaryOperator}, Min=0, Max=1)) -> (Output_Get:double)
   > RetriggerableDelay(Duration=6)
   > Set Took_Damage_Float(Took_Damage_Float=0)
   > Set PetelsTargetColor(PetelsTargetColor=0)
   n10 = Get Took_Damage_Float()
   Set Dynamics_Alpha(
    execute=@n4.then_1
    Dynamics_Alpha:double=SelectFloat(A=1, B=0, bPickA="float <= float"!(A:double=@n10, B:double=0) -> (ReturnValue:bool) {is=title K2Node_CallFunction})
   ) -> (Output_Get:double)
   > Delay(Duration=6)
   > Set Dynamics_Alpha(Dynamics_Alpha=1)
   Event BlueprintInitializeAnimation()
   > Set Skel_Mesh(Skel_Mesh=GetPartSkeletalMeshComponent())
   n11 = Event BlueprintUpdateAnimation()
   > n12 = Sequence!(execute) -> (then_0, then_1, then_2:Exec, then_3:Exec)
   > n13 = Set PetelsCurrentColor(PetelsCurrentColor:double=FInterpTo(Current=Get PetelsCurrentColor() -> (PetelsCurrentColor:double), Target=Get PetelsTargetColor(), DeltaTime=@n11.DeltaTimeX, InterpSpeed=2)) -> (Output_Get:double)
   > SetScalarParameterValueOnMaterials!(execute, self:Mesh Component Object Reference=GetPartSkeletalMeshComponent(), ParameterName:Name=MaterialCurveA, ParameterValue:float=@n13) -> (then)
   > Set FortPawn(FortPawn:Fort Pawn Object Reference=TryGetFortPawn()) -> (Output_Get:Fort Pawn Object Reference)
   n14 = Get FortPawn() -> (FortPawn:Fort Pawn Object Reference)
   > n15 = Is Valid(InputObject=@n14)
   n16 = Set weapon_visible(execute=@n15.Is Not Valid, weapon_visible:bool=false) -> (Output_Get:bool)
   n17 = Get weapon_visible() -> (weapon_visible:bool)
   n18 = Branch(execute=[@n16.then, @n22.then], Condition=@n17)
   n19 = PickaxeItemDefinitionHasAnyMetaTag(Tags=Make Array("[0]"=(TagName="Cosmetics.MetaTags.AlienFlora")), bRequiresEquipped=true)
   > Set Alien_Flora_Pickaxe(Alien_Flora_Pickaxe:double="To Float (Boolean)"(InBool=AND Boolean(A=@n17, B=@n19))) -> (Output_Get:double)
   Set Alien_Flora_Pickaxe(execute=@n18.else, Alien_Flora_Pickaxe=0)
   n20 = IsWeaponHolstered!(self:Fort Pawn Object Reference=@n14) -> (ReturnValue:bool)
   n21 = Get CurrentWeapon!(self:Fort Pawn Object Reference=@n14) -> (CurrentWeapon:Fort Weapon Object Reference)
   n22 = Set weapon_visible(execute=@n15.Is Valid, weapon_visible=AND Boolean(A=IsValid(Object=@n21), B=NOT Boolean(A=@n20)))
   Set Is_ADS(execute=@n12.then_3, Is_ADS:double=SelectFloat(A=1, B=0, bPickA=Get bIsTargeting())) -> (Output_Get:double)
   n23 = GetOwnerSkeletalMeshComponent()
   Set legSwingDelta_l(
    execute=@n12.then_2
    legSwingDelta_l:double=MapRangeClamped(Value=GetTwistAndSwingAngleOfDeltaRotationFromRefPose(self=@n23, BoneName=thigh_l).OutSwingAngle, InRangeA=16, InRangeB=105)
   ) -> (Output_Get:double)
   Set Shoulder_L_Alpha(
    execute=@n12.then_1
    Shoulder_L_Alpha:double=GetAlphaFromReferenceTransformDelta(
     SkeletalMeshComponent=Get Skel_Mesh()
     BoneName=clavicle_l
     BaseName=upperarm_l
     DeltaComponentType=QuaternionTwist
     InputRange=(X=6,Y=69)
     TwistAxis="0,0,1"
    )
   ) -> (Output_Get:double)
   Branch(Condition=Get Active_Kill() -> (Active_Kill:bool))
   n24 = float / float(A=Get KillCount() -> (KillCount:double), B=Get MaxKillCount() -> (MaxKillCount:double)) {is=title K2Node_CallFunction}
   > Set Kill_Alpha(Kill_Alpha:double=float - float(B="Clamp (Float)"(Value=@n24, Min=0, Max=1)) {is=title K2Node_CallFunction}) -> (Output_Get:double)
   > Set Active_Kill(Active_Kill=false)
   WeaponDataHasMatchingGameplayTag!(
    self:Fort Weapon Object Reference=Get CurrentWeapon(self=Get FortPawn())
    TagToCheck:Gameplay Tag Structure=(TagName="Cosmetics.MetaTags.AlienFlora")
   ) -> (ReturnValue:bool)
   Event OnCharacterCustomizationLoaded()
   Set Skel_Mesh(Skel_Mesh=GetOwnerSkeletalMeshComponent())
  ]}
]
variables [
 Active_Kill {type=boolean}
 Alien_Flora_Pickaxe {type=number}
 Did_Kill {type=boolean}
 Dynamics_Alpha {type=number}
 FortPawn {type=FortPawn, default=null}
 Is_ADS {type=number}
 Jump {type=number}
 Kill_Alpha {type=number, default=1}
 KillCount {type=number}
 Last_KillCount {type=number}
 legSwingDelta_l {type=number}
 MaxKillCount {type=number, default=5}
 PetelsCurrentColor {type=number}
 PetelsTargetColor {type=number}
 Shoulder_L_Alpha {type=number}
 Skel_Mesh {type=SkeletalMeshComponent}
 Took_Damage {type=boolean}
 Took_Damage_Float {type=number}
 weapon_visible {type=boolean}
]
