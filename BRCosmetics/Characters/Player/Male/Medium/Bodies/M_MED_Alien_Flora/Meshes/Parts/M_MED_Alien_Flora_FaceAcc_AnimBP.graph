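# Anim Blueprint for the Alien Flora skin's face accessory, adding physics sway and a damage reaction to its flowers and petals.
# It copies the parent mesh pose, then runs chained AnimDynamics on head flowers, bud, scarf and petals, with flower sway gated by Dynamics_Alpha.
# A DamageReceived stat watcher blends in additive head-flower rotations and steps up an interpolated MaterialCurveA parameter.
# Both reset after a retriggerable six-second delay.
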AnimBlueprint /BRCosmetics/Characters/Player/Male/Medium/Bodies/M_MED_Alien_Flora/Meshes/Parts/M_MED_Alien_Flora_FaceAcc_AnimBP
target_skeleton=&/M_MED_Alien_Flora_FaceAcc_Skeleton
graphs [
 AnimGraph {nodes [
   n1 = Layered blend per bone!(
    BasePose=Use cached pose 'HeadLockCache'()
    BlendPoses_0
    BlendPoses_1:Pose Link Structure=Use cached pose 'HeadLockCache'()
    BlendWeights_0=0.7
    BlendWeights_1:float=1
   ) -> (Pose) {
    blendPoses [{linkId=-1, sourceLinkId=-1}, {linkId=-1, sourceLinkId=-1}]
    blendWeights [1,1]
    layerSetup [{branchFilters [{blendDepth=1, boneName=neck_01}]}, {branchFilters [{blendDepth=1, boneName=faceAttach}]}]
   }
   n2 = Get Dynamics_Alpha() -> (Dynamics_Alpha:double)
   n3 = Anim Dynamics - Bone: dyn_head_flower_01(ComponentPose=Local To Component(LocalPose=@n1), Alpha=@n2, ExternalForce="0,0,0") {
    angularDampingOverride=0.999
    bGravityOverrideInSimSpace=true
    bOverrideAngularDamping=true
    bOverrideLinearDamping=true
    bUseGravityOverride=true
    linearDampingOverride=0.999
    alphaBoolBlend {blendOption=Linear}
    alphaScaleBiasClamp {bInterpResult=true, interpSpeedDecreasing=0, interpSpeedIncreasing=2}
    boundBone {boneName=dyn_head_flower_01}
    componentAppliedLinearAccClamp=<3000 3000 3000>
    componentLinearAccScale=<0.25 0.25 0.5>
    componentLinearVelScale=<0.25 0.25 0.5>
    gravityOverride=<0.85 0.85 0.85>
    physicsBodyDefinitions [{
      sphereCollisionRadius=0
      boundBone {boneName=dyn_head_flower_01}
      boxExtents=<6 4 2>
      constraintSetup {angularLimitsMax=<0 30 8>, angularLimitsMin=<0 -20 -8>, angularTarget=<0 0 0>}
      localJointOffset=<2 0 0>
     }]
    relativeSpaceBone {boneName=pelvis}
   }
   n4 = Anim Dynamics - Bone: dyn_head_flower_02(ComponentPose=@n3, Alpha=@n2, ExternalForce="0,0,0") {
    angularDampingOverride=0.999
    bGravityOverrideInSimSpace=true
    bOverrideAngularDamping=true
    bOverrideLinearDamping=true
    bUseGravityOverride=true
    linearDampingOverride=0.999
    alphaBoolBlend {blendOption=Linear}
    alphaScaleBiasClamp {bInterpResult=true, interpSpeedDecreasing=0, interpSpeedIncreasing=2}
    boundBone {boneName=dyn_head_flower_02}
    componentAppliedLinearAccClamp=<3000 3000 3000>
    componentLinearAccScale=<0.25 0.25 0.5>
    componentLinearVelScale=<0.25 0.25 0.5>
    gravityOverride=<0.85 0.85 0.85>
    physicsBodyDefinitions [{
      sphereCollisionRadius=0
      boundBone {boneName=dyn_head_flower_02}
      boxExtents=<5 4 1.5>
      constraintSetup {angularLimitsMax=<0 30 8>, angularLimitsMin=<0 -20 -8>, angularTarget=<0 0 0>}
      localJointOffset=<1 0 0>
     }]
    relativeSpaceBone {boneName=pelvis}
   }
   n5 = Anim Dynamics - Bone: dyn_head_flower_03(ComponentPose=@n4, Alpha=@n2, ExternalForce="0,0,0") {
    angularDampingOverride=0.999
    bGravityOverrideInSimSpace=true
    bOverrideAngularDamping=true
    bOverrideLinearDamping=true
    bUseGravityOverride=true
    linearDampingOverride=0.999
    alphaBoolBlend {blendOption=Linear}
    alphaScaleBiasClamp {bInterpResult=true, interpSpeedDecreasing=0, interpSpeedIncreasing=2}
    boundBone {boneName=dyn_head_flower_03}
    componentAppliedLinearAccClamp=<3000 3000 3000>
    componentLinearAccScale=<0.25 0.25 0.5>
    componentLinearVelScale=<0.25 0.25 0.5>
    gravityOverride=<0.85 0.85 0.85>
    physicsBodyDefinitions [{
      sphereCollisionRadius=0
      boundBone {boneName=dyn_head_flower_03}
      boxExtents=<5 4 1>
      constraintSetup {angularLimitsMax=<0 15 8>, angularLimitsMin=<0 -15 -8>, angularTarget=<0 0 0>}
      localJointOffset=<1.5 0 0>
     }]
    relativeSpaceBone {boneName=pelvis}
   }
   n6 = Anim Dynamics - Bone: dyn_head_flower_04(ComponentPose=@n5, Alpha=@n2, ExternalForce="0,0,0") {
    angularDampingOverride=0.999
    bGravityOverrideInSimSpace=true
    bOverrideAngularDamping=true
    bOverrideLinearDamping=true
    bUseGravityOverride=true
    linearDampingOverride=0.999
    alphaBoolBlend {blendOption=Linear}
    alphaScaleBiasClamp {bInterpResult=true, interpSpeedDecreasing=0, interpSpeedIncreasing=2}
    boundBone {boneName=dyn_head_flower_04}
    componentAppliedLinearAccClamp=<3000 3000 3000>
    componentLinearAccScale=<0.25 0.25 0.5>
    componentLinearVelScale=<0.25 0.25 0.5>
    gravityOverride=<0.85 0.85 0.85>
    physicsBodyDefinitions [{
      sphereCollisionRadius=0
      boundBone {boneName=dyn_head_flower_04}
      boxExtents=<5 2 4>
      constraintSetup {angularLimitsMax=<0 5 20>, angularLimitsMin=<0 -5 -20>, angularTarget=<0 0 0>}
      localJointOffset=<1.5 0 0>
     }]
    relativeSpaceBone {boneName=pelvis}
   }
   n7 = Anim Dynamics - Bone: dyn_bud_parent(ComponentPose=@n6, Alpha=1, ExternalForce="0,0,0") {
    angularDampingOverride=0.9999
    bGravityOverrideInSimSpace=true
    bOverrideAngularDamping=true
    bOverrideLinearDamping=true
    bUseGravityOverride=true
    linearDampingOverride=0.9999
    simulationSpace=RootRelative
    alphaBoolBlend {blendOption=Linear}
    boundBone {boneName=dyn_bud_parent}
    componentAppliedLinearAccClamp=<100 100 100>
    componentLinearAccScale=<0.15 0.15 0.15>
    componentLinearVelScale=<0.15 0.15 0.15>
    gravityOverride=<1 0.85 0.85>
    physicsBodyDefinitions [{
      sphereCollisionRadius=0
      boundBone {boneName=dyn_bud_parent}
      boxExtents=<8 10 3>
      constraintSetup {angularLimitsMax=<0 15 15>, angularLimitsMin=<0 -15 -15>, angularTarget=<0 0 0>}
      localJointOffset=<2 0 0>
     }]
    relativeSpaceBone {boneName=spine_05}
   }
   n8 = Anim Dynamics - Bone: dyn_petal_scarf_parent(ComponentPose=@n7, Alpha=1, ExternalForce="0,0,0") {
    angularDampingOverride=0.999
    bGravityOverrideInSimSpace=true
    bOverrideAngularDamping=true
    bOverrideLinearDamping=true
    bUseGravityOverride=true
    linearDampingOverride=0.999
    simulationSpace=RootRelative
    alphaBoolBlend {blendOption=Linear}
    boundBone {boneName=dyn_petal_scarf_parent}
    componentAppliedLinearAccClamp=<3000 3000 3000>
    componentLinearAccScale=<0.25 0.25 0.5>
    componentLinearVelScale=<0.25 0.25 0.5>
    gravityOverride=<1 0.85 0.85>
    physicsBodyDefinitions [{
      sphereCollisionRadius=0
      boundBone {boneName=dyn_petal_scarf_parent}
      boxExtents=<4 4 2>
      constraintSetup {angularLimitsMax=<0 10 10>, angularLimitsMin=<0 -10 -10>, angularTarget=<0 0 0>}
      localJointOffset=<1 0 0>
     }]
    relativeSpaceBone {boneName=spine_05}
   }
   n9 = Anim Dynamics - Bone: dyn_petals_back_parent_01(ComponentPose=@n8, Alpha=1, ExternalForce="0,0,0") {
    angularDampingOverride=0.9999
    bGravityOverrideInSimSpace=true
    bOverrideAngularDamping=true
    bOverrideLinearDamping=true
    bUseGravityOverride=true
    linearDampingOverride=0.9999
    simulationSpace=RootRelative
    alphaBoolBlend {blendOption=Linear}
    boundBone {boneName=dyn_petals_back_parent_01}
    componentAppliedLinearAccClamp=<100 100 100>
    componentLinearAccScale=<0.15 0.15 0.15>
    componentLinearVelScale=<0.15 0.15 0.15>
    gravityOverride=<0.75 0.75 0.75>
    physicsBodyDefinitions [{
      sphereCollisionRadius=0
      boundBone {boneName=dyn_petals_back_parent_01}
      boxExtents=<4 4 2>
      constraintSetup {angularLimitsMax=<0 15 15>, angularLimitsMin=<0 -10 -10>, angularTarget=<0 0 0>}
      localJointOffset=<1 0 0>
     }]
    relativeSpaceBone {boneName=spine_05}
   }
   n10 = Anim Dynamics - Bone: dyn_petals_back_parent_02(ComponentPose=@n9, Alpha=1, ExternalForce="0,0,0") {
    angularDampingOverride=0.9999
    bGravityOverrideInSimSpace=true
    bOverrideAngularDamping=true
    bOverrideLinearDamping=true
    bUseGravityOverride=true
    linearDampingOverride=0.9999
    simulationSpace=RootRelative
    alphaBoolBlend {blendOption=Linear}
    boundBone {boneName=dyn_petals_back_parent_02}
    componentAppliedLinearAccClamp=<100 100 100>
    componentLinearAccScale=<0.15 0.15 0.15>
    componentLinearVelScale=<0.15 0.15 0.15>
    gravityOverride=<0.75 0.75 0.75>
    physicsBodyDefinitions [{
      sphereCollisionRadius=0
      boundBone {boneName=dyn_petals_back_parent_02}
      boxExtents=<4 4 2>
      constraintSetup {angularLimitsMax=<0 15 15>, angularLimitsMin=<0 -10 -10>, angularTarget=<0 0 0>}
      localJointOffset=<1 0 0>
     }]
    relativeSpaceBone {boneName=spine_05}
   }
   n11 = Anim Dynamics - Bone: dyn_petals_back_parent_03(ComponentPose=@n10, Alpha=1, ExternalForce="0,0,0") {
    angularDampingOverride=0.9999
    bGravityOverrideInSimSpace=true
    bOverrideAngularDamping=true
    bOverrideLinearDamping=true
    bUseGravityOverride=true
    linearDampingOverride=0.9999
    simulationSpace=RootRelative
    alphaBoolBlend {blendOption=Linear}
    boundBone {boneName=dyn_petals_back_parent_03}
    componentAppliedLinearAccClamp=<100 100 100>
    componentLinearAccScale=<0.15 0.15 0.15>
    componentLinearVelScale=<0.15 0.15 0.15>
    gravityOverride=<0.75 0.75 0.75>
    physicsBodyDefinitions [{
      sphereCollisionRadius=0
      boundBone {boneName=dyn_petals_back_parent_03}
      boxExtents=<4 4 2>
      constraintSetup {angularLimitsMax=<0 15 15>, angularLimitsMin=<0 -10 -10>, angularTarget=<0 0 0>}
      localJointOffset=<1 0 0>
     }]
    relativeSpaceBone {boneName=spine_05}
   }
   n12 = Anim Dynamics - Bone: dyn_petals_back_parent_06(ComponentPose=@n11, Alpha=1, ExternalForce="0,0,0") {
    angularDampingOverride=0.9999
    bGravityOverrideInSimSpace=true
    bOverrideAngularDamping=true
    bOverrideLinearDamping=true
    bUseGravityOverride=true
    linearDampingOverride=0.9999
    simulationSpace=RootRelative
    alphaBoolBlend {blendOption=Linear}
    boundBone {boneName=dyn_petals_back_parent_06}
    componentAppliedLinearAccClamp=<100 100 100>
    componentLinearAccScale=<0.15 0.15 0.15>
    componentLinearVelScale=<0.15 0.15 0.15>
    gravityOverride=<0.75 0.75 0.75>
    physicsBodyDefinitions [{
      sphereCollisionRadius=0
      boundBone {boneName=dyn_petals_back_parent_06}
      boxExtents=<6 8 2>
      constraintSetup {angularLimitsMax=<0 30 20>, angularLimitsMin=<0 -20 -20>, angularTarget=<0 0 0>}
      localJointOffset=<2 0 0>
     }]
    relativeSpaceBone {boneName=root}
   }
   n13 = Anim Dynamics - Bone: dyn_petals_back_parent_05(ComponentPose=@n12, Alpha=1, ExternalForce="0,0,0") {
    angularDampingOverride=0.9999
    bGravityOverrideInSimSpace=true
    bOverrideAngularDamping=true
    bOverrideLinearDamping=true
    bUseGravityOverride=true
    linearDampingOverride=0.9999
    simulationSpace=RootRelative
    alphaBoolBlend {blendOption=Linear}
    boundBone {boneName=dyn_petals_back_parent_05}
    componentAppliedLinearAccClamp=<100 100 100>
    componentLinearAccScale=<0.15 0.15 0.15>
    componentLinearVelScale=<0.15 0.15 0.15>
    gravityOverride=<0.75 0.75 0.75>
    physicsBodyDefinitions [{
      sphereCollisionRadius=0
      boundBone {boneName=dyn_petals_back_parent_05}
      boxExtents=<4 4 2>
      constraintSetup {angularLimitsMax=<0 15 15>, angularLimitsMin=<0 -10 -10>, angularTarget=<0 0 0>}
      localJointOffset=<1 0 0>
     }]
    relativeSpaceBone {boneName=spine_05}
   }
   n14 = Anim Dynamics - Bone: dyn_petals_back_parent_04(ComponentPose=@n13, Alpha=1, ExternalForce="0,0,0") {
    angularDampingOverride=0.9999
    bGravityOverrideInSimSpace=true
    bOverrideAngularDamping=true
    bOverrideLinearDamping=true
    bUseGravityOverride=true
    linearDampingOverride=0.9999
    simulationSpace=RootRelative
    alphaBoolBlend {blendOption=Linear}
    boundBone {boneName=dyn_petals_back_parent_04}
    componentAppliedLinearAccClamp=<100 100 100>
    componentLinearAccScale=<0.15 0.15 0.15>
    componentLinearVelScale=<0.15 0.15 0.15>
    gravityOverride=<0.75 0.75 0.75>
    physicsBodyDefinitions [{
      sphereCollisionRadius=0
      boundBone {boneName=dyn_petals_back_parent_04}
      boxExtents=<4 4 2>
      constraintSetup {angularLimitsMax=<0 15 15>, angularLimitsMin=<0 -10 -10>, angularTarget=<0 0 0>}
      localJointOffset=<1 0 0>
     }]
    relativeSpaceBone {boneName=spine_05}
   }
   n15 = Anim Dynamics - Bone: dyn_petals_back_parent_07(ComponentPose=@n14, Alpha=1, ExternalForce="0,0,0") {
    angularDampingOverride=0.9999
    bGravityOverrideInSimSpace=true
    bOverrideAngularDamping=true
    bOverrideLinearDamping=true
    bUseGravityOverride=true
    linearDampingOverride=0.9999
    simulationSpace=RootRelative
    alphaBoolBlend {blendOption=Linear}
    boundBone {boneName=dyn_petals_back_parent_07}
    componentAppliedLinearAccClamp=<100 100 100>
    componentLinearAccScale=<0.15 0.15 0.15>
    componentLinearVelScale=<0.15 0.15 0.15>
    gravityOverride=<0.75 0.75 0.75>
    physicsBodyDefinitions [{
      sphereCollisionRadius=0
      boundBone {boneName=dyn_petals_back_parent_07}
      boxExtents=<4 4 2>
      constraintSetup {angularLimitsMax=<0 15 15>, angularLimitsMin=<0 -10 -10>, angularTarget=<0 0 0>}
      localJointOffset=<1 0 0>
     }]
    relativeSpaceBone {boneName=spine_05}
   }
   n16 = Anim Dynamics - Bone: dyn_petals_back_parent_08(ComponentPose=@n15, Alpha=1, ExternalForce="0,0,0") {
    angularDampingOverride=0.9999
    bGravityOverrideInSimSpace=true
    bOverrideAngularDamping=true
    bOverrideLinearDamping=true
    bUseGravityOverride=true
    linearDampingOverride=0.9999
    simulationSpace=RootRelative
    alphaBoolBlend {blendOption=Linear}
    boundBone {boneName=dyn_petals_back_parent_08}
    componentAppliedLinearAccClamp=<100 100 100>
    componentLinearAccScale=<0.15 0.15 0.15>
    componentLinearVelScale=<0.15 0.15 0.15>
    gravityOverride=<0.75 0.75 0.75>
    physicsBodyDefinitions [{
      sphereCollisionRadius=0
      boundBone {boneName=dyn_petals_back_parent_08}
      boxExtents=<4 4 2>
      constraintSetup {angularLimitsMax=<0 15 15>, angularLimitsMin=<0 -10 -10>, angularTarget=<0 0 0>}
      localJointOffset=<1 0 0>
     }]
    relativeSpaceBone {boneName=spine_05}
   }
   n17 = Anim Dynamics - Bone: dyn_petals_back_parent_09(ComponentPose=@n16, Alpha=1, ExternalForce="0,0,0") {
    angularDampingOverride=0.9999
    bGravityOverrideInSimSpace=true
    bOverrideAngularDamping=true
    bOverrideLinearDamping=true
    bUseGravityOverride=true
    linearDampingOverride=0.9999
    simulationSpace=RootRelative
    alphaBoolBlend {blendOption=Linear}
    boundBone {boneName=dyn_petals_back_parent_09}
    componentAppliedLinearAccClamp=<100 100 100>
    componentLinearAccScale=<0.15 0.15 0.15>
    componentLinearVelScale=<0.15 0.15 0.15>
    gravityOverride=<0.75 0.75 0.75>
    physicsBodyDefinitions [{
      sphereCollisionRadius=0
      boundBone {boneName=dyn_petals_back_parent_09}
      boxExtents=<4 4 2>
      constraintSetup {angularLimitsMax=<0 15 15>, angularLimitsMin=<0 -10 -10>, angularTarget=<0 0 0>}
      localJointOffset=<1 0 0>
     }]
    relativeSpaceBone {boneName=spine_05}
   }
   n18 = Anim Dynamics - Bone: dyn_petals_back_parent_10(ComponentPose=@n17, Alpha=1, ExternalForce="0,0,0") {
    angularDampingOverride=0.9999
    bGravityOverrideInSimSpace=true
    bOverrideAngularDamping=true
    bOverrideLinearDamping=true
    bUseGravityOverride=true
    linearDampingOverride=0.9999
    simulationSpace=RootRelative
    alphaBoolBlend {blendOption=Linear}
    boundBone {boneName=dyn_petals_back_parent_10}
    componentAppliedLinearAccClamp=<100 100 100>
    componentLinearAccScale=<0.15 0.15 0.15>
    componentLinearVelScale=<0.15 0.15 0.15>
    gravityOverride=<0.75 0.75 0.75>
    physicsBodyDefinitions [{
      sphereCollisionRadius=0
      boundBone {boneName=dyn_petals_back_parent_10}
      boxExtents=<4 4 2>
      constraintSetup {angularLimitsMax=<0 15 15>, angularLimitsMin=<0 -10 -10>, angularTarget=<0 0 0>}
      localJointOffset=<1 0 0>
     }]
    relativeSpaceBone {boneName=spine_05}
   }
   n19 = Anim Dynamics - Bone: dyn_petals_front_parent_01(ComponentPose=@n18, Alpha=1, ExternalForce="0,0,0") {
    angularDampingOverride=0.9999
    bGravityOverrideInSimSpace=true
    bOverrideAngularDamping=true
    bOverrideLinearDamping=true
    bUseGravityOverride=true
    linearDampingOverride=0.9999
    simulationSpace=RootRelative
    alphaBoolBlend {blendOption=Linear}
    alphaScaleBiasClamp {bias=0.2, scale=0.8}
    boundBone {boneName=dyn_petals_front_parent_01}
    componentAppliedLinearAccClamp=<100 100 100>
    componentLinearAccScale=<0.15 0.15 0.15>
    componentLinearVelScale=<0.15 0.15 0.15>
    gravityOverride=<1 0.85 0.85>
    physicsBodyDefinitions [{
      sphereCollisionRadius=0
      boundBone {boneName=dyn_petals_front_parent_01}
      boxExtents=<4 4 2>
      constraintSetup {angularLimitsMax=<0 20 20>, angularLimitsMin=<0 -10 -10>, angularTarget=<0 0 0>}
      localJointOffset=<1 0 0>
     }]
    relativeSpaceBone {boneName=root}
   }
   n20 = Anim Dynamics - Bone: dyn_petals_front_parent_02(ComponentPose=@n19, Alpha=1, ExternalForce="0,0,0") {
    angularDampingOverride=0.9999
    bGravityOverrideInSimSpace=true
    bOverrideAngularDamping=true
    bOverrideLinearDamping=true
    bUseGravityOverride=true
    linearDampingOverride=0.9999
    simulationSpace=RootRelative
    alphaBoolBlend {blendOption=Linear}
    alphaScaleBiasClamp {bias=0.2, scale=0.8}
    boundBone {boneName=dyn_petals_front_parent_02}
    componentAppliedLinearAccClamp=<100 100 100>
    componentLinearAccScale=<0.15 0.15 0.15>
    componentLinearVelScale=<0.15 0.15 0.15>
    gravityOverride=<1 0.85 0.85>
    physicsBodyDefinitions [{
      sphereCollisionRadius=0
      boundBone {boneName=dyn_petals_front_parent_02}
      boxExtents=<4 4 2>
      constraintSetup {angularLimitsMax=<0 20 20>, angularLimitsMin=<0 -10 -10>, angularTarget=<0 0 0>}
      localJointOffset=<1 0 0>
     }]
    relativeSpaceBone {boneName=root}
   }
   n21 = Anim Dynamics - Bone: dyn_petals_front_parent_03(ComponentPose=@n20, Alpha=1, ExternalForce="0,0,0") {
    angularDampingOverride=0.9999
    bGravityOverrideInSimSpace=true
    bOverrideAngularDamping=true
    bOverrideLinearDamping=true
    bUseGravityOverride=true
    linearDampingOverride=0.9999
    simulationSpace=RootRelative
    alphaBoolBlend {blendOption=Linear}
    alphaScaleBiasClamp {bias=0.2, scale=0.8}
    boundBone {boneName=dyn_petals_front_parent_03}
    componentAppliedLinearAccClamp=<100 100 100>
    componentLinearAccScale=<0.15 0.15 0.15>
    componentLinearVelScale=<0.15 0.15 0.15>
    gravityOverride=<1 0.85 0.85>
    physicsBodyDefinitions [{
      sphereCollisionRadius=0
      boundBone {boneName=dyn_petals_front_parent_03}
      boxExtents=<6 10 2>
      constraintSetup {angularLimitsMax=<0 20 20>, angularLimitsMin=<0 -10 -10>, angularTarget=<0 0 0>}
      localJointOffset=<1 0 0>
     }]
    relativeSpaceBone {boneName=spine_05}
   }
   n22 = Anim Dynamics - Bone: dyn_petals_front_parent_06(ComponentPose=@n21, Alpha=1, ExternalForce="0,0,0") {
    angularDampingOverride=0.9999
    bGravityOverrideInSimSpace=true
    bOverrideAngularDamping=true
    bOverrideLinearDamping=true
    bUseGravityOverride=true
    linearDampingOverride=0.9999
    simulationSpace=RootRelative
    alphaBoolBlend {blendOption=Linear}
    alphaScaleBiasClamp {bias=0.2, scale=0.8}
    boundBone {boneName=dyn_petals_front_parent_06}
    componentAppliedLinearAccClamp=<100 100 100>
    componentLinearAccScale=<0.15 0.15 0.15>
    componentLinearVelScale=<0.15 0.15 0.15>
    gravityOverride=<1 0.85 0.85>
    physicsBodyDefinitions [{
      sphereCollisionRadius=0
      boundBone {boneName=dyn_petals_front_parent_06}
      boxExtents=<4 4 2>
      constraintSetup {angularLimitsMax=<0 20 20>, angularLimitsMin=<0 -10 -10>, angularTarget=<0 0 0>}
      localJointOffset=<1 0 0>
     }]
    relativeSpaceBone {boneName=root}
   }
   n23 = Anim Dynamics - Bone: dyn_petals_front_parent_05(ComponentPose=@n22, Alpha=1, ExternalForce="0,0,0") {
    angularDampingOverride=0.9999
    bGravityOverrideInSimSpace=true
    bOverrideAngularDamping=true
    bOverrideLinearDamping=true
    bUseGravityOverride=true
    linearDampingOverride=0.9999
    simulationSpace=RootRelative
    alphaBoolBlend {blendOption=Linear}
    alphaScaleBiasClamp {bias=0.2, scale=0.8}
    boundBone {boneName=dyn_petals_front_parent_05}
    componentAppliedLinearAccClamp=<100 100 100>
    componentLinearAccScale=<0.15 0.15 0.15>
    componentLinearVelScale=<0.15 0.15 0.15>
    gravityOverride=<1 0.85 0.85>
    physicsBodyDefinitions [{
      sphereCollisionRadius=0
      boundBone {boneName=dyn_petals_front_parent_05}
      boxExtents=<4 4 2>
      constraintSetup {angularLimitsMax=<0 20 20>, angularLimitsMin=<0 -10 -10>, angularTarget=<0 0 0>}
      localJointOffset=<1 0 0>
     }]
    relativeSpaceBone {boneName=root}
   }
   n24 = Anim Dynamics - Bone: dyn_petals_front_parent_04(ComponentPose=@n23, Alpha=1, ExternalForce="0,0,0") {
    angularDampingOverride=0.9999
    bGravityOverrideInSimSpace=true
    bOverrideAngularDamping=true
    bOverrideLinearDamping=true
    bUseGravityOverride=true
    linearDampingOverride=0.9999
    simulationSpace=RootRelative
    alphaBoolBlend {blendOption=Linear}
    alphaScaleBiasClamp {bias=0.2, scale=0.8}
    boundBone {boneName=dyn_petals_front_parent_04}
    componentAppliedLinearAccClamp=<100 100 100>
    componentLinearAccScale=<0.15 0.15 0.15>
    componentLinearVelScale=<0.15 0.15 0.15>
    gravityOverride=<1 0.85 0.85>
    physicsBodyDefinitions [{
      sphereCollisionRadius=0
      boundBone {boneName=dyn_petals_front_parent_04}
      boxExtents=<4 4 2>
      constraintSetup {angularLimitsMax=<0 20 20>, angularLimitsMin=<0 -10 -10>, angularTarget=<0 0 0>}
      localJointOffset=<1 0 0>
     }]
    relativeSpaceBone {boneName=root}
   }
   n25 = Anim Dynamics - Bone: dyn_petals_front_parent_07(ComponentPose=@n24, Alpha=1, ExternalForce="0,0,0") {
    angularDampingOverride=0.9999
    bGravityOverrideInSimSpace=true
    bOverrideAngularDamping=true
    bOverrideLinearDamping=true
    bUseGravityOverride=true
    linearDampingOverride=0.9999
    simulationSpace=RootRelative
    alphaBoolBlend {blendOption=Linear}
    alphaScaleBiasClamp {bias=0.2, scale=0.8}
    boundBone {boneName=dyn_petals_front_parent_07}
    componentAppliedLinearAccClamp=<100 100 100>
    componentLinearAccScale=<0.15 0.15 0.15>
    componentLinearVelScale=<0.15 0.15 0.15>
    gravityOverride=<1 0.85 0.85>
    physicsBodyDefinitions [{
      sphereCollisionRadius=0
      boundBone {boneName=dyn_petals_front_parent_07}
      boxExtents=<4 4 2>
      constraintSetup {angularLimitsMax=<0 20 20>, angularLimitsMin=<0 -10 -10>, angularTarget=<0 0 0>}
      localJointOffset=<1 0 0>
     }]
    relativeSpaceBone {boneName=root}
   }
   n26 = Anim Dynamics - Bone: dyn_petals_front_parent_08(ComponentPose=@n25, Alpha=1, ExternalForce="0,0,0") {
    angularDampingOverride=0.9999
    bGravityOverrideInSimSpace=true
    bOverrideAngularDamping=true
    bOverrideLinearDamping=true
    bUseGravityOverride=true
    linearDampingOverride=0.9999
    simulationSpace=RootRelative
    alphaBoolBlend {blendOption=Linear}
    alphaScaleBiasClamp {bias=0.2, scale=0.8}
    boundBone {boneName=dyn_petals_front_parent_08}
    componentAppliedLinearAccClamp=<100 100 100>
    componentLinearAccScale=<0.15 0.15 0.15>
    componentLinearVelScale=<0.15 0.15 0.15>
    gravityOverride=<1 0.85 0.85>
    physicsBodyDefinitions [{
      sphereCollisionRadius=0
      boundBone {boneName=dyn_petals_front_parent_08}
      boxExtents=<4 4 2>
      constraintSetup {angularLimitsMax=<0 20 20>, angularLimitsMin=<0 -10 -10>, angularTarget=<0 0 0>}
      localJointOffset=<1 0 0>
     }]
    relativeSpaceBone {boneName=root}
   }
   n27 = Get Took_Damage_Float() -> (Took_Damage_Float:double)
   n28 = "Transform (Modify) Bone - Bone: dyn_head_flower_01"(ComponentPose=@n26, Alpha=@n27, Translation="0,0,0", Rotation="-85,0,0", Scale="1,1,1") {
    rotationMode=BMM_Additive
    rotationSpace=BCS_BoneSpace
    alphaBoolBlend {blendOption=Linear}
    alphaScaleBiasClamp {bInterpResult=true, interpSpeedDecreasing=4, interpSpeedIncreasing=6}
    boneToModify {boneName=dyn_head_flower_01}
    rotation {pitch=-85}
   }
   n29 = "Transform (Modify) Bone - Bone: dyn_head_flower_02"(ComponentPose=@n28, Alpha=@n27, Translation="0,0,0", Rotation="-85,0.000,0.000", Scale="1,1,1") {
    rotationMode=BMM_Additive
    rotationSpace=BCS_BoneSpace
    alphaBoolBlend {blendOption=Linear}
    alphaScaleBiasClamp {bInterpResult=true, interpSpeedDecreasing=4, interpSpeedIncreasing=6}
    boneToModify {boneName=dyn_head_flower_02}
    rotation {pitch=-85}
   }
   n30 = "Transform (Modify) Bone - Bone: dyn_head_flower_03"(ComponentPose=@n29, Alpha=@n27, Translation="0,0,0", Rotation="-85,0.000,0.000", Scale="1,1,1") {
    rotationMode=BMM_Additive
    rotationSpace=BCS_BoneSpace
    alphaBoolBlend {blendOption=Linear}
    alphaScaleBiasClamp {bInterpResult=true, interpSpeedDecreasing=4, interpSpeedIncreasing=6}
    boneToModify {boneName=dyn_head_flower_03}
    rotation {pitch=-85}
   }
   n31 = "Transform (Modify) Bone - Bone: dyn_head_flower_04"(ComponentPose=@n30, Alpha=@n27, Translation="0,0,0", Rotation="0.000,60,0.000", Scale="1,1,1") {
    rotationMode=BMM_Additive
    rotationSpace=BCS_BoneSpace
    alphaBoolBlend {blendOption=Linear}
    alphaScaleBiasClamp {bInterpResult=true, interpSpeedDecreasing=4, interpSpeedIncreasing=6}
    boneToModify {boneName=dyn_head_flower_04}
    rotation {yaw=60}
   }
   Output Pose(Result=Component To Local(ComponentPose=@n31)) {layerGroup=None}
   Save cached pose 'HeadLockCache'(Pose=Copy Pose From Mesh() {bCopyCurves=true, bUseAttachedParent=true}) {cacheName=HeadLockCache}
  ]}
 EventGraph {nodes [
   n1 = Event BlueprintUpdateAnimation()
   > n2 = Set PetelsCurrentColor(PetelsCurrentColor:double=FInterpTo(
     Current=Get PetelsCurrentColor() -> (PetelsCurrentColor:double)
     Target=Get PetelsTargetColor() -> (PetelsTargetColor:double)
     DeltaTime=@n1.DeltaTimeX
     InterpSpeed=2
    )) -> (Output_Get:double)
   > SetScalarParameterValueOnMaterials!(execute, self:Mesh Component Object Reference=GetPartSkeletalMeshComponent(), ParameterName:Name=MaterialCurveA, ParameterValue:float=@n2) -> (then)
   Event BlueprintProcessAnimInputEvent()
   > n3 = Sequence()
   > Do Once()
   > n4 = Cast To FortPlayerPawn(Object=TryGetFortPawn())
   > n5 = WatchCosmeticStat!(
    execute:Exec
    Pawn:Fort Player Pawn Object Reference=@n4
    StatTag:Fort Stat Manager Tag Structure=(TagName="StatManager.DamageReceived")
    DefaultValue:int=0
    bBroadcastInitialValue:bool=true
    InitialValueBroadcastDelay:float=-1
    ChangeValueBroadcastDelay:float=-1
    bConsumeWaitingNotifications:bool=true
   ) -> (then:Exec, OnStatChanged:Exec, StatName:Name, StatValue:int) {is=title K2Node_AsyncAction}
   n6 = "integer > integer"(A=@n5.StatValue, B=0) {is=title K2Node_CallFunction}
   Branch(execute=@n5.OnStatChanged, Condition=@n6)
   > n7 = Set Took_Damage(Took_Damage:bool=@n6) -> (Output_Get:bool)
   > Set Took_Damage_Float(Took_Damage_Float:double="To Float (Boolean)"(InBool=@n7)) -> (Output_Get:double)
   n8 = Get PetelsTargetColor()
   > Set PetelsTargetColor(PetelsTargetColor:double="Clamp (Float)"(Value=float + float(A=@n8, B=0.334) {is=title K2Node_CommutativeAssociativeBinaryOperator}, Min=0, Max=1)) -> (Output_Get:double)
   > RetriggerableDelay(Duration=6)
   > Set Took_Damage_Float(Took_Damage_Float=0)
   > Set PetelsTargetColor(PetelsTargetColor=0)
   n9 = Get Took_Damage_Float()
   Set Dynamics_Alpha(
    execute=@n3.then_1
    Dynamics_Alpha:double=SelectFloat(A=1, B=0, bPickA="float <= float"!(A:double=@n9, B:double=0) -> (ReturnValue:bool) {is=title K2Node_CallFunction})
   ) -> (Output_Get:double)
   > Delay(Duration=6)
   > Set Dynamics_Alpha(Dynamics_Alpha=1)
  ]}
]
variables [
 Active_Kill {type=boolean}
 Did_Kill {type=boolean}
 Dynamics_Alpha {type=number}
 Kill_Alpha {type=number, default=1}
 KillCount {type=number}
 Last_KillCount {type=number}
 MaxKillCount {type=number, default=5}
 PetelsCurrentColor {type=number}
 PetelsTargetColor {type=number}
 Skel_Mesh {type=SkeletalMeshComponent}
 Took_Damage {type=boolean}
 Took_Damage_Float {type=number}
]
